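# Control Rig for the M_MED_ClimbSpill male medium player body cosmetic, with functions that measure joint poses, likely for corrective deformation.
# Each function runs Spherical Pose Readers on a driver item, two relative to its parent, at offset rotations and falloffs.
# They return weights for directional poses with pre/post variants, plus crouch, aim and bend poses.
# Per-pose Viz flags toggle debug cones and on-screen value prints.
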ControlRigBlueprint /BRCosmetics/Characters/Player/Male/Medium/Bodies/M_MED_ClimbSpill/Meshes/M_MED_ClimbSpill_CtrlRig
graphs [
 New Function_ContainedGraph {nodes [
   n1 = RigVMFunctionEntryNode()
   > RigVMVariableNode(Value=/Script/ControlRig.RigUnit_HierarchyGetParent(bDefaultParent=True, Child=@n1.Driver Item) {is=struct RigVMUnitNode}.Parent) {variable=Driver Item Parent}
   To_String = RigVMDispatchNode(Value=@n1.Driver Item.Name) {template="DISPATCH_RigDispatch_ToString(in Value,out Result)"}
   > RigVMVariableNode(Value=@To_String.Result) {variable=Driver Item Name}
   > RigVMFunction_Sequence = RigVMAggregateNode()
   RigVMFunctionReturnNode(
    ExecuteContext=@RigVMFunction_Sequence.F
    Fr Post=RigVMVariableNode() {variable=Result Fr Post}.Value
    Crouch Pre=RigVMVariableNode() {variable=Result Crouch Pre}.Value
    Crouch Post=RigVMVariableNode() {variable=Result Crouch Post}.Value
    Crouch Ot=RigVMVariableNode() {variable=Result Crouch Ot}.Value
    Crouch Pre Ot=RigVMVariableNode() {variable=Result Crouch Pre Ot}.Value
    Crouch Post Ot=RigVMVariableNode() {variable=Result Crouch Post Ot}.Value
    Ba=RigVMVariableNode() {variable=Result Ba}.Value
    Ba Pre=RigVMVariableNode() {variable=Result Ba Pre}.Value
    Ba Post=RigVMVariableNode() {variable=Result Ba Post}.Value
    In=RigVMVariableNode() {variable=Result In}.Value
    In Pre=RigVMVariableNode() {variable=Result In Pre}.Value
    In Post=RigVMVariableNode() {variable=Result In Post}.Value
    Ot=RigVMVariableNode() {variable=Result Ot}.Value
    Ot Pre=RigVMVariableNode() {variable=Result Ot Pre}.Value
    Ot Post=RigVMVariableNode() {variable=Result Ot Post}.Value
    Fr=RigVMVariableNode() {variable=Result Fr}.Value
    Crouch=RigVMVariableNode() {variable=Result Crouch}.Value
    Fr Pre=RigVMVariableNode() {variable=Result Fr Pre}.Value
   )
   n2 = &RigUnit_SphericalPoseReader(
    ActiveRegionScaleFactors=(PositiveWidth=1,NegativeWidth=1,PositiveHeight=1,NegativeHeight=1)
    ActiveRegionSize=0.001
    Debug=(bDrawDebug=false,bDraw2D=False,bDrawLocalAxes=False,DebugScale=25,DebugSegments=20,DebugThickness=0.05)
    DriverAxis=(X=1,Y=0,Z=0)
    FalloffRegionScaleFactors=(PositiveWidth=1,NegativeWidth=1,PositiveHeight=1,NegativeHeight=1)
    FalloffSize=0.5
    FlipHeightScaling=False
    FlipWidthScaling=False
    RotationOffset=(X=-90,Y=0,Z=0)
    Debug.bDrawDebug=RigVMVariableNode() {variable=Fr Viz}.Value
    DriverItem=RigVMVariableNode() {variable=Driver Item}.Value
    OptionalParentItem=RigVMVariableNode() {variable=Driver Item Parent}.Value
    ExecutePin=@RigVMFunction_Sequence.A
   )
   > n3 = RigVMVariableNode(Value=@n2.OutputParam) {variable=Result Fr}
   RigVMFunction_ControlFlowBranch = &RigVMFunction_ControlFlowBranch(ExecuteContext=@n3.ExecuteContext, Condition=RigVMVariableNode() {variable=Fr Viz}.Value)
   n4 = &RigUnit_SphericalPoseReader(
    ActiveRegionScaleFactors=(PositiveWidth=1,NegativeWidth=1,PositiveHeight=1,NegativeHeight=1)
    ActiveRegionSize=0.001
    Debug=(bDrawDebug=false,bDraw2D=False,bDrawLocalAxes=False,DebugScale=25,DebugSegments=20,DebugThickness=0.05)
    DriverAxis=(X=1,Y=0,Z=0)
    FalloffRegionScaleFactors=(PositiveWidth=1,NegativeWidth=1,PositiveHeight=2,NegativeHeight=2)
    FalloffSize=0.25
    FlipHeightScaling=False
    FlipWidthScaling=False
    RotationOffset=(X=-90,Y=0,Z=45)
    ExecutePin=@RigVMFunction_ControlFlowBranch.Completed
    DriverItem=RigVMVariableNode() {variable=Driver Item}.Value
    OptionalParentItem=RigVMVariableNode() {variable=Driver Item Parent}.Value
    Debug.bDrawDebug=RigVMVariableNode() {variable=Fr Pre Viz}.Value
   )
   > n5 = RigVMVariableNode(Value=@n4.OutputParam) {variable=Result Fr Pre}
   n6 = &RigVMFunction_ControlFlowBranch(Condition=RigVMVariableNode() {variable=Fr Pre Viz}.Value, ExecuteContext=@n5.ExecuteContext)
   n7 = &RigUnit_SphericalPoseReader(
    ActiveRegionScaleFactors=(PositiveWidth=1,NegativeWidth=1,PositiveHeight=1,NegativeHeight=1)
    ActiveRegionSize=0.001
    Debug=(bDrawDebug=false,bDraw2D=False,bDrawLocalAxes=False,DebugScale=25,DebugSegments=20,DebugThickness=0.05)
    DriverAxis=(X=1,Y=0,Z=0)
    FalloffRegionScaleFactors=(PositiveWidth=0.5,NegativeWidth=0.5,PositiveHeight=1,NegativeHeight=1)
    FalloffSize=0.5
    FlipHeightScaling=False
    FlipWidthScaling=False
    RotationOffset=(X=-90,Y=0,Z=-45)
    ExecutePin=@n6.Completed
    Debug.bDrawDebug=RigVMVariableNode() {variable=Fr Post Viz}.Value
    DriverItem=RigVMVariableNode() {variable=Driver Item}.Value
    OptionalParentItem=RigVMVariableNode() {variable=Driver Item Parent}.Value
   )
   > n8 = RigVMVariableNode(Value=@n7.OutputParam) {variable=Result Fr Post}
   n9 = &RigVMFunction_ControlFlowBranch(Condition=RigVMVariableNode() {variable=Fr Post Viz}.Value, ExecuteContext=@n8.ExecuteContext)
   n10 = &RigUnit_SphericalPoseReader(
    ActiveRegionScaleFactors=(PositiveWidth=1,NegativeWidth=1,PositiveHeight=1,NegativeHeight=1)
    ActiveRegionSize=0.001
    Debug=(bDrawDebug=true,bDraw2D=False,bDrawLocalAxes=False,DebugScale=25,DebugSegments=20,DebugThickness=0.05)
    DriverAxis=(X=1,Y=0,Z=0)
    FalloffRegionScaleFactors=(PositiveWidth=1,NegativeWidth=1,PositiveHeight=0.125,NegativeHeight=0.125)
    FalloffSize=0.5
    FlipHeightScaling=False
    FlipWidthScaling=False
    RotationOffset=(X=-75,Y=0,Z=0)
    ExecutePin=@n9.Completed
    DriverItem=RigVMVariableNode() {variable=Driver Item}.Value
    Debug.bDrawDebug=RigVMVariableNode() {variable=Crouch Viz}.Value
    OptionalParentItem=RigVMVariableNode() {variable=Driver Item Parent}.Value
   )
   > n11 = RigVMVariableNode(Value=@n10.OutputParam) {variable=Result Crouch}
   n12 = &RigVMFunction_ControlFlowBranch(Condition=RigVMVariableNode() {variable=Crouch Viz}.Value, ExecuteContext=@n11.ExecuteContext)
   n13 = &RigUnit_SphericalPoseReader(
    ActiveRegionScaleFactors=(PositiveWidth=1,NegativeWidth=1,PositiveHeight=1,NegativeHeight=1)
    ActiveRegionSize=0.001
    Debug=(bDrawDebug=true,bDraw2D=False,bDrawLocalAxes=False,DebugScale=25,DebugSegments=20,DebugThickness=0.05)
    DriverAxis=(X=1,Y=0,Z=0)
    FalloffRegionScaleFactors=(PositiveWidth=0.4825,NegativeWidth=0.5,PositiveHeight=0.075,NegativeHeight=0.075)
    FalloffSize=0.5
    FlipHeightScaling=False
    FlipWidthScaling=False
    RotationOffset=(X=-101.3223,Y=-189.762,Z=42.3664)
    ExecutePin=@n12.Completed
    DriverItem=RigVMVariableNode() {variable=Driver Item}.Value
    OptionalParentItem=RigVMVariableNode() {variable=Driver Item Parent}.Value
    Debug.bDrawDebug=RigVMVariableNode() {variable=Crouch Pre Viz}.Value
   )
   > n14 = RigVMVariableNode(Value=@n13.OutputParam) {variable=Result Crouch Pre}
   n15 = &RigVMFunction_ControlFlowBranch(ExecuteContext=@n14.ExecuteContext, Condition=RigVMVariableNode() {variable=Crouch Pre Viz}.Value)
   n16 = &RigUnit_SphericalPoseReader(
    ActiveRegionScaleFactors=(PositiveWidth=1,NegativeWidth=1,PositiveHeight=1,NegativeHeight=1)
    ActiveRegionSize=0.001
    Debug=(bDrawDebug=true,bDraw2D=False,bDrawLocalAxes=False,DebugScale=25,DebugSegments=20,DebugThickness=0.05)
    DriverAxis=(X=1,Y=0,Z=0)
    FalloffRegionScaleFactors=(PositiveWidth=0.5,NegativeWidth=0.5,PositiveHeight=0.075,NegativeHeight=0.075)
    FalloffSize=0.5
    FlipHeightScaling=False
    FlipWidthScaling=False
    RotationOffset=(X=-101.5496,Y=-169.7841,Z=-42.8133)
    ExecutePin=@n15.Completed
    DriverItem=RigVMVariableNode() {variable=Driver Item}.Value
    OptionalParentItem=RigVMVariableNode() {variable=Driver Item Parent}.Value
    Debug.bDrawDebug=RigVMVariableNode() {variable=Crouch Post Viz}.Value
   )
   > n17 = RigVMVariableNode(Value=@n16.OutputParam) {variable=Result Crouch Post}
   n18 = &RigVMFunction_ControlFlowBranch(ExecuteContext=@n17.ExecuteContext, Condition=RigVMVariableNode() {variable=Crouch Post Viz}.Value)
   n19 = &RigUnit_SphericalPoseReader(
    ActiveRegionScaleFactors=(PositiveWidth=1,NegativeWidth=1,PositiveHeight=1,NegativeHeight=1)
    ActiveRegionSize=0.001
    Debug=(bDrawDebug=true,bDraw2D=False,bDrawLocalAxes=False,DebugScale=25,DebugSegments=20,DebugThickness=0.05)
    DriverAxis=(X=1,Y=0,Z=0)
    FalloffRegionScaleFactors=(PositiveWidth=1,NegativeWidth=1,PositiveHeight=0.666,NegativeHeight=0.333)
    FalloffSize=0.5
    FlipHeightScaling=False
    FlipWidthScaling=False
    RotationOffset=(X=-60,Y=0,Z=0)
    ExecutePin=@n18.Completed
    DriverItem=RigVMVariableNode() {variable=Driver Item}.Value
    OptionalParentItem=RigVMVariableNode() {variable=Driver Item Parent}.Value
    Debug.bDrawDebug=RigVMVariableNode() {variable=Crouch Ot Viz}.Value
   )
   > n20 = RigVMVariableNode(Value=@n19.OutputParam) {variable=Result Crouch Ot}
   n21 = &RigVMFunction_ControlFlowBranch(Condition=RigVMVariableNode() {variable=Crouch Ot Viz}.Value, ExecuteContext=@n20.ExecuteContext)
   n22 = &RigUnit_SphericalPoseReader(
    ActiveRegionScaleFactors=(PositiveWidth=1,NegativeWidth=1,PositiveHeight=1,NegativeHeight=1)
    ActiveRegionSize=0.001
    Debug=(bDrawDebug=true,bDraw2D=False,bDrawLocalAxes=False,DebugScale=25,DebugSegments=20,DebugThickness=0.05)
    DriverAxis=(X=1,Y=0,Z=0)
    FalloffRegionScaleFactors=(PositiveWidth=0.5,NegativeWidth=0.5,PositiveHeight=0.2,NegativeHeight=0.4)
    FalloffSize=0.5
    FlipHeightScaling=False
    FlipWidthScaling=False
    RotationOffset=(X=-110.6339,Y=-201.1783,Z=39.9802)
    ExecutePin=@n21.Completed
    DriverItem=RigVMVariableNode() {variable=Driver Item}.Value
    OptionalParentItem=RigVMVariableNode() {variable=Driver Item Parent}.Value
    Debug.bDrawDebug=RigVMVariableNode() {variable=Crouch Ot Pre Viz}.Value
   )
   > n23 = RigVMVariableNode(Value=@n22.OutputParam) {variable=Result Crouch Pre Ot}
   n24 = &RigVMFunction_ControlFlowBranch(Condition=RigVMVariableNode() {variable=Crouch Ot Pre Viz}.Value, ExecuteContext=@n23.ExecuteContext)
   n25 = &RigUnit_SphericalPoseReader(
    ActiveRegionScaleFactors=(PositiveWidth=1,NegativeWidth=1,PositiveHeight=1,NegativeHeight=1)
    ActiveRegionSize=0.001
    Debug=(bDrawDebug=true,bDraw2D=False,bDrawLocalAxes=False,DebugScale=25,DebugSegments=20,DebugThickness=0.05)
    DriverAxis=(X=1,Y=0,Z=0)
    FalloffRegionScaleFactors=(PositiveWidth=0.5,NegativeWidth=0.5,PositiveHeight=0.2,NegativeHeight=0.4)
    FalloffSize=0.5
    FlipHeightScaling=False
    FlipWidthScaling=False
    RotationOffset=(X=-112.4531,Y=-160.3786,Z=-41.3069)
    ExecutePin=@n24.Completed
    DriverItem=RigVMVariableNode() {variable=Driver Item}.Value
    OptionalParentItem=RigVMVariableNode() {variable=Driver Item Parent}.Value
    Debug.bDrawDebug=RigVMVariableNode() {variable=Crouch Ot Post Viz}.Value
   )
   > n26 = RigVMVariableNode(Value=@n25.OutputParam) {variable=Result Crouch Post Ot}
   n27 = &RigVMFunction_ControlFlowBranch(ExecuteContext=@n26.ExecuteContext, Condition=RigVMVariableNode() {variable=Crouch Ot Post Viz}.Value)
   RigVMDispatchNode(
    Enabled=True
    ScreenColor=(R=1,G=1,B=1,A=1)
    ScreenDuration=0.05
    Prefix=/Script/RigVM.RigVMFunction_StringConcat(B=" - CROUCH OT POST:   ", A=RigVMVariableNode() {variable=Driver Item Name}.Value) {is=struct RigVMUnitNode, template="Concat::Execute(in A,in B,out Result)"}.Result
    Value=RigVMVariableNode() {variable=Result Crouch Post Ot}.Value
    ExecuteContext=@n27.True
   ) {template="DISPATCH_RigVMDispatch_Print(in Prefix,in Value,in Enabled,in ScreenDuration,in ScreenColor)"}
   RigVMDispatchNode(
    Enabled=True
    ScreenColor=(R=1,G=1,B=1,A=1)
    ScreenDuration=0.05
    Prefix=&RigVMFunction_StringConcat(B=" - CROUCH OT PRE:   ", A=RigVMVariableNode() {variable=Driver Item Name}.Value) {template="Concat::Execute(in A,in B,out Result)"}.Result
    Value=RigVMVariableNode() {variable=Result Crouch Pre Ot}.Value
    ExecuteContext=@n24.True
   ) {template="DISPATCH_RigVMDispatch_Print(in Prefix,in Value,in Enabled,in ScreenDuration,in ScreenColor)"}
   RigVMDispatchNode(
    Enabled=True
    ScreenColor=(R=1,G=1,B=1,A=1)
    ScreenDuration=0.05
    Prefix=&RigVMFunction_StringConcat(B=" - Crouch Ot:   ", A=RigVMVariableNode() {variable=Driver Item Name}.Value) {template="Concat::Execute(in A,in B,out Result)"}.Result
    Value=RigVMVariableNode() {variable=Result Crouch Ot}.Value
    ExecuteContext=@n21.True
   ) {template="DISPATCH_RigVMDispatch_Print(in Prefix,in Value,in Enabled,in ScreenDuration,in ScreenColor)"}
   RigVMDispatchNode(
    Enabled=True
    ScreenColor=(R=1,G=1,B=1,A=1)
    ScreenDuration=0.05
    Prefix=&RigVMFunction_StringConcat(B=" - CROUCH POST:   ", A=RigVMVariableNode() {variable=Driver Item Name}.Value) {template="Concat::Execute(in A,in B,out Result)"}.Result
    Value=RigVMVariableNode() {variable=Result Crouch Post}.Value
    ExecuteContext=@n18.True
   ) {template="DISPATCH_RigVMDispatch_Print(in Prefix,in Value,in Enabled,in ScreenDuration,in ScreenColor)"}
   RigVMDispatchNode(
    Enabled=True
    ScreenColor=(R=1,G=1,B=1,A=1)
    ScreenDuration=0.05
    Prefix=&RigVMFunction_StringConcat(B=" - CROUCH PRE:   ", A=RigVMVariableNode() {variable=Driver Item Name}.Value) {template="Concat::Execute(in A,in B,out Result)"}.Result
    Value=RigVMVariableNode() {variable=Result Crouch Pre}.Value
    ExecuteContext=@n15.True
   ) {template="DISPATCH_RigVMDispatch_Print(in Prefix,in Value,in Enabled,in ScreenDuration,in ScreenColor)"}
   RigVMDispatchNode(
    Enabled=True
    ScreenColor=(R=1,G=1,B=1,A=1)
    ScreenDuration=0.05
    Prefix=&RigVMFunction_StringConcat(B=" - Crouch:   ", A=RigVMVariableNode() {variable=Driver Item Name}.Value) {template="Concat::Execute(in A,in B,out Result)"}.Result
    Value=RigVMVariableNode() {variable=Result Crouch}.Value
    ExecuteContext=@n12.True
   ) {template="DISPATCH_RigVMDispatch_Print(in Prefix,in Value,in Enabled,in ScreenDuration,in ScreenColor)"}
   RigVMDispatchNode(
    Enabled=True
    ScreenColor=(R=1,G=1,B=1,A=1)
    ScreenDuration=0.05
    Prefix=&RigVMFunction_StringConcat(B=" - FR POST:   ", A=RigVMVariableNode() {variable=Driver Item Name}.Value) {template="Concat::Execute(in A,in B,out Result)"}.Result
    Value=RigVMVariableNode() {variable=Result Fr Post}.Value
    ExecuteContext=@n9.True
   ) {template="DISPATCH_RigVMDispatch_Print(in Prefix,in Value,in Enabled,in ScreenDuration,in ScreenColor)"}
   RigVMDispatchNode(
    Enabled=True
    ScreenColor=(R=1,G=1,B=1,A=1)
    ScreenDuration=0.05
    Prefix=&RigVMFunction_StringConcat(B=" - FR PRE:   ", A=RigVMVariableNode() {variable=Driver Item Name}.Value) {template="Concat::Execute(in A,in B,out Result)"}.Result
    Value=RigVMVariableNode() {variable=Result Fr Pre}.Value
    ExecuteContext=@n6.True
   ) {template="DISPATCH_RigVMDispatch_Print(in Prefix,in Value,in Enabled,in ScreenDuration,in ScreenColor)"}
   Concat = &RigVMFunction_StringConcat(B=" - FR:   ", A=RigVMVariableNode() {variable=Driver Item Name}.Value) {template="Concat::Execute(in A,in B,out Result)"}
   RigVMDispatchNode(
    Enabled=True
    ScreenColor=(R=1,G=1,B=1,A=1)
    ScreenDuration=0.05
    Prefix=@Concat.Result
    Value=RigVMVariableNode() {variable=Result Fr}.Value
    ExecuteContext=@RigVMFunction_ControlFlowBranch.True
   ) {template="DISPATCH_RigVMDispatch_Print(in Prefix,in Value,in Enabled,in ScreenDuration,in ScreenColor)"}
   n28 = &RigUnit_SphericalPoseReader(
    ActiveRegionScaleFactors=(PositiveWidth=1,NegativeWidth=1,PositiveHeight=1,NegativeHeight=1)
    ActiveRegionSize=0.001
    Debug=(bDrawDebug=false,bDraw2D=False,bDrawLocalAxes=False,DebugScale=25,DebugSegments=20,DebugThickness=0.05)
    DriverAxis=(X=1,Y=0,Z=0)
    FalloffRegionScaleFactors=(PositiveWidth=1,NegativeWidth=1,PositiveHeight=1,NegativeHeight=1)
    FalloffSize=0.5
    FlipHeightScaling=False
    FlipWidthScaling=False
    RotationOffset=(X=180,Y=0,Z=0)
    DriverItem=RigVMVariableNode() {variable=Driver Item}.Value
    OptionalParentItem=RigVMVariableNode() {variable=Driver Item Parent}.Value
    Debug.bDrawDebug=RigVMVariableNode() {variable=In Viz}.Value
    ExecutePin=@RigVMFunction_Sequence.C
   )
   > n29 = RigVMVariableNode(Value=@n28.OutputParam) {variable=Result In}
   n30 = &RigVMFunction_ControlFlowBranch(ExecuteContext=@n29.ExecuteContext, Condition=RigVMVariableNode() {variable=In Viz}.Value)
   n31 = &RigUnit_SphericalPoseReader(
    ActiveRegionScaleFactors=(PositiveWidth=1,NegativeWidth=1,PositiveHeight=1,NegativeHeight=1)
    ActiveRegionSize=0.001
    Debug=(bDrawDebug=false,bDraw2D=False,bDrawLocalAxes=False,DebugScale=25,DebugSegments=20,DebugThickness=0.05)
    DriverAxis=(X=1,Y=0,Z=0)
    FalloffRegionScaleFactors=(PositiveWidth=2,NegativeWidth=2,PositiveHeight=1,NegativeHeight=1)
    FalloffSize=0.25
    FlipHeightScaling=False
    FlipWidthScaling=False
    RotationOffset=(X=225,Y=0,Z=90)
    ExecutePin=@n30.Completed
    DriverItem=RigVMVariableNode() {variable=Driver Item}.Value
    OptionalParentItem=RigVMVariableNode() {variable=Driver Item Parent}.Value
    Debug.bDrawDebug=RigVMVariableNode() {variable=In Pre Viz}.Value
   )
   > n32 = RigVMVariableNode(Value=@n31.OutputParam) {variable=Result In Pre}
   n33 = &RigVMFunction_ControlFlowBranch(ExecuteContext=@n32.ExecuteContext, Condition=RigVMVariableNode() {variable=In Pre Viz}.Value)
   n34 = &RigUnit_SphericalPoseReader(
    ActiveRegionScaleFactors=(PositiveWidth=1,NegativeWidth=1,PositiveHeight=1,NegativeHeight=1)
    ActiveRegionSize=0.001
    Debug=(bDrawDebug=false,bDraw2D=False,bDrawLocalAxes=False,DebugScale=25,DebugSegments=20,DebugThickness=0.05)
    DriverAxis=(X=1,Y=0,Z=0)
    FalloffRegionScaleFactors=(PositiveWidth=2,NegativeWidth=2,PositiveHeight=1,NegativeHeight=1)
    FalloffSize=0.25
    FlipHeightScaling=False
    FlipWidthScaling=False
    RotationOffset=(X=135,Y=0,Z=90)
    ExecutePin=@n33.Completed
    DriverItem=RigVMVariableNode() {variable=Driver Item}.Value
    OptionalParentItem=RigVMVariableNode() {variable=Driver Item Parent}.Value
    Debug.bDrawDebug=RigVMVariableNode() {variable=In Post Viz}.Value
   )
   > n35 = RigVMVariableNode(Value=@n34.OutputParam) {variable=Result In Post}
   n36 = &RigVMFunction_ControlFlowBranch(ExecuteContext=@n35.ExecuteContext, Condition=RigVMVariableNode() {variable=In Post Viz}.Value)
   RigVMDispatchNode(
    Enabled=True
    ScreenColor=(R=1,G=1,B=1,A=1)
    ScreenDuration=0.05
    Prefix=&RigVMFunction_StringConcat(B=" - IN POST:   ", A=RigVMVariableNode() {variable=Driver Item Name}.Value) {template="Concat::Execute(in A,in B,out Result)"}.Result
    ExecuteContext=@n36.True
    Value=RigVMVariableNode() {variable=Result In Post}.Value
   ) {template="DISPATCH_RigVMDispatch_Print(in Prefix,in Value,in Enabled,in ScreenDuration,in ScreenColor)"}
   RigVMDispatchNode(
    Enabled=True
    ScreenColor=(R=1,G=1,B=1,A=1)
    ScreenDuration=0.05
    Prefix=&RigVMFunction_StringConcat(B=" - IN PRE:   ", A=RigVMVariableNode() {variable=Driver Item Name}.Value) {template="Concat::Execute(in A,in B,out Result)"}.Result
    ExecuteContext=@n33.True
    Value=RigVMVariableNode() {variable=Result In Pre}.Value
   ) {template="DISPATCH_RigVMDispatch_Print(in Prefix,in Value,in Enabled,in ScreenDuration,in ScreenColor)"}
   RigVMDispatchNode(
    Enabled=True
    ScreenColor=(R=1,G=1,B=1,A=1)
    ScreenDuration=0.05
    Prefix=&RigVMFunction_StringConcat(B=" - IN:   ", A=RigVMVariableNode() {variable=Driver Item Name}.Value) {template="Concat::Execute(in A,in B,out Result)"}.Result
    ExecuteContext=@n30.True
    Value=RigVMVariableNode() {variable=Result In}.Value
   ) {template="DISPATCH_RigVMDispatch_Print(in Prefix,in Value,in Enabled,in ScreenDuration,in ScreenColor)"}
   n37 = &RigUnit_SphericalPoseReader(
    ActiveRegionScaleFactors=(PositiveWidth=1,NegativeWidth=1,PositiveHeight=1,NegativeHeight=1)
    ActiveRegionSize=0.001
    Debug=(bDrawDebug=false,bDraw2D=False,bDrawLocalAxes=False,DebugScale=25,DebugSegments=20,DebugThickness=0.05)
    DriverAxis=(X=1,Y=0,Z=0)
    FalloffRegionScaleFactors=(PositiveWidth=1,NegativeWidth=1,PositiveHeight=1,NegativeHeight=1)
    FalloffSize=0.5
    FlipHeightScaling=False
    FlipWidthScaling=False
    RotationOffset=(X=0,Y=0,Z=0)
    DriverItem=RigVMVariableNode() {variable=Driver Item}.Value
    OptionalParentItem=RigVMVariableNode() {variable=Driver Item Parent}.Value
    Debug.bDrawDebug=RigVMVariableNode() {variable=Ot Viz}.Value
    ExecutePin=@RigVMFunction_Sequence.D
   )
   > n38 = RigVMVariableNode(Value=@n37.OutputParam) {variable=Result Ot}
   n39 = &RigVMFunction_ControlFlowBranch(ExecuteContext=@n38.ExecuteContext, Condition=RigVMVariableNode() {variable=Ot Viz}.Value)
   n40 = &RigUnit_SphericalPoseReader(
    ActiveRegionScaleFactors=(PositiveWidth=1,NegativeWidth=1,PositiveHeight=1,NegativeHeight=1)
    ActiveRegionSize=0.001
    Debug=(bDrawDebug=false,bDraw2D=False,bDrawLocalAxes=False,DebugScale=25,DebugSegments=20,DebugThickness=0.05)
    DriverAxis=(X=1,Y=0,Z=0)
    FalloffRegionScaleFactors=(PositiveWidth=1,NegativeWidth=1,PositiveHeight=2,NegativeHeight=2)
    FalloffSize=0.25
    FlipHeightScaling=False
    FlipWidthScaling=False
    RotationOffset=(X=0,Y=-45,Z=0)
    ExecutePin=@n39.Completed
    Debug.bDrawDebug=RigVMVariableNode() {variable=Ot Pre Viz}.Value
    DriverItem=RigVMVariableNode() {variable=Driver Item}.Value
    OptionalParentItem=RigVMVariableNode() {variable=Driver Item Parent}.Value
   )
   > n41 = RigVMVariableNode(Value=@n40.OutputParam) {variable=Result Ot Pre}
   n42 = &RigVMFunction_ControlFlowBranch(Condition=RigVMVariableNode() {variable=Ot Pre Viz}.Value, ExecuteContext=@n41.ExecuteContext)
   n43 = &RigUnit_SphericalPoseReader(
    ActiveRegionScaleFactors=(PositiveWidth=1,NegativeWidth=1,PositiveHeight=1,NegativeHeight=1)
    ActiveRegionSize=0.001
    Debug=(bDrawDebug=false,bDraw2D=False,bDrawLocalAxes=False,DebugScale=25,DebugSegments=20,DebugThickness=0.05)
    DriverAxis=(X=1,Y=0,Z=0)
    FalloffRegionScaleFactors=(PositiveWidth=1,NegativeWidth=1,PositiveHeight=2,NegativeHeight=2)
    FalloffSize=0.25
    FlipHeightScaling=False
    FlipWidthScaling=False
    RotationOffset=(X=0,Y=45,Z=0)
    ExecutePin=@n42.Completed
    Debug.bDrawDebug=RigVMVariableNode() {variable=Ot Post Viz}.Value
    DriverItem=RigVMVariableNode() {variable=Driver Item}.Value
    OptionalParentItem=RigVMVariableNode() {variable=Driver Item Parent}.Value
   )
   > n44 = RigVMVariableNode(Value=@n43.OutputParam) {variable=Result Ot Post}
   n45 = &RigVMFunction_ControlFlowBranch(Condition=RigVMVariableNode() {variable=Ot Post Viz}.Value, ExecuteContext=@n44.ExecuteContext)
   RigVMDispatchNode(
    Enabled=True
    ScreenColor=(R=1,G=1,B=1,A=1)
    ScreenDuration=0.05
    Prefix=&RigVMFunction_StringConcat(B=" - OT POST:   ", A=RigVMVariableNode() {variable=Driver Item Name}.Value) {template="Concat::Execute(in A,in B,out Result)"}.Result
    ExecuteContext=@n45.True
    Value=RigVMVariableNode() {variable=Result Ot Post}.Value
   ) {template="DISPATCH_RigVMDispatch_Print(in Prefix,in Value,in Enabled,in ScreenDuration,in ScreenColor)"}
   RigVMDispatchNode(
    Enabled=True
    ScreenColor=(R=1,G=1,B=1,A=1)
    ScreenDuration=0.05
    Prefix=&RigVMFunction_StringConcat(B=" - OT PRE:   ", A=RigVMVariableNode() {variable=Driver Item Name}.Value) {template="Concat::Execute(in A,in B,out Result)"}.Result
    ExecuteContext=@n42.True
    Value=RigVMVariableNode() {variable=Result Ot Pre}.Value
   ) {template="DISPATCH_RigVMDispatch_Print(in Prefix,in Value,in Enabled,in ScreenDuration,in ScreenColor)"}
   RigVMDispatchNode(
    Enabled=True
    ScreenColor=(R=1,G=1,B=1,A=1)
    ScreenDuration=0.05
    Prefix=&RigVMFunction_StringConcat(B=" - OT:   ", A=RigVMVariableNode() {variable=Driver Item Name}.Value) {template="Concat::Execute(in A,in B,out Result)"}.Result
    ExecuteContext=@n39.True
    Value=RigVMVariableNode() {variable=Result Ot}.Value
   ) {template="DISPATCH_RigVMDispatch_Print(in Prefix,in Value,in Enabled,in ScreenDuration,in ScreenColor)"}
   n46 = &RigUnit_SphericalPoseReader(
    ActiveRegionScaleFactors=(PositiveWidth=1,NegativeWidth=1,PositiveHeight=1,NegativeHeight=1)
    ActiveRegionSize=0.001
    Debug=(bDrawDebug=false,bDraw2D=False,bDrawLocalAxes=False,DebugScale=25,DebugSegments=20,DebugThickness=0.05)
    DriverAxis=(X=1,Y=0,Z=0)
    FalloffRegionScaleFactors=(PositiveWidth=1,NegativeWidth=1,PositiveHeight=1,NegativeHeight=1)
    FalloffSize=0.5
    FlipHeightScaling=False
    FlipWidthScaling=False
    RotationOffset=(X=90,Y=0,Z=0)
    DriverItem=RigVMVariableNode() {variable=Driver Item}.Value
    OptionalParentItem=RigVMVariableNode() {variable=Driver Item Parent}.Value
    Debug.bDrawDebug=RigVMVariableNode() {variable=Ba Viz}.Value
    ExecutePin=@RigVMFunction_Sequence.B
   )
   > n47 = RigVMVariableNode(Value=@n46.OutputParam) {variable=Result Ba}
   n48 = &RigVMFunction_ControlFlowBranch(ExecuteContext=@n47.ExecuteContext, Condition=RigVMVariableNode() {variable=Ba Viz}.Value)
   n49 = &RigUnit_SphericalPoseReader(
    ActiveRegionScaleFactors=(PositiveWidth=1,NegativeWidth=1,PositiveHeight=1,NegativeHeight=1)
    ActiveRegionSize=0.001
    Debug=(bDrawDebug=false,bDraw2D=False,bDrawLocalAxes=False,DebugScale=25,DebugSegments=20,DebugThickness=0.05)
    DriverAxis=(X=1,Y=0,Z=0)
    FalloffRegionScaleFactors=(PositiveWidth=1,NegativeWidth=1,PositiveHeight=2,NegativeHeight=2)
    FalloffSize=0.25
    FlipHeightScaling=False
    FlipWidthScaling=False
    RotationOffset=(X=90,Y=0,Z=-45)
    ExecutePin=@n48.Completed
    DriverItem=RigVMVariableNode() {variable=Driver Item}.Value
    OptionalParentItem=RigVMVariableNode() {variable=Driver Item Parent}.Value
    Debug.bDrawDebug=RigVMVariableNode() {variable=Ba Pre Viz}.Value
   )
   > n50 = RigVMVariableNode(Value=@n49.OutputParam) {variable=Result Ba Pre}
   n51 = &RigVMFunction_ControlFlowBranch(ExecuteContext=@n50.ExecuteContext, Condition=RigVMVariableNode() {variable=Ba Pre Viz}.Value)
   n52 = &RigUnit_SphericalPoseReader(
    ActiveRegionScaleFactors=(PositiveWidth=1,NegativeWidth=1,PositiveHeight=1,NegativeHeight=1)
    ActiveRegionSize=0.001
    Debug=(bDrawDebug=false,bDraw2D=False,bDrawLocalAxes=False,DebugScale=25,DebugSegments=20,DebugThickness=0.05)
    DriverAxis=(X=1,Y=0,Z=0)
    FalloffRegionScaleFactors=(PositiveWidth=1,NegativeWidth=1,PositiveHeight=2,NegativeHeight=2)
    FalloffSize=0.25
    FlipHeightScaling=False
    FlipWidthScaling=False
    RotationOffset=(X=90,Y=0,Z=45)
    ExecutePin=@n51.Completed
    DriverItem=RigVMVariableNode() {variable=Driver Item}.Value
    OptionalParentItem=RigVMVariableNode() {variable=Driver Item Parent}.Value
    Debug.bDrawDebug=RigVMVariableNode() {variable=Ba Post Viz}.Value
   )
   > n53 = RigVMVariableNode(Value=@n52.OutputParam) {variable=Result Ba Post}
   n54 = &RigVMFunction_ControlFlowBranch(ExecuteContext=@n53.ExecuteContext, Condition=RigVMVariableNode() {variable=Ba Post Viz}.Value)
   RigVMDispatchNode(
    Enabled=True
    ScreenColor=(R=1,G=1,B=1,A=1)
    ScreenDuration=0.05
    Prefix=&RigVMFunction_StringConcat(B=" - BA POST:   ", A=RigVMVariableNode() {variable=Driver Item Name}.Value) {template="Concat::Execute(in A,in B,out Result)"}.Result
    ExecuteContext=@n54.True
    Value=RigVMVariableNode() {variable=Result Ba Post}.Value
   ) {template="DISPATCH_RigVMDispatch_Print(in Prefix,in Value,in Enabled,in ScreenDuration,in ScreenColor)"}
   RigVMDispatchNode(
    Enabled=True
    ScreenColor=(R=1,G=1,B=1,A=1)
    ScreenDuration=0.05
    Prefix=&RigVMFunction_StringConcat(B=" - BA PRE:   ", A=RigVMVariableNode() {variable=Driver Item Name}.Value) {template="Concat::Execute(in A,in B,out Result)"}.Result
    ExecuteContext=@n51.True
    Value=RigVMVariableNode() {variable=Result Ba Pre}.Value
   ) {template="DISPATCH_RigVMDispatch_Print(in Prefix,in Value,in Enabled,in ScreenDuration,in ScreenColor)"}
   RigVMDispatchNode(
    Enabled=True
    ScreenColor=(R=1,G=1,B=1,A=1)
    ScreenDuration=0.05
    Prefix=&RigVMFunction_StringConcat(B=" - BA:   ", A=RigVMVariableNode() {variable=Driver Item Name}.Value) {template="Concat::Execute(in A,in B,out Result)"}.Result
    ExecuteContext=@n48.True
    Value=RigVMVariableNode() {variable=Result Ba}.Value
   ) {template="DISPATCH_RigVMDispatch_Print(in Prefix,in Value,in Enabled,in ScreenDuration,in ScreenColor)"}
  ]}
 New Function_ContainedGraph {nodes [
   RigVMFunctionEntryNode()
   > n1 = &RigUnit_SphericalPoseReader(
    ActiveRegionScaleFactors=(PositiveWidth=1,NegativeWidth=1,PositiveHeight=1,NegativeHeight=1)
    ActiveRegionSize=0.001
    Debug=(bDrawDebug=True,bDraw2D=False,bDrawLocalAxes=False,DebugScale=12.5,DebugSegments=20,DebugThickness=0.1)
    DriverAxis=(X=1,Y=0,Z=0)
    FalloffRegionScaleFactors=(PositiveWidth=1,NegativeWidth=1,PositiveHeight=1,NegativeHeight=1)
    FalloffSize=0.5
    FlipHeightScaling=False
    FlipWidthScaling=False
    OptionalParentItem=(Type=Bone,Name="None")
    RotationOffset=(X=90,Y=0,Z=0)
    Debug.bDrawDebug=RigVMVariableNode() {variable=viz bend}.Value
    DriverItem=RigVMVariableNode() {variable=driver item}.Value
   )
   > n2 = RigVMVariableNode(Value=@n1.OutputParam) {variable=Result Bend}
   n3 = &RigVMFunction_ControlFlowBranch(Condition=RigVMVariableNode() {variable=viz bend post}.Value, ExecuteContext=@n2.ExecuteContext)
   n4 = &RigUnit_SphericalPoseReader(
    ActiveRegionScaleFactors=(PositiveWidth=1,NegativeWidth=1,PositiveHeight=1,NegativeHeight=1)
    ActiveRegionSize=0.001
    Debug=(bDrawDebug=True,bDraw2D=False,bDrawLocalAxes=False,DebugScale=12.5,DebugSegments=20,DebugThickness=0.1)
    DriverAxis=(X=1,Y=0,Z=0)
    FalloffRegionScaleFactors=(PositiveWidth=1,NegativeWidth=1,PositiveHeight=1,NegativeHeight=1)
    FalloffSize=0.5
    FlipHeightScaling=False
    FlipWidthScaling=False
    OptionalParentItem=(Type=Bone,Name="None")
    RotationOffset=(X=90,Y=0,Z=90)
    ExecutePin=@n3.Completed
    DriverItem=RigVMVariableNode() {variable=driver item}.Value
    Debug.bDrawDebug=RigVMVariableNode() {variable=viz bend post}.Value
   )
   > n5 = RigVMVariableNode(Value=@n4.OutputParam) {variable=Result Bend Post}
   RigVMFunction_ControlFlowBranch = &RigVMFunction_ControlFlowBranch(ExecuteContext=@n5.ExecuteContext, Condition=RigVMVariableNode() {variable=viz bend post}.Value)
   RigVMFunctionReturnNode(
    ExecuteContext=@RigVMFunction_ControlFlowBranch.Completed
    Bend=RigVMVariableNode() {variable=Result Bend}.Value
    Bend Post=RigVMVariableNode() {variable=Result Bend Post}.Value
   )
   To_String_1 = RigVMDispatchNode(Value=/Script/RigVM.RigVMFunction_NameConcat(B="  Bend Post  ", A=RigVMVariableNode() {variable=driver item}.Value.Name) {is=struct RigVMUnitNode, template="Concat::Execute(in A,in B,out Result)"}.Result) {template="DISPATCH_RigDispatch_ToString(in Value,out Result)"}
   RigVMDispatchNode(
    Enabled=True
    ScreenColor=(R=1,G=1,B=1,A=1)
    ScreenDuration=0.05
    Prefix=@To_String_1.Result
    Value=RigVMVariableNode() {variable=Result Bend Post}.Value
    ExecuteContext=@RigVMFunction_ControlFlowBranch.True
   ) {template="DISPATCH_RigVMDispatch_Print(in Prefix,in Value,in Enabled,in ScreenDuration,in ScreenColor)"}
   To_String = RigVMDispatchNode(Value=&RigVMFunction_NameConcat(B="  Bend  ", A=RigVMVariableNode() {variable=driver item}.Value.Name) {template="Concat::Execute(in A,in B,out Result)"}.Result) {template="DISPATCH_RigDispatch_ToString(in Value,out Result)"}
   RigVMDispatchNode(
    Enabled=True
    ScreenColor=(R=1,G=1,B=1,A=1)
    ScreenDuration=0.05
    Prefix=@To_String.Result
    Value=RigVMVariableNode() {variable=Result Bend}.Value
    ExecuteContext=@n3.True
   ) {template="DISPATCH_RigVMDispatch_Print(in Prefix,in Value,in Enabled,in ScreenDuration,in ScreenColor)"}
  ]}
 New Function_ContainedGraph {nodes [
   n1 = RigVMFunctionEntryNode()
   > RigVMVariableNode(Value=&RigUnit_HierarchyGetParent(bDefaultParent=True, Child=@n1.Driver Item).Parent) {variable=Driver Item Parent}
   To_String = RigVMDispatchNode(Value=@n1.Driver Item.Name) {template="DISPATCH_RigDispatch_ToString(in Value,out Result)"}
   > RigVMVariableNode(Value=@To_String.Result) {variable=Driver Item Name}
   > RigVMFunction_Sequence = RigVMAggregateNode()
   RigVMFunctionReturnNode(
    ExecuteContext=@RigVMFunction_Sequence.F
    Up Post=RigVMVariableNode() {variable=Result Up Post}.Value
    Aim Up In=RigVMVariableNode() {variable=Result Aim Up In}.Value
    Aim Up Ot=RigVMVariableNode() {variable=Result Aim Up Ot}.Value
    Aim Dn In=RigVMVariableNode() {variable=Result Aim Dn In}.Value
    Aim Dn Ot=RigVMVariableNode() {variable=Result Aim Dn Ot}.Value
    Dn=RigVMVariableNode() {variable=Result Dn}.Value
    Dn Pre=RigVMVariableNode() {variable=Result Dn Pre}.Value
    Dn Post=RigVMVariableNode() {variable=Result Dn Post}.Value
    Fr=RigVMVariableNode() {variable=Result Fr}.Value
    Fr Pre=RigVMVariableNode() {variable=Result Fr Pre}.Value
    Fr Post=RigVMVariableNode() {variable=Result Fr Post}.Value
    Fr Up=RigVMVariableNode() {variable=Result Fr Up}.Value
    Fr Dn=RigVMVariableNode() {variable=Result Fr Dn}.Value
    Ba=RigVMVariableNode() {variable=Result Ba}.Value
    Ba Pre=RigVMVariableNode() {variable=Result Ba Pre}.Value
    Ba Post=RigVMVariableNode() {variable=Result Ba Post}.Value
    Up=RigVMVariableNode() {variable=Result Up}.Value
    Ba Up=RigVMVariableNode() {variable=Result Ba Up}.Value
    Ba Dn=RigVMVariableNode() {variable=Result Ba Dn}.Value
    Up Pre=RigVMVariableNode() {variable=Result Up Pre}.Value
   )
   n2 = &RigUnit_SphericalPoseReader(
    ActiveRegionScaleFactors=(PositiveWidth=1,NegativeWidth=1,PositiveHeight=1,NegativeHeight=1)
    ActiveRegionSize=0.001
    Debug=(bDrawDebug=false,bDraw2D=False,bDrawLocalAxes=False,DebugScale=25,DebugSegments=20,DebugThickness=0.05)
    DriverAxis=(X=1,Y=0,Z=0)
    FalloffRegionScaleFactors=(PositiveWidth=1.3471,NegativeWidth=1,PositiveHeight=1,NegativeHeight=1)
    FalloffSize=0.5
    FlipHeightScaling=False
    FlipWidthScaling=False
    RotationOffset=(X=0,Y=31.239,Z=0)
    Debug.bDrawDebug=RigVMVariableNode() {variable=Up Viz}.Value
    DriverItem=RigVMVariableNode() {variable=Driver Item}.Value
    OptionalParentItem=RigVMVariableNode() {variable=Driver Item Parent}.Value
    ExecutePin=@RigVMFunction_Sequence.A
   )
   > n3 = RigVMVariableNode(Value=@n2.OutputParam) {variable=Result Up}
   RigVMFunction_ControlFlowBranch = &RigVMFunction_ControlFlowBranch(ExecuteContext=@n3.ExecuteContext, Condition=RigVMVariableNode() {variable=Up Viz}.Value)
   n4 = &RigUnit_SphericalPoseReader(
    ActiveRegionScaleFactors=(PositiveWidth=1,NegativeWidth=1,PositiveHeight=1,NegativeHeight=1)
    ActiveRegionSize=0.001
    Debug=(bDrawDebug=false,bDraw2D=False,bDrawLocalAxes=False,DebugScale=25,DebugSegments=20,DebugThickness=0.05)
    DriverAxis=(X=1,Y=0,Z=0)
    FalloffRegionScaleFactors=(PositiveWidth=1.6942,NegativeWidth=1,PositiveHeight=2,NegativeHeight=2)
    FalloffSize=0.25
    FlipHeightScaling=False
    FlipWidthScaling=False
    RotationOffset=(X=0,Y=-13.761,Z=0)
    ExecutePin=@RigVMFunction_ControlFlowBranch.Completed
    DriverItem=RigVMVariableNode() {variable=Driver Item}.Value
    OptionalParentItem=RigVMVariableNode() {variable=Driver Item Parent}.Value
    Debug.bDrawDebug=RigVMVariableNode() {variable=Up Pre Viz}.Value
   )
   > n5 = RigVMVariableNode(Value=@n4.OutputParam) {variable=Result Up Pre}
   n6 = &RigVMFunction_ControlFlowBranch(Condition=RigVMVariableNode() {variable=Up Pre Viz}.Value, ExecuteContext=@n5.ExecuteContext)
   n7 = &RigUnit_SphericalPoseReader(
    ActiveRegionScaleFactors=(PositiveWidth=1,NegativeWidth=1,PositiveHeight=1,NegativeHeight=1)
    ActiveRegionSize=0.001
    Debug=(bDrawDebug=false,bDraw2D=False,bDrawLocalAxes=False,DebugScale=25,DebugSegments=20,DebugThickness=0.05)
    DriverAxis=(X=1,Y=0,Z=0)
    FalloffRegionScaleFactors=(PositiveWidth=0.5,NegativeWidth=0.5,PositiveHeight=1,NegativeHeight=1)
    FalloffSize=0.5
    FlipHeightScaling=False
    FlipWidthScaling=False
    RotationOffset=(X=0,Y=76.239,Z=0)
    ExecutePin=@n6.Completed
    Debug.bDrawDebug=RigVMVariableNode() {variable=Up Post Viz}.Value
    DriverItem=RigVMVariableNode() {variable=Driver Item}.Value
    OptionalParentItem=RigVMVariableNode() {variable=Driver Item Parent}.Value
   )
   > n8 = RigVMVariableNode(Value=@n7.OutputParam) {variable=Result Up Post}
   n9 = &RigVMFunction_ControlFlowBranch(Condition=RigVMVariableNode() {variable=Up Post Viz}.Value, ExecuteContext=@n8.ExecuteContext)
   RigVMDispatchNode(
    Enabled=True
    ScreenColor=(R=1,G=1,B=1,A=1)
    ScreenDuration=0.05
    Prefix=&RigVMFunction_StringConcat(B=" - UP POST:   ", A=RigVMVariableNode() {variable=Driver Item Name}.Value) {template="Concat::Execute(in A,in B,out Result)"}.Result
    Value=RigVMVariableNode() {variable=Result Up Post}.Value
    ExecuteContext=@n9.True
   ) {template="DISPATCH_RigVMDispatch_Print(in Prefix,in Value,in Enabled,in ScreenDuration,in ScreenColor)"}
   RigVMDispatchNode(
    Enabled=True
    ScreenColor=(R=1,G=1,B=1,A=1)
    ScreenDuration=0.05
    Prefix=&RigVMFunction_StringConcat(B=" - UP PRE:   ", A=RigVMVariableNode() {variable=Driver Item Name}.Value) {template="Concat::Execute(in A,in B,out Result)"}.Result
    Value=RigVMVariableNode() {variable=Result Up Pre}.Value
    ExecuteContext=@n6.True
   ) {template="DISPATCH_RigVMDispatch_Print(in Prefix,in Value,in Enabled,in ScreenDuration,in ScreenColor)"}
   Concat = &RigVMFunction_StringConcat(B=" - UP:   ", A=RigVMVariableNode() {variable=Driver Item Name}.Value) {template="Concat::Execute(in A,in B,out Result)"}
   RigVMDispatchNode(
    Enabled=True
    ScreenColor=(R=1,G=1,B=1,A=1)
    ScreenDuration=0.05
    Prefix=@Concat.Result
    Value=RigVMVariableNode() {variable=Result Up}.Value
    ExecuteContext=@RigVMFunction_ControlFlowBranch.True
   ) {template="DISPATCH_RigVMDispatch_Print(in Prefix,in Value,in Enabled,in ScreenDuration,in ScreenColor)"}
   n10 = &RigUnit_SphericalPoseReader(
    ActiveRegionScaleFactors=(PositiveWidth=1,NegativeWidth=1,PositiveHeight=1,NegativeHeight=1)
    ActiveRegionSize=0.001
    Debug=(bDrawDebug=false,bDraw2D=False,bDrawLocalAxes=False,DebugScale=25,DebugSegments=20,DebugThickness=0.05)
    DriverAxis=(X=1,Y=0,Z=0)
    FalloffRegionScaleFactors=(PositiveWidth=1,NegativeWidth=1,PositiveHeight=1,NegativeHeight=1)
    FalloffSize=0.5
    FlipHeightScaling=False
    FlipWidthScaling=False
    RotationOffset=(X=90,Y=31.239,Z=0)
    DriverItem=RigVMVariableNode() {variable=Driver Item}.Value
    OptionalParentItem=RigVMVariableNode() {variable=Driver Item Parent}.Value
    Debug.bDrawDebug=RigVMVariableNode() {variable=Fr Viz}.Value
    ExecutePin=@RigVMFunction_Sequence.C
   )
   > n11 = RigVMVariableNode(Value=@n10.OutputParam) {variable=Result Fr}
   n12 = &RigVMFunction_ControlFlowBranch(ExecuteContext=@n11.ExecuteContext, Condition=RigVMVariableNode() {variable=Fr Viz}.Value)
   n13 = &RigUnit_SphericalPoseReader(
    ActiveRegionScaleFactors=(PositiveWidth=1,NegativeWidth=1,PositiveHeight=1,NegativeHeight=1)
    ActiveRegionSize=0.001
    Debug=(bDrawDebug=false,bDraw2D=False,bDrawLocalAxes=False,DebugScale=25,DebugSegments=20,DebugThickness=0.05)
    DriverAxis=(X=1,Y=0,Z=0)
    FalloffRegionScaleFactors=(PositiveWidth=2,NegativeWidth=2,PositiveHeight=1,NegativeHeight=1)
    FalloffSize=0.25
    FlipHeightScaling=False
    FlipWidthScaling=False
    RotationOffset=(X=45,Y=-58.761,Z=0)
    ExecutePin=@n12.Completed
    DriverItem=RigVMVariableNode() {variable=Driver Item}.Value
    OptionalParentItem=RigVMVariableNode() {variable=Driver Item Parent}.Value
    Debug.bDrawDebug=RigVMVariableNode() {variable=Fr Pre Viz}.Value
   )
   > n14 = RigVMVariableNode(Value=@n13.OutputParam) {variable=Result Fr Pre}
   n15 = &RigVMFunction_ControlFlowBranch(ExecuteContext=@n14.ExecuteContext, Condition=RigVMVariableNode() {variable=Fr Pre Viz}.Value)
   n16 = &RigUnit_SphericalPoseReader(
    ActiveRegionScaleFactors=(PositiveWidth=1,NegativeWidth=1,PositiveHeight=1,NegativeHeight=1)
    ActiveRegionSize=0.001
    Debug=(bDrawDebug=false,bDraw2D=False,bDrawLocalAxes=False,DebugScale=25,DebugSegments=20,DebugThickness=0.05)
    DriverAxis=(X=1,Y=0,Z=0)
    FalloffRegionScaleFactors=(PositiveWidth=2,NegativeWidth=2,PositiveHeight=1,NegativeHeight=1)
    FalloffSize=0.25
    FlipHeightScaling=False
    FlipWidthScaling=False
    RotationOffset=(X=135,Y=-58.761,Z=0)
    ExecutePin=@n15.Completed
    DriverItem=RigVMVariableNode() {variable=Driver Item}.Value
    OptionalParentItem=RigVMVariableNode() {variable=Driver Item Parent}.Value
    Debug.bDrawDebug=RigVMVariableNode() {variable=Fr Post Viz}.Value
   )
   > n17 = RigVMVariableNode(Value=@n16.OutputParam) {variable=Result Fr Post}
   n18 = &RigVMFunction_ControlFlowBranch(ExecuteContext=@n17.ExecuteContext, Condition=RigVMVariableNode() {variable=Fr Post Viz}.Value)
   n19 = &RigUnit_SphericalPoseReader(
    ActiveRegionScaleFactors=(PositiveWidth=1,NegativeWidth=1,PositiveHeight=1,NegativeHeight=1)
    ActiveRegionSize=0.001
    Debug=(bDrawDebug=false,bDraw2D=False,bDrawLocalAxes=False,DebugScale=25,DebugSegments=20,DebugThickness=0.05)
    DriverAxis=(X=1,Y=0,Z=0)
    FalloffRegionScaleFactors=(PositiveWidth=1,NegativeWidth=1,PositiveHeight=0.5,NegativeHeight=0.5)
    FalloffSize=0.5
    FlipHeightScaling=False
    FlipWidthScaling=False
    RotationOffset=(X=45,Y=31.239,Z=0)
    ExecutePin=@n18.Completed
    DriverItem=RigVMVariableNode() {variable=Driver Item}.Value
    OptionalParentItem=RigVMVariableNode() {variable=Driver Item Parent}.Value
    Debug.bDrawDebug=RigVMVariableNode() {variable=Fr Up Viz}.Value
   )
   > n20 = RigVMVariableNode(Value=@n19.OutputParam) {variable=Result Fr Up}
   n21 = &RigVMFunction_ControlFlowBranch(ExecuteContext=@n20.ExecuteContext, Condition=RigVMVariableNode() {variable=Fr Up Viz}.Value)
   n22 = &RigUnit_SphericalPoseReader(
    ActiveRegionScaleFactors=(PositiveWidth=1,NegativeWidth=1,PositiveHeight=1,NegativeHeight=1)
    ActiveRegionSize=0.001
    Debug=(bDrawDebug=false,bDraw2D=False,bDrawLocalAxes=False,DebugScale=25,DebugSegments=20,DebugThickness=0.05)
    DriverAxis=(X=1,Y=0,Z=0)
    FalloffRegionScaleFactors=(PositiveWidth=1,NegativeWidth=1,PositiveHeight=0.5,NegativeHeight=0.5)
    FalloffSize=0.5
    FlipHeightScaling=False
    FlipWidthScaling=False
    RotationOffset=(X=135,Y=31.239,Z=0)
    ExecutePin=@n21.Completed
    DriverItem=RigVMVariableNode() {variable=Driver Item}.Value
    OptionalParentItem=RigVMVariableNode() {variable=Driver Item Parent}.Value
    Debug.bDrawDebug=RigVMVariableNode() {variable=Fr Dn Viz}.Value
   )
   > n23 = RigVMVariableNode(Value=@n22.OutputParam) {variable=Result Fr Dn}
   n24 = &RigVMFunction_ControlFlowBranch(ExecuteContext=@n23.ExecuteContext, Condition=RigVMVariableNode() {variable=Fr Dn Viz}.Value)
   RigVMDispatchNode(
    Enabled=True
    ScreenColor=(R=1,G=1,B=1,A=1)
    ScreenDuration=0.05
    Prefix=&RigVMFunction_StringConcat(B=" - FR DN:   ", A=RigVMVariableNode() {variable=Driver Item Name}.Value) {template="Concat::Execute(in A,in B,out Result)"}.Result
    ExecuteContext=@n24.True
    Value=RigVMVariableNode() {variable=Result Fr Dn}.Value
   ) {template="DISPATCH_RigVMDispatch_Print(in Prefix,in Value,in Enabled,in ScreenDuration,in ScreenColor)"}
   RigVMDispatchNode(
    Enabled=True
    ScreenColor=(R=1,G=1,B=1,A=1)
    ScreenDuration=0.05
    Prefix=&RigVMFunction_StringConcat(B=" - FR UP:   ", A=RigVMVariableNode() {variable=Driver Item Name}.Value) {template="Concat::Execute(in A,in B,out Result)"}.Result
    ExecuteContext=@n21.True
    Value=RigVMVariableNode() {variable=Result Fr Up}.Value
   ) {template="DISPATCH_RigVMDispatch_Print(in Prefix,in Value,in Enabled,in ScreenDuration,in ScreenColor)"}
   RigVMDispatchNode(
    Enabled=True
    ScreenColor=(R=1,G=1,B=1,A=1)
    ScreenDuration=0.05
    Prefix=&RigVMFunction_StringConcat(B=" - FR POST:   ", A=RigVMVariableNode() {variable=Driver Item Name}.Value) {template="Concat::Execute(in A,in B,out Result)"}.Result
    ExecuteContext=@n18.True
    Value=RigVMVariableNode() {variable=Result Fr Post}.Value
   ) {template="DISPATCH_RigVMDispatch_Print(in Prefix,in Value,in Enabled,in ScreenDuration,in ScreenColor)"}
   RigVMDispatchNode(
    Enabled=True
    ScreenColor=(R=1,G=1,B=1,A=1)
    ScreenDuration=0.05
    Prefix=&RigVMFunction_StringConcat(B=" - FR PRE:   ", A=RigVMVariableNode() {variable=Driver Item Name}.Value) {template="Concat::Execute(in A,in B,out Result)"}.Result
    ExecuteContext=@n15.True
    Value=RigVMVariableNode() {variable=Result Fr Pre}.Value
   ) {template="DISPATCH_RigVMDispatch_Print(in Prefix,in Value,in Enabled,in ScreenDuration,in ScreenColor)"}
   RigVMDispatchNode(
    Enabled=True
    ScreenColor=(R=1,G=1,B=1,A=1)
    ScreenDuration=0.05
    Prefix=&RigVMFunction_StringConcat(B=" - FR:   ", A=RigVMVariableNode() {variable=Driver Item Name}.Value) {template="Concat::Execute(in A,in B,out Result)"}.Result
    ExecuteContext=@n12.True
    Value=RigVMVariableNode() {variable=Result Fr}.Value
   ) {template="DISPATCH_RigVMDispatch_Print(in Prefix,in Value,in Enabled,in ScreenDuration,in ScreenColor)"}
   n25 = &RigUnit_SphericalPoseReader(
    ActiveRegionScaleFactors=(PositiveWidth=1,NegativeWidth=1,PositiveHeight=1,NegativeHeight=1)
    ActiveRegionSize=0.001
    Debug=(bDrawDebug=false,bDraw2D=False,bDrawLocalAxes=False,DebugScale=25,DebugSegments=20,DebugThickness=0.05)
    DriverAxis=(X=1,Y=0,Z=0)
    FalloffRegionScaleFactors=(PositiveWidth=1,NegativeWidth=1,PositiveHeight=1,NegativeHeight=1)
    FalloffSize=0.5
    FlipHeightScaling=False
    FlipWidthScaling=False
    RotationOffset=(X=-90,Y=31.239,Z=0)
    DriverItem=RigVMVariableNode() {variable=Driver Item}.Value
    OptionalParentItem=RigVMVariableNode() {variable=Driver Item Parent}.Value
    Debug.bDrawDebug=RigVMVariableNode() {variable=Ba Viz}.Value
    ExecutePin=@RigVMFunction_Sequence.D
   )
   > n26 = RigVMVariableNode(Value=@n25.OutputParam) {variable=Result Ba}
   n27 = &RigVMFunction_ControlFlowBranch(ExecuteContext=@n26.ExecuteContext, Condition=RigVMVariableNode() {variable=Ba Viz}.Value)
   n28 = &RigUnit_SphericalPoseReader(
    ActiveRegionScaleFactors=(PositiveWidth=1,NegativeWidth=1,PositiveHeight=1,NegativeHeight=1)
    ActiveRegionSize=0.001
    Debug=(bDrawDebug=false,bDraw2D=False,bDrawLocalAxes=False,DebugScale=25,DebugSegments=20,DebugThickness=0.05)
    DriverAxis=(X=1,Y=0,Z=0)
    FalloffRegionScaleFactors=(PositiveWidth=2,NegativeWidth=2,PositiveHeight=1,NegativeHeight=1)
    FalloffSize=0.25
    FlipHeightScaling=False
    FlipWidthScaling=False
    RotationOffset=(X=-45,Y=-58.761,Z=0)
    ExecutePin=@n27.Completed
    Debug.bDrawDebug=RigVMVariableNode() {variable=Ba Pre Viz}.Value
    DriverItem=RigVMVariableNode() {variable=Driver Item}.Value
    OptionalParentItem=RigVMVariableNode() {variable=Driver Item Parent}.Value
   )
   > n29 = RigVMVariableNode(Value=@n28.OutputParam) {variable=Result Ba Pre}
   n30 = &RigVMFunction_ControlFlowBranch(Condition=RigVMVariableNode() {variable=Ba Pre Viz}.Value, ExecuteContext=@n29.ExecuteContext)
   n31 = &RigUnit_SphericalPoseReader(
    ActiveRegionScaleFactors=(PositiveWidth=1,NegativeWidth=1,PositiveHeight=1,NegativeHeight=1)
    ActiveRegionSize=0.001
    Debug=(bDrawDebug=false,bDraw2D=False,bDrawLocalAxes=False,DebugScale=25,DebugSegments=20,DebugThickness=0.05)
    DriverAxis=(X=1,Y=0,Z=0)
    FalloffRegionScaleFactors=(PositiveWidth=2,NegativeWidth=2,PositiveHeight=1,NegativeHeight=1)
    FalloffSize=0.25
    FlipHeightScaling=False
    FlipWidthScaling=False
    RotationOffset=(X=-135,Y=-58.761,Z=0)
    ExecutePin=@n30.Completed
    Debug.bDrawDebug=RigVMVariableNode() {variable=Ba Post Viz}.Value
    DriverItem=RigVMVariableNode() {variable=Driver Item}.Value
    OptionalParentItem=RigVMVariableNode() {variable=Driver Item Parent}.Value
   )
   > n32 = RigVMVariableNode(Value=@n31.OutputParam) {variable=Result Ba Post}
   n33 = &RigVMFunction_ControlFlowBranch(Condition=RigVMVariableNode() {variable=Ba Post Viz}.Value, ExecuteContext=@n32.ExecuteContext)
   n34 = &RigUnit_SphericalPoseReader(
    ActiveRegionScaleFactors=(PositiveWidth=1,NegativeWidth=1,PositiveHeight=1,NegativeHeight=1)
    ActiveRegionSize=0.001
    Debug=(bDrawDebug=false,bDraw2D=False,bDrawLocalAxes=False,DebugScale=25,DebugSegments=20,DebugThickness=0.05)
    DriverAxis=(X=1,Y=0,Z=0)
    FalloffRegionScaleFactors=(PositiveWidth=1,NegativeWidth=1,PositiveHeight=0.5,NegativeHeight=0.5)
    FalloffSize=0.5
    FlipHeightScaling=False
    FlipWidthScaling=False
    RotationOffset=(X=-45,Y=31.239,Z=0)
    ExecutePin=@n33.Completed
    Debug.bDrawDebug=RigVMVariableNode() {variable=Ba Up Viz}.Value
    DriverItem=RigVMVariableNode() {variable=Driver Item}.Value
    OptionalParentItem=RigVMVariableNode() {variable=Driver Item Parent}.Value
   )
   > n35 = RigVMVariableNode(Value=@n34.OutputParam) {variable=Result Ba Up}
   n36 = &RigVMFunction_ControlFlowBranch(Condition=RigVMVariableNode() {variable=Ba Up Viz}.Value, ExecuteContext=@n35.ExecuteContext)
   n37 = &RigUnit_SphericalPoseReader(
    ActiveRegionScaleFactors=(PositiveWidth=1,NegativeWidth=1,PositiveHeight=1,NegativeHeight=1)
    ActiveRegionSize=0.001
    Debug=(bDrawDebug=false,bDraw2D=False,bDrawLocalAxes=False,DebugScale=25,DebugSegments=20,DebugThickness=0.05)
    DriverAxis=(X=1,Y=0,Z=0)
    FalloffRegionScaleFactors=(PositiveWidth=1,NegativeWidth=1,PositiveHeight=0.5,NegativeHeight=0.5)
    FalloffSize=0.5
    FlipHeightScaling=False
    FlipWidthScaling=False
    RotationOffset=(X=-135,Y=31.239,Z=0)
    ExecutePin=@n36.Completed
    Debug.bDrawDebug=RigVMVariableNode() {variable=Ba Dn Viz}.Value
    DriverItem=RigVMVariableNode() {variable=Driver Item}.Value
    OptionalParentItem=RigVMVariableNode() {variable=Driver Item Parent}.Value
   )
   > n38 = RigVMVariableNode(Value=@n37.OutputParam) {variable=Result Ba Dn}
   n39 = &RigVMFunction_ControlFlowBranch(Condition=RigVMVariableNode() {variable=Ba Dn Viz}.Value, ExecuteContext=@n38.ExecuteContext)
   RigVMDispatchNode(
    Enabled=True
    ScreenColor=(R=1,G=1,B=1,A=1)
    ScreenDuration=0.05
    Prefix=&RigVMFunction_StringConcat(B=" - BA DN:   ", A=RigVMVariableNode() {variable=Driver Item Name}.Value) {template="Concat::Execute(in A,in B,out Result)"}.Result
    ExecuteContext=@n39.True
    Value=RigVMVariableNode() {variable=Result Ba Dn}.Value
   ) {template="DISPATCH_RigVMDispatch_Print(in Prefix,in Value,in Enabled,in ScreenDuration,in ScreenColor)"}
   RigVMDispatchNode(
    Enabled=True
    ScreenColor=(R=1,G=1,B=1,A=1)
    ScreenDuration=0.05
    Prefix=&RigVMFunction_StringConcat(B=" - BA UP:   ", A=RigVMVariableNode() {variable=Driver Item Name}.Value) {template="Concat::Execute(in A,in B,out Result)"}.Result
    ExecuteContext=@n36.True
    Value=RigVMVariableNode() {variable=Result Ba Up}.Value
   ) {template="DISPATCH_RigVMDispatch_Print(in Prefix,in Value,in Enabled,in ScreenDuration,in ScreenColor)"}
   RigVMDispatchNode(
    Enabled=True
    ScreenColor=(R=1,G=1,B=1,A=1)
    ScreenDuration=0.05
    Prefix=&RigVMFunction_StringConcat(B=" - BA POST:   ", A=RigVMVariableNode() {variable=Driver Item Name}.Value) {template="Concat::Execute(in A,in B,out Result)"}.Result
    ExecuteContext=@n33.True
    Value=RigVMVariableNode() {variable=Result Ba Post}.Value
   ) {template="DISPATCH_RigVMDispatch_Print(in Prefix,in Value,in Enabled,in ScreenDuration,in ScreenColor)"}
   RigVMDispatchNode(
    Enabled=True
    ScreenColor=(R=1,G=1,B=1,A=1)
    ScreenDuration=0.05
    Prefix=&RigVMFunction_StringConcat(B=" - BA PRE:   ", A=RigVMVariableNode() {variable=Driver Item Name}.Value) {template="Concat::Execute(in A,in B,out Result)"}.Result
    ExecuteContext=@n30.True
    Value=RigVMVariableNode() {variable=Result Ba Pre}.Value
   ) {template="DISPATCH_RigVMDispatch_Print(in Prefix,in Value,in Enabled,in ScreenDuration,in ScreenColor)"}
   RigVMDispatchNode(
    Enabled=True
    ScreenColor=(R=1,G=1,B=1,A=1)
    ScreenDuration=0.05
    Prefix=&RigVMFunction_StringConcat(B=" - BA:   ", A=RigVMVariableNode() {variable=Driver Item Name}.Value) {template="Concat::Execute(in A,in B,out Result)"}.Result
    ExecuteContext=@n27.True
    Value=RigVMVariableNode() {variable=Result Ba}.Value
   ) {template="DISPATCH_RigVMDispatch_Print(in Prefix,in Value,in Enabled,in ScreenDuration,in ScreenColor)"}
   n40 = &RigUnit_SphericalPoseReader(
    ActiveRegionScaleFactors=(PositiveWidth=1,NegativeWidth=1,PositiveHeight=1,NegativeHeight=1)
    ActiveRegionSize=0.001
    Debug=(bDrawDebug=false,bDraw2D=False,bDrawLocalAxes=False,DebugScale=25,DebugSegments=20,DebugThickness=0.05)
    DriverAxis=(X=1,Y=0,Z=0)
    FalloffRegionScaleFactors=(PositiveWidth=1,NegativeWidth=0.65,PositiveHeight=1,NegativeHeight=1)
    FalloffSize=0.5
    FlipHeightScaling=False
    FlipWidthScaling=False
    RotationOffset=(X=0,Y=-148.761,Z=0)
    DriverItem=RigVMVariableNode() {variable=Driver Item}.Value
    OptionalParentItem=RigVMVariableNode() {variable=Driver Item Parent}.Value
    Debug.bDrawDebug=RigVMVariableNode() {variable=Dn Viz}.Value
    ExecutePin=@RigVMFunction_Sequence.B
   )
   > n41 = RigVMVariableNode(Value=@n40.OutputParam) {variable=Result Dn}
   n42 = &RigVMFunction_ControlFlowBranch(ExecuteContext=@n41.ExecuteContext, Condition=RigVMVariableNode() {variable=Dn Viz}.Value)
   n43 = &RigUnit_SphericalPoseReader(
    ActiveRegionScaleFactors=(PositiveWidth=1,NegativeWidth=1,PositiveHeight=1,NegativeHeight=1)
    ActiveRegionSize=0.001
    Debug=(bDrawDebug=false,bDraw2D=False,bDrawLocalAxes=False,DebugScale=25,DebugSegments=20,DebugThickness=0.05)
    DriverAxis=(X=1,Y=0,Z=0)
    FalloffRegionScaleFactors=(PositiveWidth=1,NegativeWidth=0.298,PositiveHeight=2,NegativeHeight=2)
    FalloffSize=0.25
    FlipHeightScaling=False
    FlipWidthScaling=False
    RotationOffset=(X=0,Y=-103.761,Z=0)
    ExecutePin=@n42.Completed
    DriverItem=RigVMVariableNode() {variable=Driver Item}.Value
    OptionalParentItem=RigVMVariableNode() {variable=Driver Item Parent}.Value
    Debug.bDrawDebug=RigVMVariableNode() {variable=Dn Pre Viz}.Value
   )
   > n44 = RigVMVariableNode(Value=@n43.OutputParam) {variable=Result Dn Pre}
   n45 = &RigVMFunction_ControlFlowBranch(ExecuteContext=@n44.ExecuteContext, Condition=RigVMVariableNode() {variable=Dn Pre Viz}.Value)
   n46 = &RigUnit_SphericalPoseReader(
    ActiveRegionScaleFactors=(PositiveWidth=1,NegativeWidth=1,PositiveHeight=1,NegativeHeight=1)
    ActiveRegionSize=0.001
    Debug=(bDrawDebug=false,bDraw2D=False,bDrawLocalAxes=False,DebugScale=25,DebugSegments=20,DebugThickness=0.05)
    DriverAxis=(X=1,Y=0,Z=0)
    FalloffRegionScaleFactors=(PositiveWidth=1,NegativeWidth=1,PositiveHeight=2,NegativeHeight=2)
    FalloffSize=0.25
    FlipHeightScaling=False
    FlipWidthScaling=False
    RotationOffset=(X=0,Y=-193.761,Z=0)
    ExecutePin=@n45.Completed
    DriverItem=RigVMVariableNode() {variable=Driver Item}.Value
    OptionalParentItem=RigVMVariableNode() {variable=Driver Item Parent}.Value
    Debug.bDrawDebug=RigVMVariableNode() {variable=Dn Post Viz}.Value
   )
   > n47 = RigVMVariableNode(Value=@n46.OutputParam) {variable=Result Dn Post}
   n48 = &RigVMFunction_ControlFlowBranch(ExecuteContext=@n47.ExecuteContext, Condition=RigVMVariableNode() {variable=Dn Post Viz}.Value)
   RigVMDispatchNode(
    Enabled=True
    ScreenColor=(R=1,G=1,B=1,A=1)
    ScreenDuration=0.05
    Prefix=&RigVMFunction_StringConcat(B=" - DN POST:   ", A=RigVMVariableNode() {variable=Driver Item Name}.Value) {template="Concat::Execute(in A,in B,out Result)"}.Result
    ExecuteContext=@n48.True
    Value=RigVMVariableNode() {variable=Result Dn Post}.Value
   ) {template="DISPATCH_RigVMDispatch_Print(in Prefix,in Value,in Enabled,in ScreenDuration,in ScreenColor)"}
   RigVMDispatchNode(
    Enabled=True
    ScreenColor=(R=1,G=1,B=1,A=1)
    ScreenDuration=0.05
    Prefix=&RigVMFunction_StringConcat(B=" - DN PRE:   ", A=RigVMVariableNode() {variable=Driver Item Name}.Value) {template="Concat::Execute(in A,in B,out Result)"}.Result
    ExecuteContext=@n45.True
    Value=RigVMVariableNode() {variable=Result Dn Pre}.Value
   ) {template="DISPATCH_RigVMDispatch_Print(in Prefix,in Value,in Enabled,in ScreenDuration,in ScreenColor)"}
   RigVMDispatchNode(
    Enabled=True
    ScreenColor=(R=1,G=1,B=1,A=1)
    ScreenDuration=0.05
    Prefix=&RigVMFunction_StringConcat(B=" - DN:   ", A=RigVMVariableNode() {variable=Driver Item Name}.Value) {template="Concat::Execute(in A,in B,out Result)"}.Result
    ExecuteContext=@n42.True
    Value=RigVMVariableNode() {variable=Result Dn}.Value
   ) {template="DISPATCH_RigVMDispatch_Print(in Prefix,in Value,in Enabled,in ScreenDuration,in ScreenColor)"}
   n49 = &RigUnit_SphericalPoseReader(
    ActiveRegionScaleFactors=(PositiveWidth=1,NegativeWidth=1,PositiveHeight=1,NegativeHeight=1)
    ActiveRegionSize=0.001
    Debug=(bDrawDebug=true,bDraw2D=False,bDrawLocalAxes=False,DebugScale=25,DebugSegments=20,DebugThickness=0.05)
    DriverAxis=(X=1,Y=0,Z=0)
    FalloffRegionScaleFactors=(PositiveWidth=1,NegativeWidth=1,PositiveHeight=1,NegativeHeight=1)
    FalloffSize=0.18
    FlipHeightScaling=False
    FlipWidthScaling=False
    RotationOffset=(X=28.4478,Y=74.4998,Z=0)
    DriverItem=RigVMVariableNode() {variable=Driver Item}.Value
    OptionalParentItem=RigVMVariableNode() {variable=Driver Item Parent}.Value
    Debug.bDrawDebug=RigVMVariableNode() {variable=Aim Up In Viz}.Value
    ExecutePin=@RigVMFunction_Sequence.E
   )
   > n50 = RigVMVariableNode(Value=@n49.OutputParam) {variable=Result Aim Up In}
   n51 = &RigVMFunction_ControlFlowBranch(Condition=RigVMVariableNode() {variable=Aim Up In Viz}.Value, ExecuteContext=@n50.ExecuteContext)
   n52 = &RigUnit_SphericalPoseReader(
    ActiveRegionScaleFactors=(PositiveWidth=1,NegativeWidth=1,PositiveHeight=1,NegativeHeight=1)
    ActiveRegionSize=0.001
    Debug=(bDrawDebug=true,bDraw2D=False,bDrawLocalAxes=False,DebugScale=25,DebugSegments=20,DebugThickness=0.05)
    DriverAxis=(X=1,Y=0,Z=0)
    FalloffRegionScaleFactors=(PositiveWidth=1,NegativeWidth=1,PositiveHeight=1,NegativeHeight=1)
    FalloffSize=0.18
    FlipHeightScaling=False
    FlipWidthScaling=False
    RotationOffset=(X=31.239,Y=-1.4062,Z=0)
    ExecutePin=@n51.Completed
    DriverItem=RigVMVariableNode() {variable=Driver Item}.Value
    OptionalParentItem=RigVMVariableNode() {variable=Driver Item Parent}.Value
    Debug.bDrawDebug=RigVMVariableNode() {variable=Aim Up Ot Viz}.Value
   )
   > n53 = RigVMVariableNode(Value=@n52.OutputParam) {variable=Result Aim Up Ot}
   n54 = &RigVMFunction_ControlFlowBranch(Condition=RigVMVariableNode() {variable=Aim Up Ot Viz}.Value, ExecuteContext=@n53.ExecuteContext)
   n55 = &RigUnit_SphericalPoseReader(
    ActiveRegionScaleFactors=(PositiveWidth=1,NegativeWidth=1,PositiveHeight=1,NegativeHeight=1)
    ActiveRegionSize=0.001
    Debug=(bDrawDebug=true,bDraw2D=False,bDrawLocalAxes=False,DebugScale=25,DebugSegments=20,DebugThickness=0.05)
    DriverAxis=(X=1,Y=0,Z=0)
    FalloffRegionScaleFactors=(PositiveWidth=1,NegativeWidth=1,PositiveHeight=1,NegativeHeight=1)
    FalloffSize=0.18
    FlipHeightScaling=False
    FlipWidthScaling=False
    RotationOffset=(X=147.9195,Y=-7.4723,Z=1.8383)
    ExecutePin=@n54.Completed
    DriverItem=RigVMVariableNode() {variable=Driver Item}.Value
    OptionalParentItem=RigVMVariableNode() {variable=Driver Item Parent}.Value
    Debug.bDrawDebug=RigVMVariableNode() {variable=Aim Dn In Viz}.Value
   )
   > n56 = RigVMVariableNode(Value=@n55.OutputParam) {variable=Result Aim Dn In}
   n57 = &RigVMFunction_ControlFlowBranch(Condition=RigVMVariableNode() {variable=Aim Dn In Viz}.Value, ExecuteContext=@n56.ExecuteContext)
   RigVMDispatchNode(
    Enabled=True
    ScreenColor=(R=1,G=1,B=1,A=1)
    ScreenDuration=0.05
    Prefix=&RigVMFunction_StringConcat(B=" - AIM DN IN:   ", A=RigVMVariableNode() {variable=Driver Item Name}.Value) {template="Concat::Execute(in A,in B,out Result)"}.Result
    ExecuteContext=@n57.True
    Value=RigVMVariableNode() {variable=Result Aim Dn In}.Value
   ) {template="DISPATCH_RigVMDispatch_Print(in Prefix,in Value,in Enabled,in ScreenDuration,in ScreenColor)"}
   RigVMDispatchNode(
    Enabled=True
    ScreenColor=(R=1,G=1,B=1,A=1)
    ScreenDuration=0.05
    Prefix=&RigVMFunction_StringConcat(B=" - AIM UP OT:   ", A=RigVMVariableNode() {variable=Driver Item Name}.Value) {template="Concat::Execute(in A,in B,out Result)"}.Result
    ExecuteContext=@n54.True
    Value=RigVMVariableNode() {variable=Result Aim Up Ot}.Value
   ) {template="DISPATCH_RigVMDispatch_Print(in Prefix,in Value,in Enabled,in ScreenDuration,in ScreenColor)"}
   RigVMDispatchNode(
    Enabled=True
    ScreenColor=(R=1,G=1,B=1,A=1)
    ScreenDuration=0.05
    Prefix=&RigVMFunction_StringConcat(B=" - AIM UP IN:   ", A=RigVMVariableNode() {variable=Driver Item Name}.Value) {template="Concat::Execute(in A,in B,out Result)"}.Result
    ExecuteContext=@n51.True
    Value=RigVMVariableNode() {variable=Result Aim Up In}.Value
   ) {template="DISPATCH_RigVMDispatch_Print(in Prefix,in Value,in Enabled,in ScreenDuration,in ScreenColor)"}
   n58 = &RigUnit_SphericalPoseReader(
    ActiveRegionScaleFactors=(PositiveWidth=1,NegativeWidth=1,PositiveHeight=1,NegativeHeight=1)
    ActiveRegionSize=0.001
    Debug=(bDrawDebug=true,bDraw2D=False,bDrawLocalAxes=False,DebugScale=25,DebugSegments=20,DebugThickness=0.05)
    DriverAxis=(X=1,Y=0,Z=0)
    ExecutePin=()
    FalloffRegionScaleFactors=(PositiveWidth=1,NegativeWidth=1,PositiveHeight=1,NegativeHeight=1)
    FalloffSize=0.18
    FlipHeightScaling=False
    FlipWidthScaling=False
    RotationOffset=(X=147.9195,Y=-7.4723,Z=1.8383)
    DriverItem=RigVMVariableNode() {variable=Driver Item}.Value
    OptionalParentItem=RigVMVariableNode() {variable=Driver Item Parent}.Value
    Debug.bDrawDebug=RigVMVariableNode() {variable=Aim Dn In Viz}.Value
   )
   > n59 = RigVMVariableNode(Value=@n58.OutputParam) {variable=Result Aim Dn In}
   n60 = &RigVMFunction_ControlFlowBranch(Condition=RigVMVariableNode() {variable=Aim Dn In Viz}.Value, ExecuteContext=@n59.ExecuteContext)
   RigVMDispatchNode(
    Enabled=True
    ScreenColor=(R=1,G=1,B=1,A=1)
    ScreenDuration=0.05
    Prefix=&RigVMFunction_StringConcat(B=" - AIM DN IN:   ", A=RigVMVariableNode() {variable=Driver Item Name}.Value) {template="Concat::Execute(in A,in B,out Result)"}.Result
    ExecuteContext=@n60.True
    Value=RigVMVariableNode() {variable=Result Aim Dn In}.Value
   ) {template="DISPATCH_RigVMDispatch_Print(in Prefix,in Value,in Enabled,in ScreenDuration,in ScreenColor)"}
  ]}
 New Function_ContainedGraph {nodes [
   RigVMFunctionEntryNode()
   > SequenceAggregate_1 = RigVMAggregateNode()
   n1 = &RigUnit_SphericalPoseReader(
    ActiveRegionScaleFactors=(PositiveWidth=1,NegativeWidth=1,PositiveHeight=1,NegativeHeight=1)
    ActiveRegionSize=0.15
    Debug=(bDrawDebug=false,bDraw2D=False,bDrawLocalAxes=False,DebugScale=25,DebugSegments=20,DebugThickness=0.25)
    DriverAxis=(X=1,Y=0,Z=0)
    DriverItem=(Type=Bone,Name="upperarm_l")
    FalloffRegionScaleFactors=(PositiveWidth=0.5,NegativeWidth=0.5,PositiveHeight=0.5,NegativeHeight=0.5)
    FalloffSize=0.4
    FlipHeightScaling=False
    FlipWidthScaling=False
    OptionalParentItem=(Type=Bone,Name="spine_05")
    RotationOffset=(X=0,Y=-125,Z=0)
    ExecutePin=@SequenceAggregate_1.A
   )
   > RigVMVariableNode(Value=@n1.OutputParam) {variable=SPR_upperarm_l_down_A_alpha}
   > n2 = &RigUnit_SphericalPoseReader(
    ActiveRegionScaleFactors=(PositiveWidth=1,NegativeWidth=1,PositiveHeight=1,NegativeHeight=1)
    ActiveRegionSize=0.15
    Debug=(bDrawDebug=true,bDraw2D=False,bDrawLocalAxes=False,DebugScale=25,DebugSegments=20,DebugThickness=0.25)
    DriverAxis=(X=-1,Y=0,Z=0)
    DriverItem=(Type=Bone,Name="upperarm_r")
    FalloffRegionScaleFactors=(PositiveWidth=1,NegativeWidth=1,PositiveHeight=1,NegativeHeight=1)
    FalloffSize=0.4
    FlipHeightScaling=False
    FlipWidthScaling=False
    OptionalParentItem=(Type=Bone,Name="None")
    RotationOffset=(X=180,Y=-125,Z=0)
   )
   > RigVMVariableNode(Value=@n2.OutputParam) {variable=SPR_upperarm_r_down_A_alpha}
   > n3 = &RigUnit_SphericalPoseReader(
    ActiveRegionScaleFactors=(PositiveWidth=1,NegativeWidth=1,PositiveHeight=1,NegativeHeight=1)
    ActiveRegionSize=0.15
    Debug=(bDrawDebug=true,bDraw2D=False,bDrawLocalAxes=false,DebugScale=25,DebugSegments=20,DebugThickness=0.25)
    DriverAxis=(X=-1,Y=0,Z=0)
    DriverItem=(Type=Bone,Name="upperarm_r")
    FalloffRegionScaleFactors=(PositiveWidth=1,NegativeWidth=1,PositiveHeight=1,NegativeHeight=1)
    FalloffSize=0.4
    FlipHeightScaling=False
    FlipWidthScaling=False
    OptionalParentItem=(Type=Bone,Name="None")
    RotationOffset=(X=180,Y=-125,Z=0)
   )
   > RigVMVariableNode(Value=@n3.OutputParam) {variable=SPR_upperarm_r_down_B_alpha}
   SequenceAggregate = &RigVMFunction_Sequence(ExecuteContext=@SequenceAggregate_1.C)
   RigUnit_ItemArray = &RigUnit_ItemArray(Items=((Type=Bone,Name="clavicle_l"),(Type=Bone,Name="upperarm_l"),(Type=Bone,Name="lowerarm_l")))
   DISPATCH_RigVMDispatch_ArrayIterator = RigVMDispatchNode(Array=@RigUnit_ItemArray.Items, ExecuteContext=@SequenceAggregate.A) {template="DISPATCH_RigVMDispatch_ArrayIterator(in Array,out Element,out Index,out Count,out Ratio)"}
   DISPATCH_RigVMDispatch_ArrayGetAtIndex = RigVMDispatchNode(
    Array=&RigUnit_ItemArray(Items=((Type=Bone,Name="None"),(Type=Bone,Name="None"),(Type=Bone,Name="None"))).Items
    Index=@DISPATCH_RigVMDispatch_ArrayIterator.Index
   ) {template="DISPATCH_RigVMDispatch_ArrayGetAtIndex(in Array,in Index,out Element)"}
   > &RigUnit_SetRotation(
    Space=GlobalSpace
    Weight=1
    bInitial=False
    bPropagateToChildren=true
    Item=@DISPATCH_RigVMDispatch_ArrayGetAtIndex.Element
    Value=&RigUnit_GetTransform(Item=@DISPATCH_RigVMDispatch_ArrayIterator.Element).Transform.Rotation
   )
   PBIK = /Script/PBIK.RigUnit_PBIK(
    BoneSettings=((Bone="None",X=Locked,Y=Locked,Z=Free,PreferredAngles=(X=0,Y=0,Z=0)))
    Debug=(DrawScale=1)
    Effectors=((Bone="None",Transform=(Rotation=(X=0,Y=0,Z=0,W=1),Translation=(X=0,Y=0,Z=0),Scale3D=(X=1,Y=1,Z=1)),PositionAlpha=1,RotationAlpha=1,StrengthAlpha=1,PinRotation=1,ChainDepth=0))
    ExcludedBones=()
    Root=None
    Settings="(Iterations=20,SubIterations=0,MassMultiplier=1,bAllowStretch=False,RootBehavior=PinToInput,PrePullRootSettings=(RotationAlpha=0,RotationAlphaX=1,RotationAlphaY=1,RotationAlphaZ=1,PositionAlpha=1,PositionAlphaX=1,PositionAlphaY=1,PositionAlphaZ=1),GlobalPu ...(truncated)"
    Effectors.0.Transform=&RigUnit_GetTransform(Item=(Type=Bone,Name="hand_l")).Transform
    ExecutePin=@DISPATCH_RigVMDispatch_ArrayIterator.Completed
   ) {is=struct RigVMUnitNode}
   n4 = &RigUnit_ItemArray(Items=((Type=Bone,Name="upperarm_twist_01_l"),(Type=Bone,Name="upperarm_twist_02_l"),(Type=Bone,Name="lowerarm_twist_01_l"),(Type=Bone,Name="lowerarm_twist_02_l")))
   > n5 = RigVMDispatchNode(Array=@n4.Items) {template="DISPATCH_RigVMDispatch_ArrayIterator(in Array,out Element,out Index,out Count,out Ratio)"}
   > &RigUnit_SetTransform(
    Weight=1
    bInitial=False
    Item=RigVMDispatchNode(
     Array=&RigUnit_ItemArray(Items=((Type=Bone,Name="None"),(Type=Bone,Name="None"),(Type=Bone,Name="None"),(Type=Bone,Name="None"))).Items
     Index=@n5.Index
    ) {template="DISPATCH_RigVMDispatch_ArrayGetAtIndex(in Array,in Index,out Element)"}.Element
    Value=&RigUnit_GetTransform(Item=@n5.Element).Transform
   )
   n6 = RigVMDispatchNode(
    Array=&RigUnit_ItemArray(Items=((Type=Bone,Name="clavicle_r"),(Type=Bone,Name="upperarm_r"),(Type=Bone,Name="lowerarm_r"))).Items
    ExecuteContext=@SequenceAggregate.B
   ) {template="DISPATCH_RigVMDispatch_ArrayIterator(in Array,out Element,out Index,out Count,out Ratio)"}
   > &RigUnit_SetRotation(
    Space=GlobalSpace
    Weight=1
    bInitial=False
    bPropagateToChildren=true
    Item=RigVMDispatchNode(
     Array=&RigUnit_ItemArray(Items=((Type=Bone,Name="None"),(Type=Bone,Name="None"),(Type=Bone,Name="None"))).Items
     Index=@n6.Index
    ) {template="DISPATCH_RigVMDispatch_ArrayGetAtIndex(in Array,in Index,out Element)"}.Element
    Value=&RigUnit_GetTransform(Item=@n6.Element).Transform.Rotation
   )
   &RigUnit_PBIK(
    BoneSettings=((Bone="None",X=Locked,Y=Locked,Z=Free,PreferredAngles=(X=0,Y=0,Z=0)))
    Debug=(DrawScale=1)
    Effectors=((Bone="None",Transform=(Rotation=(X=0,Y=0,Z=0,W=1),Translation=(X=0,Y=0,Z=0),Scale3D=(X=1,Y=1,Z=1)),PositionAlpha=1,RotationAlpha=1,StrengthAlpha=1,PinRotation=1,ChainDepth=0))
    ExcludedBones=()
    Root=None
    Settings="(Iterations=20,SubIterations=0,MassMultiplier=1,bAllowStretch=False,RootBehavior=PinToInput,PrePullRootSettings=(RotationAlpha=0,RotationAlphaX=1,RotationAlphaY=1,RotationAlphaZ=1,PositionAlpha=1,PositionAlphaX=1,PositionAlphaY=1,PositionAlphaZ=1),GlobalPu ...(truncated)"
    Effectors.0.Transform=&RigUnit_GetTransform(Item=(Type=Bone,Name="hand_r")).Transform
    ExecutePin=@n6.Completed
   )
   n7 = &RigUnit_ItemArray(Items=((Type=Bone,Name="upperarm_twist_01_r"),(Type=Bone,Name="upperarm_twist_02_r"),(Type=Bone,Name="lowerarm_twist_01_r"),(Type=Bone,Name="lowerarm_twist_02_r")))
   > n8 = RigVMDispatchNode(Array=@n7.Items) {template="DISPATCH_RigVMDispatch_ArrayIterator(in Array,out Element,out Index,out Count,out Ratio)"}
   > &RigUnit_SetTransform(
    Weight=1
    bInitial=False
    Item=RigVMDispatchNode(
     Array=&RigUnit_ItemArray(Items=((Type=Bone,Name="None"),(Type=Bone,Name="None"),(Type=Bone,Name="None"),(Type=Bone,Name="None"))).Items
     Index=@n8.Index
    ) {template="DISPATCH_RigVMDispatch_ArrayGetAtIndex(in Array,in Index,out Element)"}.Element
    Value=&RigUnit_GetTransform(Item=@n8.Element).Transform
   )
   SequenceAggregate_2 = &RigVMFunction_Sequence(ExecuteContext=@SequenceAggregate_1.D)
   Subtract = /Script/RigVM.RigVMFunction_MathVectorSub(
    B=&RigUnit_GetTransform(Item=(Type=Bone,Name="None")).Transform.Translation
    A=&RigUnit_GetTransform(Item=(Type=Bone,Name="None")).Transform.Translation
   ) {is=struct RigVMUnitNode, template="Subtract::Execute(in A,in B,out Result)"}
   Unit = /Script/RigVM.RigVMFunction_MathVectorUnit(Value=@Subtract.Result) {is=struct RigVMUnitNode, template="Unit::Execute(in Value,out Result)"}
   RigVMFunction_MathQuaternionFromAxisAndAngle = /Script/RigVM.RigVMFunction_MathQuaternionFromAxisAndAngle(Angle=-0.5, Axis=@Unit.Result) {is=struct RigVMUnitNode, template="FromAxisAndAngle::Execute(in Axis,in Angle,out Result)"}
   Set Transform = &RigUnit_SetRotation(
    Item=(Type=Bone,Name="None")
    Space=GlobalSpace
    bInitial=false
    Value=/Script/RigVM.RigVMFunction_MathQuaternionMul(
     B=&RigUnit_GetTransform(Item=(Type=Bone,Name="None")).Transform.Rotation
     A=@RigVMFunction_MathQuaternionFromAxisAndAngle.Result
    ) {is=struct RigVMUnitNode, template="Multiply::Execute(in A,in B,out Result)"}.Result
    Weight=&RigVMFunction_MathDoubleMul(A=RigVMVariableNode() {variable=mult_spr_dst_lowerarm_r_pelvis_alpha}.Value, B=RigVMVariableNode() {variable=GauntletElbowAdjustAlpha}.Value).Result
    ExecutePin=@SequenceAggregate_2.A
   )
   n9 = &RigVMFunction_MathVectorUnit(Value=&RigVMFunction_MathVectorSub(
     B=&RigUnit_GetTransform(Item=(Type=Bone,Name="None"), bInitial=False).Transform.Translation
     A=&RigUnit_GetTransform(Item=(Type=Bone,Name="None"), bInitial=False).Transform.Translation
    ) {template="Subtract::Execute(in A,in B,out Result)"}.Result) {template="Unit::Execute(in Value,out Result)"}
   Set Transform_1 = &RigUnit_SetRotation(
    Item=(Type=Bone,Name="None")
    Space=GlobalSpace
    bInitial=False
    Value=&RigVMFunction_MathQuaternionMul(
     B=&RigUnit_GetTransform(Item=(Type=Bone,Name="None"), bInitial=False).Transform.Rotation
     A=&RigVMFunction_MathQuaternionFromAxisAndAngle(Angle=0.35, Axis=@n9.Result) {template="FromAxisAndAngle::Execute(in Axis,in Angle,out Result)"}.Result
    ) {template="Multiply::Execute(in A,in B,out Result)"}.Result
    Weight=&RigVMFunction_MathDoubleMul(B=RigVMVariableNode() {variable=GauntletElbowAdjustAlpha}.Value, A=RigVMVariableNode() {variable=SPR_upperarm_r_down_B_alpha}.Value).Result
    ExecutePin=@SequenceAggregate_2.B
   )
   RigVMFunctionReturnNode(ExecuteContext=@SequenceAggregate_1.G)
   Remap = &RigVMFunction_MathDoubleRemap(
    SourceMaximum=40
    SourceMinimum=16
    TargetMaximum=1
    TargetMinimum=0
    Value=&RigVMFunction_MathVectorDistance(
     A=&RigUnit_GetTransform(Item=(Type=Bone,Name="pelvis"), bInitial=False).Transform.Translation
     B=&RigUnit_GetTransform(Item=(Type=Bone,Name="hand_l"), bInitial=False).Transform.Translation
    ).Result
   )
   Multiply = &RigVMFunction_MathDoubleMul(B=&RigVMFunction_MathDoubleSub(A=1, B=@Remap.Result).Result, A=RigVMVariableNode() {variable=SPR_upperarm_l_down_A_alpha}.Value)
   RigVMVariableNode(Value=@Multiply.Result, ExecuteContext=@SequenceAggregate_1.B) {variable=mult_spr_dst_lowerarm_r_pelvis_alpha}
   RigVMFunction_MathVectorDistance = &RigVMFunction_MathVectorDistance(
    A=&RigUnit_GetTransform(Item=(Type=Bone,Name="hand_r"), bInitial=False).Transform.Translation
    B=&RigUnit_GetTransform(Item=(Type=Bone,Name="thigh_twist_01_r"), bInitial=False).Transform.Translation
   )
   > RigVMVariableNode(Value=&RigVMFunction_MathDoubleSub(
     A=1
     B=&RigVMFunction_MathDoubleRemap(SourceMaximum=40, SourceMinimum=16, TargetMaximum=1, TargetMinimum=0, Value=@RigVMFunction_MathVectorDistance.Result).Result
    ).Result) {variable=mult_spr_dst_hand_r_thigh_twist_r_alpha}
   n10 = &RigVMFunction_MathQuaternionSwingTwist(
    TwistAxis=(X=1,Y=0,Z=0)
    Input=&RigVMFunction_MathTransformMakeRelative(
     Parent=&RigUnit_GetTransform(Item=(Type=Bone,Name="hand_l"), Space=LocalSpace, bInitial=true).Transform
     Global=&RigUnit_GetTransform(Item=(Type=Bone,Name="hand_l"), Space=LocalSpace).Transform
    ).Local.Rotation
   )
   &RigUnit_SetRotation(Item=(Type=Bone,Name="None"), Weight=0.75, bPropagateToChildren=false, Value=@n10.Twist, ExecutePin=@SequenceAggregate_1.E)
   > &RigUnit_SetRotation(Item=(Type=Bone,Name="None"), Weight=0.5, bPropagateToChildren=false, Value=@n10.Twist)
   n11 = &RigVMFunction_MathQuaternionSwingTwist(
    TwistAxis=(X=1,Y=0,Z=0)
    Input=&RigVMFunction_MathTransformMakeRelative(
     Parent=&RigUnit_GetTransform(Item=(Type=Bone,Name="hand_r"), Space=LocalSpace, bInitial=true).Transform
     Global=&RigUnit_GetTransform(Item=(Type=Bone,Name="hand_r"), Space=LocalSpace).Transform
    ).Local.Rotation
   )
   > &RigUnit_SetRotation(Item=(Type=Bone,Name="None"), Weight=0.75, bPropagateToChildren=false, Value=@n11.Twist)
   > &RigUnit_SetRotation(Item=(Type=Bone,Name="None"), Weight=0.5, bPropagateToChildren=false, Value=@n11.Twist)
   &RigUnit_ModifyTransforms(
    ItemToModify=((Item=(Type=Bone,Name="def_shirt_fr_l"),Transform=(Rotation=(X=0,Y=0,Z=0,W=1),Translation=(X=0,Y=-6,Z=0),Scale3D=(X=1,Y=1,Z=1))))
    WeightMinimum=0
    Weight=&RigVMFunction_MathDoubleMul(B=RigVMVariableNode() {variable=IsNotEmotingAlpha}.Value, A=RigVMVariableNode() {variable=mult_spr_dst_lowerarm_r_pelvis_alpha}.Value).Result
    ExecutePin=@SequenceAggregate_1.F
   )
   > n12 = &RigUnit_SphericalPoseReader(
    ActiveRegionScaleFactors=(PositiveWidth=1,NegativeWidth=1,PositiveHeight=1,NegativeHeight=1)
    ActiveRegionSize=0.15
    Debug=(bDrawDebug=false,bDraw2D=False,bDrawLocalAxes=False,DebugScale=25,DebugSegments=20,DebugThickness=0.25)
    DriverAxis=(X=1,Y=0,Z=0)
    DriverItem=(Type=Bone,Name="upperarm_l")
    FalloffRegionScaleFactors=(PositiveWidth=1,NegativeWidth=1,PositiveHeight=1,NegativeHeight=1)
    FalloffSize=0.4
    FlipHeightScaling=False
    FlipWidthScaling=False
    OptionalParentItem=(Type=Bone,Name="None")
    RotationOffset=(X=0,Y=-125,Z=0)
   )
   > &RigUnit_ModifyTransforms(
    ItemToModify=((Item=(Type=Bone,Name="stomach_deform_l"),Transform=(Rotation=(X=0,Y=0,Z=0,W=1),Translation=(X=-5,Y=0,Z=0),Scale3D=(X=1,Y=1,Z=1))))
    WeightMinimum=0
    Weight=@n12.OutputParam
   )
   > n13 = &RigUnit_SphericalPoseReader(
    ActiveRegionScaleFactors=(PositiveWidth=1,NegativeWidth=1,PositiveHeight=1,NegativeHeight=1)
    ActiveRegionSize=0.15
    Debug=(bDrawDebug=false,bDraw2D=False,bDrawLocalAxes=False,DebugScale=25,DebugSegments=20,DebugThickness=0.25)
    DriverAxis=(X=-1,Y=0,Z=0)
    DriverItem=(Type=Bone,Name="upperarm_r")
    FalloffRegionScaleFactors=(PositiveWidth=1,NegativeWidth=1,PositiveHeight=1,NegativeHeight=1)
    FalloffSize=0.4
    FlipHeightScaling=false
    FlipWidthScaling=false
    OptionalParentItem=(Type=Bone,Name="None")
    RotationOffset=(X=180,Y=-125,Z=0)
   )
   > &RigUnit_ModifyTransforms(
    ItemToModify=((Item=(Type=Bone,Name="stomach_deform_r"),Transform=(Rotation=(X=0,Y=0,Z=0,W=1),Translation=(X=5,Y=0,Z=0),Scale3D=(X=1,Y=1,Z=1))))
    WeightMinimum=0
    Weight=@n13.OutputParam
   )
  ]}
 RigVMFunctionLibrary {nodes [
   ArmElbowAdjustment = RigVMCollapseNode(ExecuteContext=())
   Elbow Readers = RigVMCollapseNode(ExecuteContext=(), driver item=(Type=None,Name="None"))
   Thigh Readers = RigVMCollapseNode(Driver Item=(Type=None,Name="None"), ExecuteContext=())
   UpperArm Readers = RigVMCollapseNode(Driver Item=(Type=None,Name="None"), ExecuteContext=())
  ]}
 RigVMFunction_Sequence_1_ContainedGraph {nodes [
   n1 = RigVMFunctionEntryNode()
   RigVMFunction_Sequence = &RigVMFunction_Sequence(ExecuteContext=@n1.ExecuteContext)
   n2 = &RigVMFunction_Sequence(ExecuteContext=@RigVMFunction_Sequence.B)
   n3 = &RigVMFunction_Sequence(ExecuteContext=@n2.B)
   n4 = &RigVMFunction_Sequence(ExecuteContext=@n3.B)
   n5 = &RigVMFunction_Sequence(ExecuteContext=@n4.B)
   n6 = &RigVMFunction_Sequence(ExecuteContext=@n5.B)
   n7 = &RigVMFunction_Sequence(ExecuteContext=@n6.B)
   n8 = &RigVMFunction_Sequence(ExecuteContext=@n7.B)
   n9 = &RigVMFunction_Sequence(ExecuteContext=@n8.B)
   n10 = &RigVMFunction_Sequence(ExecuteContext=@n9.B)
   n11 = &RigVMFunction_Sequence(ExecuteContext=@n10.B)
   n12 = &RigVMFunction_Sequence(ExecuteContext=@n11.B)
   n13 = &RigVMFunction_Sequence(ExecuteContext=@n12.B)
   n14 = &RigVMFunction_Sequence(ExecuteContext=@n13.B)
   RigVMFunctionReturnNode(
    A=@RigVMFunction_Sequence.A
    B=@n2.A
    K=@n11.A
    L=@n12.A
    M=@n13.A
    N=@n14.A
    O=@n14.B
    C=@n3.A
    D=@n4.A
    E=@n5.A
    F=@n6.A
    G=@n7.A
    H=@n8.A
    I=@n9.A
    J=@n10.A
   )
  ]}
 RigVMFunction_Sequence_1_ContainedGraph {nodes [
   n1 = RigVMFunctionEntryNode()
   RigVMFunction_Sequence = &RigVMFunction_Sequence(ExecuteContext=@n1.ExecuteContext)
   n2 = &RigVMFunction_Sequence(ExecuteContext=@RigVMFunction_Sequence.B)
   n3 = &RigVMFunction_Sequence(ExecuteContext=@n2.B)
   n4 = &RigVMFunction_Sequence(ExecuteContext=@n3.B)
   n5 = &RigVMFunction_Sequence(ExecuteContext=@n4.B)
   RigVMFunctionReturnNode(A=@RigVMFunction_Sequence.A, B=@n2.A, C=@n3.A, D=@n4.A, E=@n5.A, F=@n5.B)
  ]}
 RigVMFunction_Sequence_1_ContainedGraph {nodes [
   n1 = RigVMFunctionEntryNode()
   RigVMFunction_Sequence = &RigVMFunction_Sequence(ExecuteContext=@n1.ExecuteContext)
   n2 = &RigVMFunction_Sequence(ExecuteContext=@RigVMFunction_Sequence.B)
   n3 = &RigVMFunction_Sequence(ExecuteContext=@n2.B)
   n4 = &RigVMFunction_Sequence(ExecuteContext=@n3.B)
   n5 = &RigVMFunction_Sequence(ExecuteContext=@n4.B)
   RigVMFunctionReturnNode(A=@RigVMFunction_Sequence.A, B=@n2.A, C=@n3.A, D=@n4.A, E=@n5.A, F=@n5.B)
  ]}
 RigVMModel {nodes [
   RigUnit_BeginExecution = &RigUnit_BeginExecution()
   RigVMFunction_Sequence = RigVMAggregateNode(ExecuteContext=@RigUnit_BeginExecution.ExecutePin)
   n1 = &RigUnit_SphericalPoseReader(
    ActiveRegionScaleFactors=(PositiveWidth=1,NegativeWidth=1,PositiveHeight=1,NegativeHeight=1)
    ActiveRegionSize=0.001
    Debug=(bDrawDebug=false,bDraw2D=False,bDrawLocalAxes=False,DebugScale=25,DebugSegments=20,DebugThickness=0.25)
    DriverAxis=(X=1,Y=0,Z=0)
    DriverItem=(Type=Bone,Name="clavicle_l")
    FalloffRegionScaleFactors=(PositiveWidth=1,NegativeWidth=1,PositiveHeight=1,NegativeHeight=1)
    FalloffSize=0.5
    FlipHeightScaling=False
    FlipWidthScaling=False
    OptionalParentItem=(Type=Bone,Name="None")
    RotationOffset=(X=90,Y=0,Z=0)
    ExecutePin=@RigVMFunction_Sequence.F
   )
   > &RigUnit_ModifyTransforms(
    ItemToModify=((Item=(Type=Bone,Name="def_clavicle_l_bt"),Transform=(Rotation=(X=0,Y=0,Z=0.5373,W=0.8434),Translation=(X=-6.0792,Y=17.0739,Z=0.4163),Scale3D=(X=1,Y=1,Z=1))))
    WeightMinimum=0
    Weight=@n1.OutputParam
   )
   > n2 = &RigUnit_SphericalPoseReader(
    ActiveRegionScaleFactors=(PositiveWidth=1,NegativeWidth=1,PositiveHeight=1,NegativeHeight=1)
    ActiveRegionSize=0.001
    Debug=(bDrawDebug=false,bDraw2D=False,bDrawLocalAxes=False,DebugScale=25,DebugSegments=20,DebugThickness=0.25)
    DriverAxis=(X=1,Y=0,Z=0)
    DriverItem=(Type=Bone,Name="clavicle_l")
    FalloffRegionScaleFactors=(PositiveWidth=1,NegativeWidth=1,PositiveHeight=1,NegativeHeight=1)
    FalloffSize=0.5
    FlipHeightScaling=False
    FlipWidthScaling=False
    OptionalParentItem=(Type=Bone,Name="None")
    RotationOffset=(X=-90,Y=0,Z=0)
   )
   > &RigUnit_ModifyTransforms(
    ItemToModify=((Item=(Type=Bone,Name="def_clavicle_l_bt"),Transform=(Rotation=(X=0,Y=0,Z=-0.5373,W=0.8434),Translation=(X=-6.0792,Y=-17.0739,Z=0.4163),Scale3D=(X=1,Y=1,Z=1))))
    WeightMinimum=0
    Weight=@n2.OutputParam
   )
   > n3 = &RigUnit_SphericalPoseReader(
    ActiveRegionScaleFactors=(PositiveWidth=1,NegativeWidth=1,PositiveHeight=1,NegativeHeight=1)
    ActiveRegionSize=0.001
    Debug=(bDrawDebug=false,bDraw2D=False,bDrawLocalAxes=False,DebugScale=25,DebugSegments=20,DebugThickness=0.25)
    DriverAxis=(X=1,Y=0,Z=0)
    DriverItem=(Type=Bone,Name="clavicle_l")
    FalloffRegionScaleFactors=(PositiveWidth=1,NegativeWidth=1,PositiveHeight=1,NegativeHeight=1)
    FalloffSize=0.5
    FlipHeightScaling=False
    FlipWidthScaling=False
    OptionalParentItem=(Type=Bone,Name="None")
    RotationOffset=(X=0,Y=0,Z=0)
   )
   > &RigUnit_ModifyTransforms(
    ItemToModify="((Item=(Type=Bone,Name=\"def_clavicle_l_bt\"),Transform=(Rotation=(X=0,Y=0,Z=0,W=1),Translation=(X=0,Y=0,Z=22.1423),Scale3D=(X=1,Y=1,Z=1))),(Item=(Type=Bone,Name=\"def_clavicle_l_tp\"),Transform=(Rotation=(X=0,Y=0.56,Z=0,W=0.8285),Translation=(X=9.3604,Y=0,Z=- ...(truncated)"
    WeightMinimum=0
    Weight=@n3.OutputParam
   )
   > n4 = &RigUnit_SphericalPoseReader(
    ActiveRegionScaleFactors=(PositiveWidth=1,NegativeWidth=1,PositiveHeight=1,NegativeHeight=1)
    ActiveRegionSize=0.001
    Debug=(bDrawDebug=false,bDraw2D=False,bDrawLocalAxes=False,DebugScale=25,DebugSegments=20,DebugThickness=0.25)
    DriverAxis=(X=1,Y=0,Z=0)
    DriverItem=(Type=Bone,Name="clavicle_r")
    FalloffRegionScaleFactors=(PositiveWidth=1,NegativeWidth=1,PositiveHeight=1,NegativeHeight=1)
    FalloffSize=0.5
    FlipHeightScaling=False
    FlipWidthScaling=False
    OptionalParentItem=(Type=Bone,Name="None")
    RotationOffset=(X=90,Y=0,Z=0)
   )
   > &RigUnit_ModifyTransforms(
    ItemToModify=((Item=(Type=Bone,Name="def_clavicle_r_bt"),Transform=(Rotation=(X=0,Y=0,Z=0.5,W=0.866),Translation=(X=5.0473,Y=-16.5283,Z=2.9483),Scale3D=(X=1,Y=1,Z=1))))
    WeightMinimum=0
    Weight=@n4.OutputParam
   )
   > n5 = &RigUnit_SphericalPoseReader(
    ActiveRegionScaleFactors=(PositiveWidth=1,NegativeWidth=1,PositiveHeight=1,NegativeHeight=1)
    ActiveRegionSize=0.001
    Debug=(bDrawDebug=false,bDraw2D=False,bDrawLocalAxes=False,DebugScale=25,DebugSegments=20,DebugThickness=0.25)
    DriverAxis=(X=1,Y=0,Z=0)
    DriverItem=(Type=Bone,Name="clavicle_r")
    FalloffRegionScaleFactors=(PositiveWidth=1,NegativeWidth=1,PositiveHeight=1,NegativeHeight=1)
    FalloffSize=0.5
    FlipHeightScaling=False
    FlipWidthScaling=False
    OptionalParentItem=(Type=Bone,Name="None")
    RotationOffset=(X=-90,Y=0,Z=0)
   )
   > &RigUnit_ModifyTransforms(
    ItemToModify=((Item=(Type=Bone,Name="def_clavicle_r_bt"),Transform=(Rotation=(X=0,Y=0,Z=-0.5373,W=0.8434),Translation=(X=11.782,Y=15.6068,Z=0),Scale3D=(X=1,Y=1,Z=1))))
    WeightMinimum=0
    Weight=@n5.OutputParam
   )
   > n6 = &RigUnit_SphericalPoseReader(
    ActiveRegionScaleFactors=(PositiveWidth=1,NegativeWidth=1,PositiveHeight=1,NegativeHeight=1)
    ActiveRegionSize=0.001
    Debug=(bDrawDebug=false,bDraw2D=False,bDrawLocalAxes=False,DebugScale=25,DebugSegments=20,DebugThickness=0.25)
    DriverAxis=(X=1,Y=0,Z=0)
    DriverItem=(Type=Bone,Name="clavicle_r")
    FalloffRegionScaleFactors=(PositiveWidth=1,NegativeWidth=1,PositiveHeight=1,NegativeHeight=1)
    FalloffSize=0.5
    FlipHeightScaling=False
    FlipWidthScaling=False
    OptionalParentItem=(Type=Bone,Name="None")
    RotationOffset=(X=0,Y=0,Z=0)
   )
   > &RigUnit_ModifyTransforms(
    ItemToModify="((Item=(Type=Bone,Name=\"def_clavicle_r_bt\"),Transform=(Rotation=(X=0,Y=-0.5661,Z=0,W=0.8243),Translation=(X=1.7194,Y=0,Z=-20.6673),Scale3D=(X=1,Y=1,Z=1))),(Item=(Type=Bone,Name=\"def_clavicle_r_fr\"),Transform=(Rotation=(X=0,Y=0.3827,Z=0,W=0.9239),Translatio ...(truncated)"
    WeightMinimum=0
    Weight=@n6.OutputParam
   )
   RigVMFunction_ControlFlowBranch = &RigVMFunction_ControlFlowBranch(Condition=RigVMVariableNode() {variable=Debug Controls On}.Value, ExecuteContext=@RigVMFunction_Sequence.A)
   For_Each = RigVMDispatchNode(ExecuteContext=@RigVMFunction_ControlFlowBranch.True, Array=RigVMVariableNode() {variable=Debug Ctrls}.Value) {template="DISPATCH_RigVMDispatch_ArrayIterator(in Array,out Element,out Index,out Count,out Ratio)"}
   > &RigUnit_SetTransform(
    Weight=1
    bInitial=False
    Item=RigVMDispatchNode(Array=RigVMVariableNode() {variable=Debug Bones}.Value, Index=@For_Each.Index) {template="DISPATCH_RigVMDispatch_ArrayGetAtIndex(in Array,in Index,out Element)"}.Element
    Value=&RigUnit_GetTransform(bInitial=False, Item=@For_Each.Element).Transform
   )
   &RigUnit_ModifyTransforms(
    ItemToModify=((Item=(Type=Bone,Name="def_crease_r_fr"),Transform=(Rotation=(X=-0.0001,Y=0.0001,Z=0.0098,W=1),Translation=(X=-3.4376,Y=4.2651,Z=0),Scale3D=(X=1,Y=1,Z=1))))
    WeightMinimum=0
    Weight=RigVMVariableNode() {variable=Thigh R Fr Pre}.Value
    ExecutePin=@RigVMFunction_Sequence.D
   )
   n7 = RigVMVariableNode() {variable=Thigh R Fr}
   > &RigUnit_ModifyTransforms(
    ItemToModify="((Item=(Type=Bone,Name=\"def_crease_r_in\"),Transform=(Rotation=(X=0,Y=0,Z=0,W=1),Translation=(X=2.8656,Y=-5.7114,Z=1.4162),Scale3D=(X=1,Y=1,Z=1))),(Item=(Type=Bone,Name=\"def_crease_r_fr\"),Transform=(Rotation=(X=0,Y=0,Z=0.7345,W=-0.6787),Translation=(X=-33.9 ...(truncated)"
    WeightMinimum=0
    Weight=@n7.Value
   )
   > &RigUnit_ModifyTransforms(
    ItemToModify="((Item=(Type=Bone,Name=\"def_shirt_fr_r\"),Transform=(Rotation=(X=0.0062,Y=0.0995,Z=-0.3193,W=0.9424),Translation=(X=16.5929,Y=4.3504,Z=-3.0547),Scale3D=(X=0.5,Y=1,Z=1))),(Item=(Type=Bone,Name=\"def_shirt_fr_l\"),Transform=(Rotation=(X=0,Y=0.0565,Z=0,W=0.9984) ...(truncated)"
    WeightMinimum=0
    Weight=&RigVMFunction_MathDoubleClamp(Value=&RigVMFunction_MathDoubleRemap(SourceMaximum=1, SourceMinimum=0.45, TargetMaximum=1, TargetMinimum=0, bClamp=False, Value=@n7.Value).Result).Result
   )
   n8 = RigVMVariableNode() {variable=Thigh R Fr Post}
   > &RigUnit_ModifyTransforms(
    ItemToModify="((Item=(Type=Bone,Name=\"def_crease_r_ba\"),Transform=(Rotation=(X=0,Y=0,Z=-0.5653,W=0.8249),Translation=(X=5.9719,Y=5.9146,Z=0),Scale3D=(X=1,Y=1,Z=1))),(Item=(Type=Bone,Name=\"def_crease_r_fr\"),Transform=(Rotation=(X=0,Y=0,Z=0.3827,W=-0.9239),Translation=(X= ...(truncated)"
    WeightMinimum=0
    Weight=@n8.Value
   )
   > &RigUnit_ModifyTransforms(
    ItemToModify="((Item=(Type=Bone,Name=\"def_shirt_fr_r\"),Transform=(Rotation=(X=0,Y=0,Z=-0.5085,W=0.8611),Translation=(X=50.5818,Y=9.5932,Z=-8.6742),Scale3D=(X=0.4978,Y=1,Z=1.2412))),(Item=(Type=Bone,Name=\"def_shirt_fr_l\"),Transform=(Rotation=(X=0,Y=0.0565,Z=0,W=0.9984),T ...(truncated)"
    WeightMinimum=0
    Weight=@n8.Value
   )
   n9 = RigVMVariableNode() {variable=Thigh R Ot}
   > &RigUnit_ModifyTransforms(
    ItemToModify="((Item=(Type=Bone,Name=\"def_thigh_r_ot\"),Transform=(Rotation=(X=0,Y=0.1104,Z=0,W=-0.9939),Translation=(X=-0.8106,Y=0,Z=0.2031),Scale3D=(X=1,Y=1,Z=1))),(Item=(Type=Bone,Name=\"def_belt_side_r\"),Transform=(Rotation=(X=0,Y=0.5,Z=0,W=0.866),Translation=(X=20.13 ...(truncated)"
    WeightMinimum=0
    Weight=@n9.Value
   )
   > &RigUnit_ModifyTransforms(
    ItemToModify="((Item=(Type=Bone,Name=\"def_shirt_side_r\"),Transform=(Rotation=(X=0,Y=0.3503,Z=0,W=0.9366),Translation=(X=17.6646,Y=0,Z=0.7619),Scale3D=(X=0.8235,Y=1,Z=1))),(Item=(Type=Bone,Name=\"def_shirt_fr_r\"),Transform=(Rotation=(X=0,Y=0.201,Z=0,W=0.9796),Translation= ...(truncated)"
    WeightMinimum=0
    Weight=&RigVMFunction_MathDoubleClamp(Value=&RigVMFunction_MathDoubleRemap(SourceMaximum=1, SourceMinimum=0.3, TargetMaximum=1, TargetMinimum=0, bClamp=False, Value=@n9.Value).Result).Result
   )
   n10 = RigVMVariableNode() {variable=Thigh R Ot Post}
   > &RigUnit_ModifyTransforms(
    ItemToModify="((Item=(Type=Bone,Name=\"def_crease_r_ot\"),Transform=(Rotation=(X=0,Y=0.8768,Z=0,W=-0.4809),Translation=(X=-26.8506,Y=0,Z=4.7142),Scale3D=(X=1,Y=1,Z=1))),(Item=(Type=Bone,Name=\"def_crease_l_fr\"),Transform=(Rotation=(X=0,Y=-0.6364,Z=0,W=0.7713),Translation=( ...(truncated)"
    WeightMinimum=0
    Weight=@n10.Value
   )
   > &RigUnit_ModifyTransforms(
    ItemToModify="((Item=(Type=Bone,Name=\"def_shirt_side_r\"),Transform=(Rotation=(X=0,Y=0.661,Z=0,W=0.7504),Translation=(X=33.9109,Y=0.7353,Z=11.6097),Scale3D=(X=0.2911,Y=1,Z=1))),(Item=(Type=Bone,Name=\"def_shirt_fr_r\"),Transform=(Rotation=(X=0,Y=0.3362,Z=0,W=0.9418),Transl ...(truncated)"
    WeightMinimum=0
    Weight=@n10.Value
   )
   n11 = RigVMVariableNode() {variable=Thigh R Ba}
   > &RigUnit_ModifyTransforms(
    ItemToModify="((Item=(Type=Bone,Name=\"def_crease_r_ot\"),Transform=(Rotation=(X=0,Y=0,Z=0.6633,W=0.7483),Translation=(X=-19.8653,Y=21.3872,Z=0.001),Scale3D=(X=1,Y=1,Z=1))),(Item=(Type=Bone,Name=\"def_crease_r_ba\"),Transform=(Rotation=(X=0,Y=0,Z=0.6582,W=0.7528),Translatio ...(truncated)"
    WeightMinimum=0
    Weight=@n11.Value
   )
   > &RigUnit_ModifyTransforms(
    ItemToModify="((Item=(Type=Bone,Name=\"def_shirt_ba_r\"),Transform=(Rotation=(X=0,Y=0,Z=0.4352,W=0.9003),Translation=(X=19.1442,Y=-1.8878,Z=-0.0336),Scale3D=(X=0.55,Y=1,Z=1))),(Item=(Type=Bone,Name=\"def_shirt_ba_l\"),Transform=(Rotation=(X=0,Y=0.0838,Z=0,W=0.9965),Translat ...(truncated)"
    WeightMinimum=0
    Weight=&RigVMFunction_MathDoubleClamp(Value=&RigVMFunction_MathDoubleRemap(SourceMaximum=1, SourceMinimum=0.15, TargetMaximum=1, TargetMinimum=0, bClamp=False, Value=@n11.Value).Result).Result
   )
   n12 = RigVMVariableNode() {variable=Thigh R Crouch}
   > &RigUnit_ModifyTransforms(
    ItemToModify="((Item=(Type=Bone,Name=\"def_shirt_fr_r\"),Transform=(Rotation=(X=0.0007,Y=0.057,Z=-0.0341,W=0.9978),Translation=(X=-1.245,Y=2.5647,Z=-1.8382),Scale3D=(X=1,Y=1,Z=1))),(Item=(Type=Bone,Name=\"def_shirt_side_r\"),Transform=(Rotation=(X=-0.1164,Y=0.0755,Z=-0.1191 ...(truncated)"
    WeightMinimum=0
    Weight=&RigVMFunction_MathDoubleClamp(Value=&RigVMFunction_MathDoubleRemap(SourceMaximum=1, SourceMinimum=0.25, TargetMaximum=1, TargetMinimum=0, bClamp=False, Value=@n12.Value).Result).Result
   )
   > &RigUnit_ModifyTransforms(
    ItemToModify="((Item=(Type=Bone,Name=\"def_shirt_fr_r\"),Transform=(Rotation=(X=0,Y=0,Z=0,W=1),Translation=(X=-14.0756,Y=-4.9629,Z=-0.3786),Scale3D=(X=1,Y=1,Z=1))),(Item=(Type=Bone,Name=\"def_shirt_side_r\"),Transform=(Rotation=(X=0,Y=0,Z=0,W=1),Translation=(X=-30.5711,Y=10 ...(truncated)"
    WeightMinimum=0
    Weight=RigVMVariableNode() {variable=Thigh R Crouch Post}.Value
   )
   n13 = RigVMVariableNode() {variable=Thigh R Crouch Ot}
   > &RigUnit_ModifyTransforms(
    ItemToModify="((Item=(Type=Bone,Name=\"def_crease_r_ba\"),Transform=(Rotation=(X=0,Y=0,Z=0,W=1),Translation=(X=9.3843,Y=-1.1701,Z=0),Scale3D=(X=1,Y=1,Z=1))),(Item=(Type=Bone,Name=\"def_crease_r_fr\"),Transform=(Rotation=(X=0,Y=0,Z=0,W=-1),Translation=(X=-1.704,Y=-9.6742,Z=2 ...(truncated)"
    WeightMinimum=0
    Weight=@n13.Value
   )
   > &RigUnit_ModifyTransforms(
    ItemToModify="((Item=(Type=Bone,Name=\"def_shirt_fr_r\"),Transform=(Rotation=(X=0.0253,Y=0.1587,Z=-0.1007,W=0.9818),Translation=(X=10.3014,Y=2.8519,Z=4.9718),Scale3D=(X=1,Y=1,Z=1))),(Item=(Type=Bone,Name=\"def_shirt_side_r\"),Transform=(Rotation=(X=-0.0084,Y=0.1391,Z=-0.077 ...(truncated)"
    WeightMinimum=0
    Weight=&RigVMFunction_MathDoubleClamp(Value=&RigVMFunction_MathDoubleRemap(SourceMaximum=1, SourceMinimum=0.25, TargetMaximum=1, TargetMinimum=0, bClamp=False, Value=@n13.Value).Result).Result
   )
   > &RigUnit_ModifyTransforms(
    ItemToModify="((Item=(Type=Bone,Name=\"def_shirt_fr_r\"),Transform=(Rotation=(X=0.1303,Y=0.2981,Z=-0.3855,W=0.8634),Translation=(X=-10.3621,Y=-11.1898,Z=11.0125),Scale3D=(X=1,Y=1,Z=1))),(Item=(Type=Bone,Name=\"def_shirt_side_r\"),Transform=(Rotation=(X=-0.1815,Y=-0.3851,Z=- ...(truncated)"
    WeightMinimum=0
    Weight=RigVMVariableNode() {variable=Thigh R Crouch Post Ot}.Value
   )
   &RigUnit_ModifyTransforms(
    ItemToModify=((Item=(Type=Bone,Name="def_crease_l_fr"),Transform=(Rotation=(X=0,Y=0,Z=0,W=1),Translation=(X=-6.2735,Y=4.2651,Z=0),Scale3D=(X=1,Y=1,Z=1))))
    WeightMinimum=0
    Weight=RigVMVariableNode() {variable=Thigh L Pre Fr}.Value
    ExecutePin=@RigVMFunction_Sequence.C
   )
   n14 = RigVMVariableNode() {variable=Thigh L Fr}
   > &RigUnit_ModifyTransforms(
    ItemToModify="((Item=(Type=Bone,Name=\"def_crease_l_ba\"),Transform=(Rotation=(X=0,Y=0,Z=-0.5977,W=0.8017),Translation=(X=3.979,Y=4.3677,Z=0),Scale3D=(X=1,Y=1,Z=1))),(Item=(Type=Bone,Name=\"def_crease_l_fr\"),Transform=(Rotation=(X=0,Y=0,Z=-0.4339,W=0.901),Translation=(X=45 ...(truncated)"
    WeightMinimum=0
    Weight=@n14.Value
   )
   Clamp = &RigVMFunction_MathDoubleClamp(Value=&RigVMFunction_MathDoubleRemap(SourceMaximum=1, SourceMinimum=0.45, TargetMaximum=1, TargetMinimum=0, bClamp=False, Value=@n14.Value).Result)
   > &RigUnit_ModifyTransforms(
    ItemToModify="((Item=(Type=Bone,Name=\"def_shirt_fr_l\"),Transform=(Rotation=(X=-0.0309,Y=-0.0808,Z=-0.356,W=0.9305),Translation=(X=18.277,Y=5.1756,Z=0.9412),Scale3D=(X=0.4978,Y=1,Z=1))),(Item=(Type=Bone,Name=\"def_shirt_fr_r\"),Transform=(Rotation=(X=0,Y=-0.0565,Z=0,W=0.99 ...(truncated)"
    WeightMinimum=0
    Weight=@Clamp.Result
   )
   n15 = RigVMVariableNode() {variable=Thigh L Post Fr}
   > &RigUnit_ModifyTransforms(
    ItemToModify="((Item=(Type=Bone,Name=\"def_thigh_l_fr\"),Transform=(Rotation=(X=0,Y=0,Z=0,W=1),Translation=(X=-11.1302,Y=1.5498,Z=-1.4162),Scale3D=(X=1,Y=1,Z=1))),(Item=(Type=Bone,Name=\"def_crease_l_fr\"),Transform=(Rotation=(X=0,Y=0,Z=-0.3827,W=0.9239),Translation=(X=84.0 ...(truncated)"
    WeightMinimum=0
    Weight=@n15.Value
   )
   > &RigUnit_ModifyTransforms(
    ItemToModify="((Item=(Type=Bone,Name=\"def_shirt_fr_l\"),Transform=(Rotation=(X=0,Y=0,Z=-0.5879,W=0.8089),Translation=(X=54.2721,Y=21.6031,Z=11.3189),Scale3D=(X=0.4978,Y=1,Z=1.2412))),(Item=(Type=Bone,Name=\"def_shirt_fr_r\"),Transform=(Rotation=(X=0,Y=-0.0565,Z=0,W=0.9984) ...(truncated)"
    WeightMinimum=0
    Weight=@n15.Value
   )
   n16 = RigVMVariableNode() {variable=Thigh L Ot}
   > &RigUnit_ModifyTransforms(
    ItemToModify="((Item=(Type=Bone,Name=\"def_crease_l_in\"),Transform=(Rotation=(X=0,Y=-0.5759,Z=0,W=0.8175),Translation=(X=3.3104,Y=0,Z=-1.5937),Scale3D=(X=1,Y=1,Z=1))),(Item=(Type=Bone,Name=\"def_crease_l_ot\"),Transform=(Rotation=(X=0,Y=-0.543,Z=0,W=0.8397),Translation=(X= ...(truncated)"
    WeightMinimum=0
    Weight=@n16.Value
   )
   > &RigUnit_ModifyTransforms(
    ItemToModify="((Item=(Type=Bone,Name=\"def_shirt_side_l\"),Transform=(Rotation=(X=0,Y=-0.3503,Z=0,W=0.9366),Translation=(X=18.5073,Y=0.0001,Z=-1.4731),Scale3D=(X=0.8235,Y=1,Z=1))),(Item=(Type=Bone,Name=\"def_shirt_fr_l\"),Transform=(Rotation=(X=0,Y=-0.201,Z=0,W=0.9796),Tran ...(truncated)"
    WeightMinimum=0
    Weight=&RigVMFunction_MathDoubleClamp(Value=&RigVMFunction_MathDoubleRemap(SourceMaximum=1, SourceMinimum=0.3, TargetMaximum=1, TargetMinimum=0, bClamp=False, Value=@n16.Value).Result).Result
   )
   n17 = RigVMVariableNode() {variable=Thigh L Ot Post}
   > &RigUnit_ModifyTransforms(
    ItemToModify="((Item=(Type=Bone,Name=\"def_crease_l_ot\"),Transform=(Rotation=(X=0,Y=0.8768,Z=0,W=-0.4809),Translation=(X=26.8506,Y=0,Z=-4.7142),Scale3D=(X=1,Y=1,Z=1))),(Item=(Type=Bone,Name=\"def_crease_l_fr\"),Transform=(Rotation=(X=0,Y=-0.6364,Z=0,W=0.7713),Translation=( ...(truncated)"
    WeightMinimum=0
    Weight=@n17.Value
   )
   > &RigUnit_ModifyTransforms(
    ItemToModify="((Item=(Type=Bone,Name=\"def_shirt_side_l\"),Transform=(Rotation=(X=0,Y=-0.4069,Z=0,W=0.9135),Translation=(X=33.6783,Y=0,Z=-12.697),Scale3D=(X=0.2911,Y=1,Z=1))),(Item=(Type=Bone,Name=\"def_shirt_fr_l\"),Transform=(Rotation=(X=0,Y=-0.3362,Z=0,W=0.9418),Translat ...(truncated)"
    WeightMinimum=0
    Weight=@n17.Value
   )
   n18 = RigVMVariableNode() {variable=Thigh L Ba}
   > &RigUnit_ModifyTransforms(
    ItemToModify="((Item=(Type=Bone,Name=\"def_crease_l_ot\"),Transform=(Rotation=(X=0,Y=0,Z=0.5357,W=0.8444),Translation=(X=17.5849,Y=-22.4288,Z=0),Scale3D=(X=1,Y=1,Z=1))),(Item=(Type=Bone,Name=\"def_crease_l_ba\"),Transform=(Rotation=(X=0,Y=0,Z=-0.0534,W=0.9986),Translation=( ...(truncated)"
    WeightMinimum=0
    Weight=@n18.Value
   )
   > &RigUnit_ModifyTransforms(
    ItemToModify="((Item=(Type=Bone,Name=\"def_shirt_ba_l\"),Transform=(Rotation=(X=0,Y=0,Z=0.1955,W=0.9807),Translation=(X=17.1664,Y=-8.0085,Z=2.0682),Scale3D=(X=0.55,Y=1,Z=1))),(Item=(Type=Bone,Name=\"def_shirt_ba_r\"),Transform=(Rotation=(X=0,Y=-0.0838,Z=0,W=0.9965),Translat ...(truncated)"
    WeightMinimum=0
    Weight=&RigVMFunction_MathDoubleClamp(Value=&RigVMFunction_MathDoubleRemap(SourceMaximum=1, SourceMinimum=0.15, TargetMaximum=1, TargetMinimum=0, bClamp=False, Value=@n18.Value).Result).Result
   )
   n19 = RigVMVariableNode() {variable=Thigh L Crouch}
   > &RigUnit_ModifyTransforms(
    ItemToModify="((Item=(Type=Bone,Name=\"def_crease_l_fr\"),Transform=(Rotation=(X=0,Y=0,Z=0,W=1),Translation=(X=3.8828,Y=-4.1558,Z=1.8747),Scale3D=(X=1,Y=1,Z=1))),(Item=(Type=Bone,Name=\"def_crease_l_ba\"),Transform=(Rotation=(X=0,Y=0,Z=0,W=1),Translation=(X=-4.3817,Y=0,Z=0) ...(truncated)"
    WeightMinimum=0
    Weight=@n19.Value
   )
   > &RigUnit_ModifyTransforms(
    ItemToModify="((Item=(Type=Bone,Name=\"def_shirt_fr_l\"),Transform=(Rotation=(X=0.1102,Y=-0.0243,Z=0.0437,W=0.9926),Translation=(X=2.5,Y=1.0231,Z=0.5129),Scale3D=(X=1,Y=1,Z=1))),(Item=(Type=Bone,Name=\"def_shirt_side_l\"),Transform=(Rotation=(X=0.0261,Y=-0.0941,Z=-0.0636,W= ...(truncated)"
    WeightMinimum=0
    Weight=&RigVMFunction_MathDoubleClamp(Value=&RigVMFunction_MathDoubleRemap(SourceMaximum=1, SourceMinimum=0.25, TargetMaximum=1, TargetMinimum=0, bClamp=False, Value=@n19.Value).Result).Result
   )
   > &RigUnit_ModifyTransforms(
    ItemToModify="((Item=(Type=Bone,Name=\"def_shirt_fr_l\"),Transform=(Rotation=(X=0,Y=0,Z=0,W=1),Translation=(X=7.3953,Y=1.9651,Z=-1.1314),Scale3D=(X=1,Y=1,Z=1))),(Item=(Type=Bone,Name=\"def_shirt_side_l\"),Transform=(Rotation=(X=0,Y=0,Z=0,W=1),Translation=(X=-16.4467,Y=16.34 ...(truncated)"
    WeightMinimum=0
    Weight=RigVMVariableNode() {variable=Thigh L Crouch Post}.Value
   )
   n20 = RigVMVariableNode() {variable=Thigh L Crouch Ot}
   > &RigUnit_ModifyTransforms(
    ItemToModify="((Item=(Type=Bone,Name=\"def_crease_l_fr\"),Transform=(Rotation=(X=0,Y=0,Z=0,W=1),Translation=(X=3.8191,Y=-2.6177,Z=-0.2287),Scale3D=(X=1,Y=1,Z=1))),(Item=(Type=Bone,Name=\"def_crease_l_in\"),Transform=(Rotation=(X=0.0246,Y=-0.1867,Z=-0.1282,W=-0.9737),Transla ...(truncated)"
    WeightMinimum=0
    Weight=@n20.Value
   )
   > &RigUnit_ModifyTransforms(
    ItemToModify="((Item=(Type=Bone,Name=\"def_shirt_fr_l\"),Transform=(Rotation=(X=0.0946,Y=-0.2481,Z=-0.044,W=0.9631),Translation=(X=7.8447,Y=1.4522,Z=-2.6292),Scale3D=(X=1,Y=1,Z=1))),(Item=(Type=Bone,Name=\"def_shirt_side_l\"),Transform=(Rotation=(X=-0.0495,Y=-0.1892,Z=-0.06 ...(truncated)"
    WeightMinimum=0
    Weight=&RigVMFunction_MathDoubleClamp(Value=&RigVMFunction_MathDoubleRemap(SourceMaximum=1, SourceMinimum=0.25, TargetMaximum=1, TargetMinimum=0, bClamp=False, Value=@n20.Value).Result).Result
   )
   > &RigUnit_ModifyTransforms(
    ItemToModify="((Item=(Type=Bone,Name=\"def_shirt_fr_l\"),Transform=(Rotation=(X=0.0564,Y=-0.1747,Z=-0.1304,W=0.9743),Translation=(X=-0.6119,Y=-8.4619,Z=-6.5323),Scale3D=(X=1,Y=1,Z=1))),(Item=(Type=Bone,Name=\"def_shirt_side_l\"),Transform=(Rotation=(X=-0.0084,Y=-0.078,Z=-0. ...(truncated)"
    WeightMinimum=0
    Weight=RigVMVariableNode() {variable=Thigh L Crouch Ot Post}.Value
   )
   RigUnit_GetTransform = &RigUnit_GetTransform(Item=(Type=Bone,Name="def_clavicle_l_tp"), bInitial=False)
   n21 = &RigUnit_GetTransform(Item=(Type=Bone,Name="def_upperarm_l_tp"), bInitial=False)
   RigUnit_SetTransform = &RigUnit_SetTransform(
    Item=(Type=Bone,Name="def_upperarm_l_tp")
    Value=(Rotation=(X=-0.0185,Y=-0.143,Z=-0.0057,W=0.9895),Translation=(X=12.5125,Y=-4.9433,Z=132.2893),Scale3D=(X=1,Y=1,Z=1))
    Weight=1
    bInitial=False
    Value.Translation=@RigUnit_GetTransform.Transform.Translation
    Value.Rotation=@n21.Transform.Rotation
    Value.Scale3D=@n21.Transform.Scale3D
    ExecutePin=@RigVMFunction_Sequence.L
   )
   n22 = &RigUnit_GetTransform(Item=(Type=Bone,Name="def_upperarm_l_fr"), bInitial=False)
   > &RigUnit_SetTransform(
    Item=(Type=Bone,Name="def_upperarm_l_fr")
    Value=(Rotation=(X=-0.0185,Y=-0.143,Z=-0.0057,W=0.9895),Translation=(X=12.5125,Y=-4.9433,Z=132.2893),Scale3D=(X=1,Y=1,Z=1))
    Weight=1
    bInitial=False
    Value.Translation=&RigUnit_GetTransform(Item=(Type=Bone,Name="def_clavicle_l_fr"), bInitial=False).Transform.Translation
    Value.Rotation=@n22.Transform.Rotation
    Value.Scale3D=@n22.Transform.Scale3D
   )
   n23 = &RigUnit_GetTransform(Item=(Type=Bone,Name="def_upperarm_l_bt"), bInitial=False)
   > &RigUnit_SetTransform(
    Item=(Type=Bone,Name="def_upperarm_l_bt")
    Value=(Rotation=(X=-0.0185,Y=-0.143,Z=-0.0057,W=0.9895),Translation=(X=12.5125,Y=-4.9433,Z=132.2893),Scale3D=(X=1,Y=1,Z=1))
    Weight=1
    bInitial=False
    Value.Translation=&RigUnit_GetTransform(Item=(Type=Bone,Name="def_clavicle_l_bt"), bInitial=False).Transform.Translation
    Value.Rotation=@n23.Transform.Rotation
    Value.Scale3D=@n23.Transform.Scale3D
   )
   n24 = &RigUnit_GetTransform(Item=(Type=Bone,Name="def_upperarm_l_ba"), bInitial=False)
   > &RigUnit_SetTransform(
    Item=(Type=Bone,Name="def_upperarm_l_ba")
    Value=(Rotation=(X=-0.0185,Y=-0.143,Z=-0.0057,W=0.9895),Translation=(X=12.5125,Y=-4.9433,Z=132.2893),Scale3D=(X=1,Y=1,Z=1))
    Weight=1
    bInitial=False
    Value.Translation=&RigUnit_GetTransform(Item=(Type=Bone,Name="def_clavicle_l_ba"), bInitial=False).Transform.Translation
    Value.Rotation=@n24.Transform.Rotation
    Value.Scale3D=@n24.Transform.Scale3D
   )
   &RigUnit_SetTransform(
    Item=(Type=Bone,Name="def_elbow_l")
    Space=LocalSpace
    Weight=1
    bInitial=False
    Value=/Script/RigVM.RigVMFunction_MathTransformLerp(
     T=0.5
     A=&RigUnit_GetTransform(Item=(Type=Bone,Name="lowerarm_l"), Space=LocalSpace, bInitial=true).Transform
     B=&RigUnit_GetTransform(Item=(Type=Bone,Name="lowerarm_l"), Space=LocalSpace, bInitial=False).Transform
    ) {is=struct RigVMUnitNode, template="Interpolate::Execute(in A,in B,in T,out Result)"}.Result
    ExecutePin=@RigVMFunction_Sequence.K
   )
   > Elbow Readers = RigVMFunctionReferenceNode(driver item=(Type=Bone,Name="lowerarm_l"))
   > &RigUnit_ModifyTransforms(
    ItemToModify="((Item=(Type=Bone,Name=\"def_elbow_l_fr\"),Transform=(Rotation=(X=0,Y=0,Z=0,W=1),Translation=(X=2.6553,Y=6.01,Z=0),Scale3D=(X=1,Y=1,Z=1))),(Item=(Type=Bone,Name=\"def_elbow_l_fr_bt\"),Transform=(Rotation=(X=0,Y=0,Z=-0.4798,W=-0.8774),Translation=(X=0,Y=0,Z=0), ...(truncated)"
    WeightMinimum=0
    Weight=@"Elbow Readers".Bend
   )
   > &RigUnit_ModifyTransforms(
    ItemToModify="((Item=(Type=Bone,Name=\"def_elbow_l_fr\"),Transform=(Rotation=(X=0,Y=0,Z=0,W=1),Translation=(X=2.6553,Y=7.5185,Z=0),Scale3D=(X=1,Y=1,Z=1))),(Item=(Type=Bone,Name=\"def_elbow_l_fr_bt\"),Transform=(Rotation=(X=0,Y=0,Z=-0.8561,W=-0.5168),Translation=(X=0,Y=0,Z=0 ...(truncated)"
    WeightMinimum=0
    Weight=@"Elbow Readers".Bend Post
   )
   > &RigUnit_SetTransform(
    Item=(Type=Bone,Name="def_elbow_r")
    Space=LocalSpace
    Weight=1
    bInitial=False
    Value=&RigVMFunction_MathTransformLerp(
     T=0.5
     A=&RigUnit_GetTransform(Item=(Type=Bone,Name="lowerarm_r"), Space=LocalSpace, bInitial=true).Transform
     B=&RigUnit_GetTransform(Item=(Type=Bone,Name="lowerarm_r"), Space=LocalSpace, bInitial=False).Transform
    ) {template="Interpolate::Execute(in A,in B,in T,out Result)"}.Result
   )
   > Elbow Readers_1 = RigVMFunctionReferenceNode(driver item=(Type=Bone,Name="lowerarm_r"))
   > &RigUnit_ModifyTransforms(
    ItemToModify="((Item=(Type=Bone,Name=\"def_elbow_r_fr\"),Transform=(Rotation=(X=0,Y=0,Z=0,W=1),Translation=(X=-3.0095,Y=-4.7483,Z=0),Scale3D=(X=1,Y=1,Z=1))),(Item=(Type=Bone,Name=\"def_elbow_r_fr_bt\"),Transform=(Rotation=(X=0,Y=0,Z=0.4798,W=0.8774),Translation=(X=0,Y=0,Z=0 ...(truncated)"
    WeightMinimum=0
    Weight=@"Elbow Readers_1".Bend
   )
   > &RigUnit_ModifyTransforms(
    ItemToModify="((Item=(Type=Bone,Name=\"def_elbow_r_fr\"),Transform=(Rotation=(X=0,Y=0,Z=0,W=1),Translation=(X=-2.6553,Y=-7.5185,Z=0),Scale3D=(X=1,Y=1,Z=1))),(Item=(Type=Bone,Name=\"def_elbow_r_fr_bt\"),Transform=(Rotation=(X=0,Y=0,Z=0.8561,W=0.5168),Translation=(X=0,Y=0,Z=0 ...(truncated)"
    WeightMinimum=0
    Weight=@"Elbow Readers_1".Bend Post
   )
   &RigUnit_SetTransform(
    Item=(Type=Bone,Name="def_knee_l")
    Space=LocalSpace
    Weight=1
    bInitial=False
    Value=&RigVMFunction_MathTransformLerp(
     T=0.5
     A=&RigUnit_GetTransform(Item=(Type=Bone,Name="calf_l"), Space=LocalSpace, bInitial=true).Transform
     B=&RigUnit_GetTransform(Item=(Type=Bone,Name="calf_l"), Space=LocalSpace, bInitial=False).Transform
    ) {template="Interpolate::Execute(in A,in B,in T,out Result)"}.Result
    ExecutePin=@RigVMFunction_Sequence.E
   )
   > Elbow Readers_2 = RigVMFunctionReferenceNode(driver item=(Type=Bone,Name="calf_l"))
   > &RigUnit_ModifyTransforms(
    ItemToModify="((Item=(Type=Bone,Name=\"def_knee_l\"),Transform=(Rotation=(X=0,Y=0,Z=0,W=-1),Translation=(X=-9.9708,Y=0.3851,Z=0),Scale3D=(X=1,Y=1,Z=1))),(Item=(Type=Bone,Name=\"def_knee_l_ba\"),Transform=(Rotation=(X=0,Y=0,Z=0,W=-1),Translation=(X=0.9817,Y=-5.1584,Z=0),Scal ...(truncated)"
    WeightMinimum=0
    Weight=@"Elbow Readers_2".Bend
   )
   > &RigUnit_ModifyTransforms(
    ItemToModify="((Item=(Type=Bone,Name=\"def_knee_l\"),Transform=(Rotation=(X=0,Y=0,Z=0,W=1),Translation=(X=-9.9708,Y=-2.5171,Z=0),Scale3D=(X=1,Y=1,Z=1))),(Item=(Type=Bone,Name=\"def_knee_l_ba\"),Transform=(Rotation=(X=0,Y=0,Z=0,W=1),Translation=(X=0,Y=-5.1042,Z=0),Scale3D=(X ...(truncated)"
    WeightMinimum=0
    Weight=@"Elbow Readers_2".Bend Post
   )
   > &RigUnit_ModifyTransforms(
    ItemToModify="((Item=(Type=Bone,Name=\"def_thigh_l_push\"),Transform=(Rotation=(X=0,Y=0,Z=0,W=1),Translation=(X=0,Y=3.1575,Z=0),Scale3D=(X=1,Y=1,Z=1))),(Item=(Type=Bone,Name=\"def_knee_l\"),Transform=(Rotation=(X=0,Y=0,Z=0.0532,W=0.9986),Translation=(X=2.7987,Y=6.1619,Z=0), ...(truncated)"
    WeightMinimum=0
    Weight=&RigVMFunction_MathDoubleClamp(Value=&RigVMFunction_MathDoubleRemap(SourceMaximum=1, SourceMinimum=0.2, TargetMaximum=1, TargetMinimum=0, bClamp=False, Value=@"Elbow Readers_2".Bend Post).Result).Result
   )
   > &RigUnit_SetTransform(
    Item=(Type=Bone,Name="def_knee_r")
    Space=LocalSpace
    Weight=1
    bInitial=False
    Value=&RigVMFunction_MathTransformLerp(
     T=0.5
     A=&RigUnit_GetTransform(Item=(Type=Bone,Name="calf_r"), Space=LocalSpace, bInitial=true).Transform
     B=&RigUnit_GetTransform(Item=(Type=Bone,Name="calf_r"), Space=LocalSpace, bInitial=False).Transform
    ) {template="Interpolate::Execute(in A,in B,in T,out Result)"}.Result
   )
   > Elbow Readers_3 = RigVMFunctionReferenceNode(driver item=(Type=Bone,Name="calf_r"))
   > &RigUnit_ModifyTransforms(
    ItemToModify="((Item=(Type=Bone,Name=\"def_knee_r\"),Transform=(Rotation=(X=0,Y=0,Z=0,W=-1),Translation=(X=11.1101,Y=-1.6978,Z=0),Scale3D=(X=1,Y=1,Z=1))),(Item=(Type=Bone,Name=\"def_knee_r_ba\"),Transform=(Rotation=(X=0,Y=0,Z=-0.0648,W=-0.9979),Translation=(X=-4.4642,Y=2.95 ...(truncated)"
    WeightMinimum=0
    Weight=@"Elbow Readers_3".Bend
   )
   > &RigUnit_ModifyTransforms(
    ItemToModify="((Item=(Type=Bone,Name=\"def_knee_r\"),Transform=(Rotation=(X=0,Y=0,Z=-0.1334,W=-0.9911),Translation=(X=5.9475,Y=-3.413,Z=0),Scale3D=(X=1,Y=1,Z=1))),(Item=(Type=Bone,Name=\"def_knee_r_ba\"),Transform=(Rotation=(X=0,Y=0,Z=0,W=-1),Translation=(X=0,Y=10.6904,Z=0) ...(truncated)"
    WeightMinimum=0
    Weight=@"Elbow Readers_3".Bend Post
   )
   > &RigUnit_ModifyTransforms(
    ItemToModify=((Item=(Type=Bone,Name="def_thigh_r_push"),Transform=(Rotation=(X=0,Y=0,Z=0,W=-1),Translation=(X=0,Y=-11.2535,Z=0),Scale3D=(X=1,Y=1,Z=1))))
    WeightMinimum=0
    Weight=&RigVMFunction_MathDoubleClamp(Value=&RigVMFunction_MathDoubleRemap(SourceMaximum=1, SourceMinimum=0.2, TargetMaximum=1, TargetMinimum=0, bClamp=False, Value=@"Elbow Readers_3".Bend Post).Result).Result
   )
   Thigh Readers_6 = RigVMFunctionReferenceNode(
    Crouch Ot Post Viz=false
    Crouch Ot Pre Viz=false
    Crouch Ot Viz=false
    Crouch Post Viz=false
    Crouch Pre Viz=false
    Crouch Viz=false
    Driver Item=(Type=Bone,Name="thigh_l")
    Fr Post Viz=false
    Fr Pre Viz=false
    Fr Viz=false
    Ot Viz=false
    ExecuteContext=@RigVMFunction_Sequence.B
   )
   > RigVMVariableNode(Value=@"Thigh Readers_6".Fr) {variable=Thigh L Fr}
   > RigVMVariableNode(Value=@"Thigh Readers_6".Fr Pre) {variable=Thigh L Pre Fr}
   > RigVMVariableNode(Value=@"Thigh Readers_6".Fr Post) {variable=Thigh L Post Fr}
   > RigVMVariableNode(Value=@"Thigh Readers_6".Ba) {variable=Thigh L Ba}
   > RigVMVariableNode(Value=@"Thigh Readers_6".Ba Pre) {variable=Thigh L Ba Pre}
   > RigVMVariableNode(Value=@"Thigh Readers_6".Ot) {variable=Thigh L Ot}
   > RigVMVariableNode(Value=@"Thigh Readers_6".Ot Pre) {variable=Thigh L Ot Pre}
   > RigVMVariableNode(Value=@"Thigh Readers_6".Ot Post) {variable=Thigh L Ot Post}
   > RigVMVariableNode(Value=@"Thigh Readers_6".Crouch) {variable=Thigh L Crouch}
   > RigVMVariableNode(Value=@"Thigh Readers_6".Crouch Pre) {variable=Thigh L Crouch Pre}
   > RigVMVariableNode(Value=@"Thigh Readers_6".Crouch Post) {variable=Thigh L Crouch Post}
   > RigVMVariableNode(Value=@"Thigh Readers_6".Crouch Ot) {variable=Thigh L Crouch Ot}
   > RigVMVariableNode(Value=@"Thigh Readers_6".Crouch Pre Ot) {variable=Thigh L Crouch Ot Pre}
   > RigVMVariableNode(Value=@"Thigh Readers_6".Crouch Post Ot) {variable=Thigh L Crouch Ot Post}
   > Thigh Readers_1 = RigVMFunctionReferenceNode(
    Crouch Ot Post Viz=false
    Crouch Ot Pre Viz=false
    Crouch Ot Viz=false
    Crouch Post Viz=false
    Crouch Pre Viz=false
    Crouch Viz=false
    Driver Item=(Type=Bone,Name="thigh_r")
    Fr Post Viz=false
    Fr Pre Viz=false
    Fr Viz=false
    Ot Viz=false
   )
   > RigVMVariableNode(Value=@"Thigh Readers_1".Fr) {variable=Thigh R Fr}
   > RigVMVariableNode(Value=@"Thigh Readers_1".Fr Pre) {variable=Thigh R Fr Pre}
   > RigVMVariableNode(Value=@"Thigh Readers_1".Fr Post) {variable=Thigh R Fr Post}
   > RigVMVariableNode(Value=@"Thigh Readers_1".Ba) {variable=Thigh R Ba}
   > RigVMVariableNode(Value=@"Thigh Readers_1".Ba Pre) {variable=Thigh R Ba Pre}
   > RigVMVariableNode(Value=@"Thigh Readers_1".Ot) {variable=Thigh R Ot}
   > RigVMVariableNode(Value=@"Thigh Readers_1".Ot Pre) {variable=Thigh R Ot Pre}
   > RigVMVariableNode(Value=@"Thigh Readers_1".Ot Post) {variable=Thigh R Ot Post}
   > RigVMVariableNode(Value=@"Thigh Readers_1".Crouch) {variable=Thigh R Crouch}
   > RigVMVariableNode(Value=@"Thigh Readers_1".Crouch Pre) {variable=Thigh R Crouch Pre}
   > RigVMVariableNode(Value=@"Thigh Readers_1".Crouch Post) {variable=Thigh R Crouch Post}
   > RigVMVariableNode(Value=@"Thigh Readers_1".Crouch Ot) {variable=Thigh R Crouch Ot}
   > RigVMVariableNode(Value=@"Thigh Readers_1".Crouch Pre Ot) {variable=Thigh R Crouch Pre Ot}
   > RigVMVariableNode(Value=@"Thigh Readers_1".Crouch Post Ot) {variable=Thigh R Crouch Post Ot}
   n25 = &RigUnit_GetTransform(Item=(Type=Bone,Name="def_upperarm_r_tp"), bInitial=False)
   &RigUnit_SetTransform(
    Item=(Type=Bone,Name="def_upperarm_r_tp")
    Value=(Rotation=(X=-0.0185,Y=-0.143,Z=-0.0057,W=0.9895),Translation=(X=12.5125,Y=-4.9433,Z=132.2893),Scale3D=(X=1,Y=1,Z=1))
    Weight=1
    bInitial=False
    Value.Translation=&RigUnit_GetTransform(Item=(Type=Bone,Name="def_clavicle_r_tp"), bInitial=False).Transform.Translation
    Value.Rotation=@n25.Transform.Rotation
    Value.Scale3D=@n25.Transform.Scale3D
    ExecutePin=@RigVMFunction_Sequence.M
   )
   n26 = &RigUnit_GetTransform(Item=(Type=Bone,Name="def_upperarm_r_fr"), bInitial=False)
   > &RigUnit_SetTransform(
    Item=(Type=Bone,Name="def_upperarm_r_fr")
    Value=(Rotation=(X=-0.0185,Y=-0.143,Z=-0.0057,W=0.9895),Translation=(X=12.5125,Y=-4.9433,Z=132.2893),Scale3D=(X=1,Y=1,Z=1))
    Weight=1
    bInitial=False
    Value.Translation=&RigUnit_GetTransform(Item=(Type=Bone,Name="def_clavicle_r_fr"), bInitial=False).Transform.Translation
    Value.Rotation=@n26.Transform.Rotation
    Value.Scale3D=@n26.Transform.Scale3D
   )
   n27 = &RigUnit_GetTransform(Item=(Type=Bone,Name="def_upperarm_r_bt"), bInitial=False)
   > &RigUnit_SetTransform(
    Item=(Type=Bone,Name="def_upperarm_r_bt")
    Value=(Rotation=(X=-0.0185,Y=-0.143,Z=-0.0057,W=0.9895),Translation=(X=12.5125,Y=-4.9433,Z=132.2893),Scale3D=(X=1,Y=1,Z=1))
    Weight=1
    bInitial=False
    Value.Translation=&RigUnit_GetTransform(Item=(Type=Bone,Name="def_clavicle_r_bt"), bInitial=False).Transform.Translation
    Value.Rotation=@n27.Transform.Rotation
    Value.Scale3D=@n27.Transform.Scale3D
   )
   n28 = &RigUnit_GetTransform(Item=(Type=Bone,Name="def_upperarm_r_ba"), bInitial=False)
   > &RigUnit_SetTransform(
    Item=(Type=Bone,Name="def_upperarm_r_ba")
    Value=(Rotation=(X=-0.0185,Y=-0.143,Z=-0.0057,W=0.9895),Translation=(X=12.5125,Y=-4.9433,Z=132.2893),Scale3D=(X=1,Y=1,Z=1))
    Weight=1
    bInitial=False
    Value.Translation=&RigUnit_GetTransform(Item=(Type=Bone,Name="def_clavicle_r_ba"), bInitial=False).Transform.Translation
    Value.Rotation=@n28.Transform.Rotation
    Value.Scale3D=@n28.Transform.Scale3D
   )
   UpperArm Readers_1 = RigVMFunctionReferenceNode(Dn Viz=false, Driver Item=(Type=Bone,Name="upperarm_l"), Up Viz=false, ExecuteContext=@RigVMFunction_Sequence.N)
   > &RigUnit_ModifyTransforms(
    ItemToModify="((Item=(Type=Bone,Name=\"def_clavicle_l_tp\"),Transform=(Rotation=(X=0,Y=0,Z=0,W=1),Translation=(X=-0.4259,Y=0,Z=2.0517),Scale3D=(X=1,Y=1,Z=1))),(Item=(Type=Bone,Name=\"def_upperarm_l_tp\"),Transform=(Rotation=(X=0,Y=0.3882,Z=0,W=0.9216),Translation=(X=9.1796, ...(truncated)"
    WeightMinimum=0
    Weight=@"UpperArm Readers_1".Up
   )
   > &RigUnit_ModifyTransforms(
    ItemToModify="((Item=(Type=Bone,Name=\"upperarm_twist_02_l\"),Transform=(Rotation=(X=0,Y=0,Z=0,W=-1),Translation=(X=6.2882,Y=0,Z=7.6228),Scale3D=(X=1,Y=1,Z=1))),(Item=(Type=Bone,Name=\"def_sleeve_l_tp\"),Transform=(Rotation=(X=0,Y=0,Z=0,W=1),Translation=(X=1.6775,Y=0,Z=-6.6 ...(truncated)"
    WeightMinimum=0
    Weight=@"UpperArm Readers_1".Dn
   )
   > &RigUnit_ModifyTransforms(
    ItemToModify="((Item=(Type=Bone,Name=\"upperarm_twist_02_l\"),Transform=(Rotation=(X=0,Y=0,Z=0,W=-1),Translation=(X=6.4543,Y=0.7084,Z=0),Scale3D=(X=1,Y=1,Z=1))),(Item=(Type=Bone,Name=\"def_sleeve_l_fr\"),Transform=(Rotation=(X=0,Y=0,Z=0.0876,W=0.9962),Translation=(X=0.5915, ...(truncated)"
    WeightMinimum=0
    Weight=@"UpperArm Readers_1".Fr
   )
   > UpperArm Readers_2 = RigVMFunctionReferenceNode(Dn Viz=false, Driver Item=(Type=Bone,Name="upperarm_r"), Up Viz=false)
   > &RigUnit_ModifyTransforms(
    ItemToModify="((Item=(Type=Bone,Name=\"def_clavicle_r_tp\"),Transform=(Rotation=(X=0,Y=0,Z=0,W=1),Translation=(X=1.6533,Y=0,Z=-3.7384),Scale3D=(X=1,Y=1,Z=1))),(Item=(Type=Bone,Name=\"def_upperarm_r_tp\"),Transform=(Rotation=(X=0,Y=0.3894,Z=0,W=0.9211),Translation=(X=-6.8521 ...(truncated)"
    WeightMinimum=0
    Weight=@"UpperArm Readers_2".Up
   )
   > &RigUnit_ModifyTransforms(
    ItemToModify="((Item=(Type=Bone,Name=\"upperarm_twist_02_r\"),Transform=(Rotation=(X=0,Y=0.0851,Z=0,W=0.9964),Translation=(X=-7.544,Y=0,Z=0.4257),Scale3D=(X=1,Y=1,Z=1))),(Item=(Type=Bone,Name=\"def_sleeve_r_bt\"),Transform=(Rotation=(X=0,Y=0,Z=0,W=-1),Translation=(X=0,Y=0,Z ...(truncated)"
    WeightMinimum=0
    Weight=@"UpperArm Readers_2".Dn
   )
   > &RigUnit_ModifyTransforms(
    ItemToModify="((Item=(Type=Bone,Name=\"upperarm_twist_02_r\"),Transform=(Rotation=(X=0,Y=0,Z=0,W=-1),Translation=(X=-3.045,Y=0,Z=0),Scale3D=(X=1,Y=1,Z=1))),(Item=(Type=Bone,Name=\"def_sleeve_r_fr\"),Transform=(Rotation=(X=0,Y=0,Z=0,W=1),Translation=(X=-0.7354,Y=2.914,Z=0),S ...(truncated)"
    WeightMinimum=0
    Weight=@"UpperArm Readers_2".Fr
   )
   ArmElbowAdjustment = RigVMFunctionReferenceNode(ExecuteContext=())
   &RigUnit_ModifyTransforms(
    ExecutePin=()
    ItemToModify=((Item=(Type=Bone,Name="None"),Transform=(Translation=(X=0,Y=0,Z=0),Scale3D=(X=1,Y=1,Z=1))))
    WeightMinimum=0
    Weight=RigVMVariableNode() {variable=Thigh R Ot Pre}.Value
   )
   &RigUnit_ModifyTransforms(
    ExecutePin=()
    ItemToModify=((Item=(Type=Bone,Name="None"),Transform=(Translation=(X=0,Y=0,Z=0),Scale3D=(X=1,Y=1,Z=1))))
    WeightMinimum=0
    Weight=RigVMVariableNode() {variable=Thigh R Ba Pre}.Value
   )
   &RigUnit_ModifyTransforms(
    ExecutePin=()
    ItemToModify=((Item=(Type=Bone,Name="None"),Transform=(Translation=(X=0,Y=0,Z=0),Scale3D=(X=1,Y=1,Z=1))))
    WeightMinimum=0
    Weight=RigVMVariableNode() {variable=Thigh L Ba Pre}.Value
   )
   &RigUnit_ModifyTransforms(
    ExecutePin=()
    ItemToModify=((Item=(Type=Bone,Name="None"),Transform=(Translation=(X=0,Y=0,Z=0),Scale3D=(X=1,Y=1,Z=1))))
    WeightMinimum=0
    Weight=RigVMVariableNode() {variable=Thigh L Ot Pre}.Value
   )
   &RigUnit_ModifyTransforms(
    ExecutePin=()
    ItemToModify="((Item=(Type=Bone,Name=\"None\"),Transform=(Rotation=(X=0,Y=0,Z=0,W=1),Translation=(X=0,Y=0,Z=0),Scale3D=(X=1,Y=1,Z=1))),(Item=(Type=Bone,Name=\"def_belt_side_r\"),Transform=(Rotation=(X=0,Y=0,Z=0.4078,W=0.9131),Translation=(X=-7.8618,Y=0,Z=0),Scale3D=(X=1,Y=1 ...(truncated)"
    WeightMinimum=0
    Weight=@n12.Value
   )
  ]}
 RigVMModel Construction Event {nodes [
   n1 = &RigUnit_PrepareForExecution()
   RigVMFunction_Sequence = &RigVMFunction_Sequence(ExecuteContext=@n1.ExecutePin)
   RigVMFunction_ControlFlowBranch = &RigVMFunction_ControlFlowBranch(ExecuteContext=@RigVMFunction_Sequence.A, Condition=RigVMVariableNode() {variable=Debug Controls On}.Value)
   RigVMDispatchNode(ExecuteContext=@RigVMFunction_ControlFlowBranch.True, Array=RigVMVariableNode() {variable=Debug Ctrls}.Value) {template="DISPATCH_RigVMDispatch_ArrayReset(io Array)"}
   ItemArray_1 = RigVMDispatchNode(Value="((Type=Bone,Name=\"calf_l\"),(Type=Bone,Name=\"calf_r\"),(Type=Bone,Name=\"clavicle_l\"),(Type=Bone,Name=\"clavicle_r\"),(Type=Bone,Name=\"upperarm_l\"),(Type=Bone,Name=\"upperarm_r\"),(Type=Bone,Name=\"lowerarm_l\"),(Type=Bone,Name=\"lowerarm_r\"),(Type=Bone,Name=\"thigh_ ...(truncated)") {template="DISPATCH_RigVMDispatch_Constant(io Value)"}
   > RigVMVariableNode(Value=@ItemArray_1.Value) {variable=Debug Bones}
   > For_Each = RigVMDispatchNode(Array=@ItemArray_1.Value) {template="DISPATCH_RigVMDispatch_ArrayIterator(in Array,out Element,out Index,out Count,out Ratio)"}
   Concat = &RigVMFunction_NameConcat(B=" Ctrl", A=@For_Each.Element.Name) {template="Concat::Execute(in A,in B,out Result)"}
   > n2 = /Script/ControlRig.RigUnit_HierarchyAddControlTransform(
    InitialValue=(Rotation=(X=0,Y=0,Z=0,W=1),Translation=(X=0,Y=0,Z=0),Scale3D=(X=1,Y=1,Z=1))
    OffsetSpace=GlobalSpace
    Settings="(InitialSpace=LocalSpace,bUsePreferredRotationOrder=False,PreferredRotationOrder=YZX,Limits=(LimitTranslationX=(bMinimum=False,bMaximum=False),LimitTranslationY=(bMinimum=False,bMaximum=False),LimitTranslationZ=(bMinimum=False,bMaximum=False),LimitPitch=(b ...(truncated)"
    Name=@Concat.Result
    Parent=&RigUnit_HierarchyGetParent(bDefaultParent=True, Child=@For_Each.Element).Parent
    OffsetTransform=&RigUnit_GetTransform(bInitial=False, Item=@For_Each.Element).Transform
   ) {
    is=struct RigVMUnitNode
    template="SpawnControl::Execute(in InitialValue,in Settings,in OffsetTransform,in OffsetSpace,in Parent,in Name,out Item)"
   }
   > RigVMDispatchNode(Element=@n2.Item, Array=RigVMVariableNode() {variable=Debug Ctrls}.Value) {template="DISPATCH_RigVMDispatch_ArrayAdd(io Array,in Element,out Index)"}
   RigVMRerouteNode(Value=@RigVMFunction_Sequence.B)
  ]}
 SequenceAggregate_1_1_ContainedGraph {nodes [
   n1 = RigVMFunctionEntryNode()
   SequenceAggregate_1 = &RigVMFunction_Sequence(ExecuteContext=@n1.ExecuteContext)
   SequenceAggregate_1_1 = &RigVMFunction_Sequence(ExecuteContext=@SequenceAggregate_1.B)
   SequenceAggregate_1_2 = &RigVMFunction_Sequence(ExecuteContext=@SequenceAggregate_1_1.B)
   SequenceAggregate_1_3 = &RigVMFunction_Sequence(ExecuteContext=@SequenceAggregate_1_2.B)
   SequenceAggregate_1_4 = &RigVMFunction_Sequence(ExecuteContext=@SequenceAggregate_1_3.B)
   SequenceAggregate_1_5 = &RigVMFunction_Sequence(ExecuteContext=@SequenceAggregate_1_4.B)
   RigVMFunctionReturnNode(
    A=@SequenceAggregate_1.A
    B=@SequenceAggregate_1_1.A
    C=@SequenceAggregate_1_2.A
    D=@SequenceAggregate_1_3.A
    E=@SequenceAggregate_1_4.A
    F=@SequenceAggregate_1_5.A
    G=@SequenceAggregate_1_5.B
   )
  ]}
]
hierarchy [
 root {loc=<0 0 0>, rot=<0 0 0 1>}
 pelvis {parent=root, loc=<0 -0.7071 93.1191>, rot=<0.0127 0.707 -0.0127 -0.707>}
 spine_01 {parent=pelvis, loc=<6.7494 -0.6388 0>, rot=<0 0 0.034 -0.9994>}
 spine_02 {parent=spine_01, loc=<7.2518 -0.035 0>, rot=<0 0 0.0238 -0.9997>}
 spine_03 {parent=spine_02, loc=<6.0104 0 0>, rot=<0 0 0.0149 -0.9999>}
 spine_04 {parent=spine_03, loc=<8.1024 0 0>, rot=<0 0 -0.0132 -0.9999>}
 spine_05 {parent=spine_04, loc=<10.0146 0 0>, rot=<0 0 -0.0015 -1>}
 neck_01 {parent=spine_05, loc=<11.3138 0.1529 0>, rot=<0 0 -0.2003 -0.9797>}
 neck_02 {parent=neck_01, loc=<4.5627 0 0>, rot=<0 0 0.0504 -0.9987>}
 head {parent=neck_02, loc=<5.3914 0 0>, rot=<0 0 0.115 -0.9934>}
 attach_cape {parent=spine_05, loc=<9.017 -11.714 0>, rot=<0.0616 -0.7044 -0.0616 -0.7044>}
 cape {parent=attach_cape, type=Null, loc=<0 0 0>, rot=<0 0 0 1>}
 attach_backpack {parent=spine_05, loc=<-0.123 -13.473 0>, rot=<-0.0093 -0.707 0.0093 -0.707>}
 Backpack_BR {parent=attach_backpack, type=Null, loc=<0 0 0>, rot=<0 0 0 1>}
 clavicle_l {parent=spine_05, loc=<4.3575 -0.1466 -2.3223>, rot=<0.0465 -0.7289 -0.009 -0.6829>}
 upperarm_l {parent=clavicle_l, loc=<15.5111 0 0>, rot=<0.0226 -0.3234 -0.0088 -0.9459>}
 lowerarm_l {parent=upperarm_l, loc=<26.6514 0 0>, rot=<0.0391 -0.0099 -0.2453 -0.9686>}
 hand_l {parent=lowerarm_l, loc=<22.3491 0 0>, rot=<-0.7081 -0.0507 0.0975 -0.6975>}
 thumb_01_l {parent=hand_l, loc=<1.3118 -0.3191 -2.459>, rot=<0.6546 -0.4042 0.0664 -0.6354>}
 thumb_02_l {parent=thumb_01_l, loc=<5.9336 -0.0448 0.0401>, rot=<-0.0037 0.0213 0.298 -0.9543>}
 thumb_03_l {parent=thumb_02_l, loc=<2.2223 0.3682 -0.0907>, rot=<0.0118 -0.013 0.0066 -0.9998>}
 index_metacarpal_l {parent=hand_l, loc=<1.6535 0.0183 -2.96>, rot=<0.0158 -0.0124 -0.0173 -0.9996>}
 index_01_l {parent=index_metacarpal_l, loc=<8.3008 0.0357 -0.0524>, rot=<-0.0044 0.0485 0.0924 -0.9945>}
 index_02_l {parent=index_01_l, loc=<3.3587 -0.3004 0.0452>, rot=<-0.002 0.0027 0.3225 -0.9466>}
 index_03_l {parent=index_02_l, loc=<2.5578 -0.123 -0.0252>, rot=<-0.002 0.0067 0.2558 -0.9667>}
 middle_metacarpal_l {parent=hand_l, loc=<2.1856 -0.1725 -0.5984>, rot=<0.006 0.0168 -0.0329 -0.9993>}
 middle_01_l {parent=middle_metacarpal_l, loc=<7.3275 -0.2447 0.0129>, rot=<-0.0208 0.0042 0.0858 -0.9961>}
 middle_02_l {parent=middle_01_l, loc=<4.0123 0.1401 0.1175>, rot=<-0.0218 0.0258 0.3512 -0.9357>}
 middle_03_l {parent=middle_02_l, loc=<2.6077 0.0534 0.1373>, rot=<0.0082 0.0188 0.1708 -0.9851>}
 ring_metacarpal_l {parent=hand_l, loc=<2.2074 -0.3625 1.3638>, rot=<-0.0284 0.0722 -0.0458 -0.9959>}
 ring_01_l {parent=ring_metacarpal_l, loc=<7.0295 -0.6197 -0.1446>, rot=<-0.0489 -0.0252 0.1892 -0.9804>}
 ring_02_l {parent=ring_01_l, loc=<3.6174 0.4647 -0.1027>, rot=<0 0 0.2806 -0.9598>}
 ring_03_l {parent=ring_02_l, loc=<2.4152 0.114 0.0199>, rot=<0 0 0.1189 -0.9929>}
 pinky_metacarpal_l {parent=hand_l, loc=<1.7616 -0.9753 3.3814>, rot=<-0.0397 0.0814 -0.069 -0.9935>}
 pinky_01_l {parent=pinky_metacarpal_l, loc=<7.1709 -0.4831 -0.1094>, rot=<-0.0654 -0.0035 0.2242 -0.9723>}
 pinky_02_l {parent=pinky_01_l, loc=<2.8274 0.0893 -0.1043>, rot=<0.0037 0.0011 0.2854 -0.9584>}
 pinky_03_l {parent=pinky_02_l, loc=<2.2046 0.0307 -0.0078>, rot=<-0.0002 0.0098 0.1445 -0.9895>}
 lowerarm_twist_02_l {parent=lowerarm_l, loc=<9.6199 0 0>, rot=<0 0 0 -1>}
 lowerarm_twist_01_l {parent=lowerarm_l, loc=<17.8549 0 0>, rot=<0 0 0 -1>}
 upperarm_twist_01_l {parent=upperarm_l, loc=<0.0041 0 0>, rot=<0 0 0 -1>}
 def_upperarm_l_tp {parent=upperarm_twist_01_l, loc=<-5.6661 -0.1109 -0.585>, rot=<0 0 0 -1>}
 def_upperarm_l_bt {parent=upperarm_twist_01_l, loc=<7.4081 0.4996 -10.9429>, rot=<0 0 0 -1>}
 def_upperarm_l_fr {parent=upperarm_twist_01_l, loc=<-0.2653 8.0095 -4.5415>, rot=<0 0 0 -1>}
 def_upperarm_l_ba {parent=upperarm_twist_01_l, loc=<-1.18 -7.724 -5.7515>, rot=<0 0 0 -1>}
 upperarm_twist_02_l {parent=upperarm_l, loc=<13.7941 0 0>, rot=<0 0 0 -1>}
 def_sleeve_l_fr {parent=upperarm_twist_02_l, loc=<-2.2125 10.0989 -2.7786>, rot=<0 0 0 1>}
 def_sleeve_l_ba {parent=upperarm_twist_02_l, loc=<-0.7791 -9.9371 -1.7186>, rot=<0 0 0 1>}
 def_sleeve_l_tp {parent=upperarm_twist_02_l, loc=<-1.9613 -0.3572 6.6602>, rot=<0 0 0 1>}
 def_sleeve_l_bt {parent=upperarm_twist_02_l, loc=<-0.9701 1.1542 -10.0525>, rot=<0 0 0 1>}
 def_elbow_l {parent=upperarm_l, loc=<26.6514 0 0>, rot=<0.0391 -0.0099 -0.2453 -0.9686>}
 def_elbow_l_fr {parent=def_elbow_l, loc=<2.3705 8.9439 -1.468>, rot=<0 0 0 1>}
 def_elbow_l_fr_bt {parent=def_elbow_l_fr, loc=<0 0 0>, rot=<0 0 0 -1>}
 def_elbow_l_fr_bt_end {parent=def_elbow_l_fr_bt, loc=<3.5378 -1.1109 0.1497>, rot=<0 0 0 1>}
 def_elbow_l_fr_tp {parent=def_elbow_l_fr, loc=<0 0 0>, rot=<-0.0391 0.0099 0.2453 -0.9686>}
 def_elbow_l_fr_tp_end {parent=def_elbow_l_fr_tp, loc=<-3.7033 -0.5166 -0.1437>, rot=<0 0 0 1>}
 def_elbow_l_ba {parent=def_elbow_l, loc=<-1.3081 -5.3937 -2.3231>, rot=<0 0 0 1>}
 def_elbow_l_ba_bt {parent=def_elbow_l_ba, loc=<0 0 0>, rot=<0 0 0 -1>}
 def_elbow_l_ba_bt_end {parent=def_elbow_l_ba_bt, loc=<4.3551 -0.9125 0.2119>, rot=<0 0 0 1>}
 def_elbow_l_ba_tp {parent=def_elbow_l_ba, loc=<0 0 0>, rot=<-0.0391 0.0099 0.2453 -0.9686>}
 def_elbow_l_ba_tp_end {parent=def_elbow_l_ba_tp, loc=<-4.4225 -0.7795 -0.1379>, rot=<0 0 0 1>}
 def_clavicle_l_tp {parent=clavicle_l, loc=<10.6804 -0.1509 3.0091>, rot=<0 0 0 1>}
 def_clavicle_l_fr {parent=clavicle_l, loc=<12.2763 7.7792 -3.7236>, rot=<0 0 0 1>}
 def_clavicle_l_ba {parent=clavicle_l, loc=<11.3063 -7.9963 -3.5358>, rot=<0 0 0 1>}
 clavicle_r {parent=spine_05, loc=<4.3575 -0.1466 2.3223>, rot=<0.7289 0.0465 -0.6829 0.009>}
 upperarm_r {parent=clavicle_r, loc=<-15.5111 0 0>, rot=<0.0226 -0.3234 -0.0088 -0.9459>}
 lowerarm_r {parent=upperarm_r, loc=<-26.6514 0 0>, rot=<0.0391 -0.0099 -0.2453 -0.9686>}
 hand_r {parent=lowerarm_r, loc=<-22.3491 0 0>, rot=<-0.7081 -0.0507 0.0975 -0.6975>}
 thumb_01_r {parent=hand_r, loc=<-1.3118 0.3191 2.459>, rot=<0.6546 -0.4042 0.0664 -0.6354>}
 thumb_02_r {parent=thumb_01_r, loc=<-5.9336 0.0448 -0.0401>, rot=<-0.0037 0.0213 0.298 -0.9543>}
 thumb_03_r {parent=thumb_02_r, loc=<-2.2223 -0.3682 0.0907>, rot=<0.0118 -0.013 0.0066 -0.9998>}
 index_metacarpal_r {parent=hand_r, loc=<-1.6535 -0.0183 2.96>, rot=<0.0158 -0.0124 -0.0173 -0.9996>}
 index_01_r {parent=index_metacarpal_r, loc=<-8.3008 -0.0357 0.0524>, rot=<-0.0044 0.0485 0.0924 -0.9945>}
 index_02_r {parent=index_01_r, loc=<-3.3587 0.3004 -0.0452>, rot=<-0.002 0.0027 0.3225 -0.9466>}
 index_03_r {parent=index_02_r, loc=<-2.5578 0.123 0.0252>, rot=<-0.002 0.0067 0.2558 -0.9667>}
 middle_metacarpal_r {parent=hand_r, loc=<-2.1856 0.1725 0.5984>, rot=<0.006 0.0168 -0.0329 -0.9993>}
 middle_01_r {parent=middle_metacarpal_r, loc=<-7.3275 0.2447 -0.0129>, rot=<-0.0208 0.0042 0.0858 -0.9961>}
 middle_02_r {parent=middle_01_r, loc=<-4.0123 -0.1401 -0.1175>, rot=<-0.0218 0.0258 0.3512 -0.9357>}
 middle_03_r {parent=middle_02_r, loc=<-2.6077 -0.0534 -0.1373>, rot=<0.0082 0.0188 0.1708 -0.9851>}
 ring_metacarpal_r {parent=hand_r, loc=<-2.2074 0.3625 -1.3638>, rot=<-0.0284 0.0722 -0.0458 -0.9959>}
 ring_01_r {parent=ring_metacarpal_r, loc=<-7.0295 0.6197 0.1446>, rot=<-0.0489 -0.0252 0.1892 -0.9804>}
 ring_02_r {parent=ring_01_r, loc=<-3.6174 -0.4647 0.1027>, rot=<0 0 0.2806 -0.9598>}
 ring_03_r {parent=ring_02_r, loc=<-2.4152 -0.114 -0.0199>, rot=<0 0 0.1189 -0.9929>}
 pinky_metacarpal_r {parent=hand_r, loc=<-1.7616 0.9753 -3.3814>, rot=<-0.0397 0.0814 -0.069 -0.9935>}
 pinky_01_r {parent=pinky_metacarpal_r, loc=<-7.1709 0.4831 0.1094>, rot=<-0.0654 -0.0035 0.2242 -0.9723>}
 pinky_02_r {parent=pinky_01_r, loc=<-2.8274 -0.0893 0.1043>, rot=<0.0037 0.0011 0.2854 -0.9584>}
 pinky_03_r {parent=pinky_02_r, loc=<-2.2046 -0.0307 0.0078>, rot=<-0.0002 0.0098 0.1445 -0.9895>}
 lowerarm_twist_02_r {parent=lowerarm_r, loc=<-9.6199 0 0>, rot=<0 0 0 -1>}
 lowerarm_twist_01_r {parent=lowerarm_r, loc=<-17.8549 0 0>, rot=<0 0 0 -1>}
 upperarm_twist_01_r {parent=upperarm_r, loc=<-0.0041 0 0>, rot=<0 0 0 -1>}
 def_upperarm_r_tp {parent=upperarm_twist_01_r, loc=<5.6661 0.1109 0.585>, rot=<0 0 0 -1>}
 def_upperarm_r_bt {parent=upperarm_twist_01_r, loc=<-7.4081 -0.4996 10.9429>, rot=<0 0 0 -1>}
 def_upperarm_r_fr {parent=upperarm_twist_01_r, loc=<0.2653 -8.0095 4.5415>, rot=<0 0 0 -1>}
 def_upperarm_r_ba {parent=upperarm_twist_01_r, loc=<1.18 7.724 5.7515>, rot=<0 0 0 -1>}
 upperarm_twist_02_r {parent=upperarm_r, loc=<-13.7941 0 0>, rot=<0 0 0 -1>}
 def_sleeve_r_fr {parent=upperarm_twist_02_r, loc=<2.2145 -10.0989 2.7726>, rot=<0 0 0 1>}
 def_sleeve_r_ba {parent=upperarm_twist_02_r, loc=<0.7893 9.9889 1.6437>, rot=<0 0 0 1>}
 def_sleeve_r_tp {parent=upperarm_twist_02_r, loc=<2.1174 0.3246 -6.7907>, rot=<0 0 0 1>}
 def_sleeve_r_bt {parent=upperarm_twist_02_r, loc=<0.9701 -1.1542 10.0525>, rot=<0 0 0 1>}
 def_elbow_r {parent=upperarm_r, loc=<-26.6514 0 0>, rot=<0.0391 -0.0099 -0.2453 -0.9686>}
 def_elbow_r_fr {parent=def_elbow_r, loc=<-2.3705 -8.9439 1.468>, rot=<0 0 0 1>}
 def_elbow_r_fr_tp {parent=def_elbow_r_fr, loc=<0 0 0>, rot=<0 0 0 1>}
 def_elbow_r_fr_tp_end {parent=def_elbow_r_fr_tp, loc=<3.4966 -1.3198 0.1848>, rot=<0 0 0 1>}
 def_elbow_r_fr_bt {parent=def_elbow_r_fr, loc=<0 0 0>, rot=<0 0 0 1>}
 def_elbow_r_fr_bt_end {parent=def_elbow_r_fr_bt, loc=<-3.5378 1.1109 -0.1497>, rot=<0 0 0 1>}
 def_elbow_r_ba {parent=def_elbow_r, loc=<1.3081 5.3937 2.3231>, rot=<0 0 0 1>}
 def_elbow_r_ba_tp {parent=def_elbow_r_ba, loc=<0 0 0>, rot=<0 0 0 1>}
 def_elbow_r_ba_tp_end {parent=def_elbow_r_ba_tp, loc=<4.2541 -1.4311 0.2002>, rot=<0 0 0 1>}
 def_elbow_r_ba_bt {parent=def_elbow_r_ba, loc=<0 0 0>, rot=<0 0 0 1>}
 def_elbow_r_ba_bt_end {parent=def_elbow_r_ba_bt, loc=<-4.3551 0.9125 -0.2119>, rot=<0 0 0 1>}
 def_clavicle_r_tp {parent=clavicle_r, loc=<-10.6804 0.1509 -3.0091>, rot=<0 0 0 1>}
 def_clavicle_r_fr {parent=clavicle_r, loc=<-12.2763 -7.7792 3.7236>, rot=<0 0 0 1>}
 def_clavicle_r_ba {parent=clavicle_r, loc=<-11.3063 7.9963 3.5358>, rot=<0 0 0 1>}
 def_collar_l {parent=spine_05, loc=<7.9226 -0.0545 -10.5538>, rot=<0.0465 -0.7289 -0.009 -0.6829>}
 def_collar_r {parent=spine_05, loc=<7.9226 -0.0545 10.5538>, rot=<0.7289 0.0465 -0.6829 0.009>}
 def_clavicle_l_bt {parent=spine_04, loc=<0.3271 -2.0161 -16.0493>, rot=<0.0476 -0.7289 -0.01 -0.6829>}
 def_clavicle_r_bt {parent=spine_04, loc=<0.3271 -2.0161 16.0493>, rot=<0.7289 0.0476 -0.6829 0.01>}
 thigh_l {parent=pelvis, loc=<-2.5693 -0.0109 -7.7501>, rot=<0.0012 0.0572 0.0155 -0.9982>}
 calf_l {parent=thigh_l, loc=<-39.0484 0 0>, rot=<0.0006 -0.0161 -0.0391 -0.9991>}
 foot_l {parent=calf_l, loc=<-42.561 0.1077 0.4292>, rot=<0.0418 -0.0455 0.0343 -0.9975>}
 ball_l {parent=foot_l, loc=<-6.1599 13.8264 0.6806>, rot=<0.0001 -0.0001 -0.7063 -0.7079>}
 toe_slope_01_l {parent=ball_l, loc=<0 0 0>, rot=<0 0 -0.1594 -0.9872>}
 toe_slope_02_l {parent=toe_slope_01_l, loc=<8.93 0 0>, rot=<0 0 0.1632 -0.9866>}
 toe_slope_03_l {parent=toe_slope_02_l, loc=<8 0 0>, rot=<0 0 0 1>}
 calf_twist_02_l {parent=calf_l, loc=<-18.8064 -0.0967 0.0397>, rot=<0.0237 0.0095 -0.0035 -0.9997>}
 calf_twist_01_l {parent=calf_l, loc=<-35.813 -0.0967 0.0397>, rot=<0.0237 0.0095 -0.0035 -0.9997>}
 thigh_twist_01_l {parent=thigh_l, loc=<-21.6303 0 0>, rot=<0.0013 0 0 -1>}
 def_thigh_l_push {parent=thigh_twist_01_l, loc=<0 0 0>, rot=<0 0 0 1>}
 def_knee_l {parent=thigh_l, loc=<-39.0484 0 0>, rot=<0.0006 -0.0161 -0.0391 -0.9991>}
 def_knee_l_fr {parent=def_knee_l, loc=<2.3881 12.9388 1.1483>, rot=<0 0 0 1>}
 def_knee_l_fr_tp {parent=def_knee_l_fr, loc=<0 0 0>, rot=<-0.0006 0.0161 0.0391 -0.9991>}
 def_knee_l_fr_tp_end {parent=def_knee_l_fr_tp, loc=<5.6659 0.1198 -0.4497>, rot=<0 0 0 1>}
 def_knee_l_fr_bt {parent=def_knee_l_fr, loc=<0 0 0>, rot=<0 0 0 1>}
 def_knee_l_fr_bt_end {parent=def_knee_l_fr_bt, loc=<-5.8369 0.3287 0.2373>, rot=<0 0 0 1>}
 def_knee_l_ba {parent=def_knee_l, loc=<-0.0857 -15.0419 -0.719>, rot=<0 0 0 1>}
 def_knee_l_ba_tp {parent=def_knee_l_ba, loc=<0 0 0>, rot=<-0.0006 0.0161 0.0391 -0.9991>}
 def_knee_l_ba_tp_end {parent=def_knee_l_ba_tp, loc=<5.6553 0.1009 -0.5426>, rot=<0 0 0 1>}
 def_knee_l_ba_bt {parent=def_knee_l_ba, loc=<0 0 0>, rot=<0 0 0 1>}
 def_knee_l_ba_bt_end {parent=def_knee_l_ba_bt, loc=<-5.8299 0.3573 0.2933>, rot=<0 0 0 1>}
 thigh_r {parent=pelvis, loc=<-2.5693 -0.0109 7.7501>, rot=<0.0572 -0.0012 0.9982 0.0155>}
 calf_r {parent=thigh_r, loc=<39.0484 0 0>, rot=<0.0006 -0.0161 -0.0391 -0.9991>}
 foot_r {parent=calf_r, loc=<42.561 -0.1077 -0.4292>, rot=<0.0418 -0.0455 0.0343 -0.9975>}
 ball_r {parent=foot_r, loc=<6.1599 -13.8264 -0.6806>, rot=<0.0001 -0.0001 -0.7063 -0.7079>}
 toe_slope_01_r {parent=ball_r, loc=<0 0 0>, rot=<0 0 -0.1594 -0.9872>}
 toe_slope_02_r {parent=toe_slope_01_r, loc=<-8.93 0 0>, rot=<0 0 0.1632 -0.9866>}
 toe_slope_03_r {parent=toe_slope_02_r, loc=<-8 0 0>, rot=<0 0 0 1>}
 calf_twist_02_r {parent=calf_r, loc=<18.8064 0.0967 -0.0397>, rot=<0.0237 0.0095 -0.0035 -0.9997>}
 calf_twist_01_r {parent=calf_r, loc=<35.813 0.0967 -0.0397>, rot=<0.0237 0.0095 -0.0035 -0.9997>}
 thigh_twist_01_r {parent=thigh_r, loc=<21.6303 0 0>, rot=<0.0013 0 0 -1>}
 def_thigh_r_push {parent=thigh_twist_01_r, loc=<0 0 0>, rot=<0 0 0 1>}
 def_knee_r {parent=thigh_r, loc=<39.0484 0 0>, rot=<0.0006 -0.0161 -0.0391 -0.9991>}
 def_knee_r_fr {parent=def_knee_r, loc=<-2.3881 -12.9388 -1.1483>, rot=<0 0 0 1>}
 def_knee_r_fr_tp {parent=def_knee_r_fr, loc=<0 0 0>, rot=<-0.0006 0.0161 0.0391 -0.9991>}
 def_knee_r_fr_tp_end {parent=def_knee_r_fr_tp, loc=<-5.6659 -0.1198 0.4497>, rot=<0 0 0 1>}
 def_knee_r_fr_bt {parent=def_knee_r_fr, loc=<0 0 0>, rot=<0 0 0 1>}
 def_knee_r_fr_bt_end {parent=def_knee_r_fr_bt, loc=<5.8369 -0.3287 -0.2373>, rot=<0 0 0 1>}
 def_knee_r_ba {parent=def_knee_r, loc=<0.0857 15.0419 0.719>, rot=<0 0 0 1>}
 def_knee_r_ba_tp {parent=def_knee_r_ba, loc=<0 0 0>, rot=<-0.0006 0.0161 0.0391 -0.9991>}
 def_knee_r_ba_tp_end {parent=def_knee_r_ba_tp, loc=<-5.6553 -0.1009 0.5426>, rot=<0 0 0 1>}
 def_knee_r_ba_bt {parent=def_knee_r_ba, loc=<0 0 0>, rot=<0 0 0 1>}
 def_knee_r_ba_bt_end {parent=def_knee_r_ba_bt, loc=<5.8299 -0.3573 -0.2933>, rot=<0 0 0 1>}
 pelvis_l {parent=pelvis, loc=<-2.5693 -0.0109 -7.7501>, rot=<0.0012 0.0572 0.0155 -0.9982>}
 pelvis_r {parent=pelvis, loc=<-2.5693 -0.0109 7.7501>, rot=<0.0572 -0.0012 0.9982 0.0155>}
 belt {parent=pelvis, loc=<0 0 0>, rot=<0 0 0 1>}
 def_belt_fr_l {parent=pelvis, loc=<0.7544 20.8309 -11.5341>, rot=<0 0 0 1>}
 def_belt_fr_r {parent=pelvis, loc=<0.7544 20.8309 11.5341>, rot=<0 0 0 1>}
 def_belt_ba_l {parent=pelvis, loc=<-0.666 -22.3416 -10.6008>, rot=<0 0 0 1>}
 def_belt_ba_r {parent=pelvis, loc=<-0.666 -22.3416 10.6008>, rot=<0 0 0 1>}
 def_belt_side_l {parent=pelvis, loc=<-0.0998 -1.9691 -24.8124>, rot=<0 0 0 1>}
 def_belt_side_r {parent=pelvis, loc=<-0.0998 -1.9691 24.8124>, rot=<0 0 0 1>}
 def_crease_l_fr {parent=pelvis, loc=<-18.6744 14.1684 -10.4041>, rot=<0 0 0 -1>}
 def_crease_l_ba {parent=pelvis, loc=<-17.0001 -16.1093 -8.8719>, rot=<0.0025 0.0571 0.0155 -0.9982>}
 def_crease_l_ot {parent=pelvis, loc=<-14.0274 -1.839 -22.2805>, rot=<0 0 0 -1>}
 def_crease_r_ba {parent=pelvis, loc=<-16.7979 -16.9206 6.6896>, rot=<0 0 1 0>}
 def_crease_r_fr {parent=pelvis, loc=<-18.6744 14.1684 10.4041>, rot=<0 0 1 0>}
 def_crease_r_ot {parent=pelvis, loc=<-14.0274 -1.839 22.2805>, rot=<0 0 1 0>}
 def_shirt_fr_l {parent=pelvis, loc=<-4.56 21.9876 -11.5668>, rot=<0 0 0 1>}
 def_shirt_fr_r {parent=pelvis, loc=<-4.56 21.9876 11.567>, rot=<0 0 0 1>}
 def_shirt_ba_l {parent=pelvis, loc=<-4.7366 -23.4509 -10.9992>, rot=<0 0 0 1>}
 def_shirt_ba_r {parent=pelvis, loc=<-4.7366 -23.4509 10.9992>, rot=<0 0 0 1>}
 def_shirt_side_l {parent=pelvis, loc=<-4.6773 -1.8845 -25.7358>, rot=<0 0 0 1>}
 def_shirt_side_r {parent=pelvis, loc=<-4.6773 -1.8845 25.736>, rot=<0 0 0 1>}
 def_crease_l_in {parent=pelvis, loc=<-23.1466 -0.1923 -0.6515>, rot=<0.0025 0.0571 0.0155 -0.9982>}
 def_crease_r_in {parent=pelvis, loc=<-23.1466 -0.1923 0.6515>, rot=<0.0571 -0.0025 0.9982 0.0155>}
]
variables [
 Debug Bones {type="TArray<FRigElementKey>", default=((Type=Bone,Name="calf_l"),(Type=Bone,Name="calf_r"))}
 Debug Controls On {type=bool, default=False}
 Debug Ctrls {type="TArray<FRigElementKey>", default=(())}
 GauntletElbowAdjustAlpha {type=double, default=0}
 IsNotEmotingAlpha {type=double, default=0}
 ItemToModify_4_Transform {type=FTransform, default=(Rotation=(X=0,Y=0,Z=0,W=1),Translation=(X=0,Y=0,Z=0),Scale3D=(X=1,Y=1,Z=1))}
 ModifyTransforms_2_ItemToModify {
  type="TArray<FRigUnit_ModifyTransforms_PerItem>"
  default="((Item=(Name=\"def_thigh_l_fr\"),Transform=(Rotation=(X=0,Y=0,Z=0.0436,W=0.999),Translation=(X=-17.4817,Y=0.7394,Z=0))),(Item=(Name=\"def_crease_l_fr\"),Transform=(Rotation=(X=0,Y=0,Z=-0.7934,W=0.6088),Translation=(X=29.9125,Y=5.8217,Z=11.1686),Scale3D=(X=0.5, ...(truncated)"
 }
 SPR_upperarm_l_down_A_alpha {type=double, default=0}
 SPR_upperarm_r_down_A_alpha {type=double, default=0}
 SPR_upperarm_r_down_B_alpha {type=double, default=0}
 Set_Metadata_Item {type=FRigElementKey, default=(Type=Bone,Name="thigh_l")}
 Thigh L Ba {type=double, default=0}
 Thigh L Ba Pre {type=double, default=0}
 Thigh L Crouch {type=double, default=0}
 Thigh L Crouch Ot {type=double, default=0}
 Thigh L Crouch Ot Post {type=double, default=0}
 Thigh L Crouch Ot Pre {type=double, default=0}
 Thigh L Crouch Post {type=double, default=0}
 Thigh L Crouch Pre {type=double, default=0}
 Thigh L Fr {type=double, default=0}
 Thigh L Ot {type=double, default=0}
 Thigh L Ot Post {type=double, default=0}
 Thigh L Ot Pre {type=double, default=0}
 Thigh L Post Fr {type=double, default=0}
 Thigh L Pre Fr {type=double, default=0}
 Thigh R Ba {type=double, default=0}
 Thigh R Ba Pre {type=double, default=0}
 Thigh R Crouch {type=double, default=0}
 Thigh R Crouch Ot {type=double, default=0}
 Thigh R Crouch Post {type=double, default=0}
 Thigh R Crouch Post Ot {type=double, default=0}
 Thigh R Crouch Pre {type=double, default=0}
 Thigh R Crouch Pre Ot {type=double, default=0}
 Thigh R Fr {type=double, default=0}
 Thigh R Fr Post {type=double, default=0}
 Thigh R Fr Pre {type=double, default=0}
 Thigh R Ot {type=double, default=0}
 Thigh R Ot Post {type=double, default=0}
 Thigh R Ot Pre {type=double, default=0}
 mult_spr_dst_hand_r_thigh_twist_r_alpha {type=double, default=0}
 mult_spr_dst_lowerarm_r_pelvis_alpha {type=double, default=0}
]
