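# Post-process style AnimBlueprint for the M_MED_ClimbSpill male player body, copying the attached parent mesh's pose into a body-specific Control Rig.
# Each update sets two rig alphas: one drops to zero during emotes, the other (gauntlet elbow adjust) when MeleeDualWield is active and both arms are hidden.
# The weapon type comes from the owner mesh's FortPlayerAnimInstance.
# A pelvis rotation-constraint chain sits unconnected to the output pose, at zero alpha.
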AnimBlueprint /BRCosmetics/Characters/Player/Male/Medium/Bodies/M_MED_ClimbSpill/Meshes/M_MED_ClimbSpill_AnimBP
target_skeleton=&/M_MED_ClimbSpill_Skeleton
graphs [
 AnimGraph {nodes [
   n1 = "Constraint - pelvis_l (RRR)"!(ComponentPose=Local To Component(), Alpha=0, ConstraintWeights_0=0.375, ConstraintWeights_1:float=0.5, ConstraintWeights_2:float=0.125) -> (Pose) {
    boneToModify {boneName=pelvis_l}
    constraintSetup [
     {transformType=Rotation, targetBone {boneName=pelvis}}
     {transformType=Rotation, targetBone {boneName=thigh_l}}
     {transformType=Rotation, targetBone {boneName=thigh_r}}
    ]
    constraintWeights [1,1,1]
   }
   n2 = "Constraint - pelvis_r (RRR)"!(ComponentPose=@n1, Alpha=0, ConstraintWeights_0=0.375, ConstraintWeights_1:float=0.125, ConstraintWeights_2:float=0.5) -> (Pose) {
    boneToModify {boneName=pelvis_r}
    constraintSetup [
     {transformType=Rotation, targetBone {boneName=pelvis}}
     {transformType=Rotation, targetBone {boneName=thigh_l}}
     {transformType=Rotation, targetBone {boneName=thigh_r}}
    ]
    constraintWeights [1,1,1]
   }
   Component To Local(ComponentPose=@n2)
   n3 = Control Rig!(
    Source=Copy Pose From Mesh() {bUseAttachedParent=true}
    Alpha
    IsNotEmotingAlpha:double=Get Is Not Emoting Alpha() -> (Is Not Emoting Alpha:double)
    GauntletElbowAdjustAlpha:double=Get GauntletElbowAdjustAlpha() -> (GauntletElbowAdjustAlpha:double)
   ) -> (Pose) {controlRigAssetReference {blueprintRigClass=&/M_MED_ClimbSpill_CtrlRig.M_MED_ClimbSpill_CtrlRig_C}}
   Output Pose(Result=@n3) {layerGroup=None}
  ]}
 EventGraph {nodes [
   Event BlueprintUpdateAnimation()
   > n1 = Sequence!(execute) -> (then_0, then_1, then_2:Exec)
   n2 = Get bIsPlayingEmote()
   > Set Is Not Emoting Alpha(Is Not Emoting Alpha:double="To Float (Boolean)"(InBool=NOT Boolean(A=@n2))) -> (Output_Get:double)
   n3 = Cast To FortPlayerAnimInstance(execute=@n1.then_1, Object=GetAnimInstance(self=GetOwnerSkeletalMeshComponent()))
   > Set Is Gauntlet(Is Gauntlet:bool=false) -> (Output_Get:bool)
   > n4 = Switch on EFortWeaponCoreAnimation!(execute, Selection:EFortWeaponCoreAnimation Enum=Get WeaponCoreAnim(self=@n3)) -> (
    Melee:Exec
    Pistol:Exec
    Shotgun:Exec
    PaperBlueprint:Exec
    Rifle:Exec
    MeleeOneHand:Exec
    MachinePistol:Exec
    RocketLauncher:Exec
    GrenadeLauncher:Exec
    GoingCommando:Exec
    AssaultRifle:Exec
    TacticalShotgun:Exec
    SniperRifle:Exec
    TrapPlacement:Exec
    ShoulderLauncher:Exec
    AbilityDecoTool:Exec
    Crossbow:Exec
    C4:Exec
    RemoteControl:Exec
    DualWield:Exec
    AR_BullPup:Exec
    AR_ForwardGrip:Exec
    MedPackPaddles:Exec
    SMG_P90:Exec
    AR_DrumGun:Exec
    Consumable_Small:Exec
    Consumable_Large:Exec
    Balloon:Exec
    MountedTurret:Exec
    CreativeTool:Exec
    ExplosiveBow:Exec
    AshtonIndigo:Exec
    AshtonChicago:Exec
    MeleeDualWield:Exec
    Unarmed:Exec
   )
   Set Is Gauntlet(execute=@n4.MeleeDualWield, Is Gauntlet=true)
   n5 = AND Boolean!(
    A=Get Is Gauntlet() -> (Is Gauntlet:bool)
    B=Get bIsLeftArmHidden() -> (bIsLeftArmHidden:bool)
    C:bool=Get bIsRightArmHidden() -> (bIsRightArmHidden:bool)
   ) -> (ReturnValue)
   Set GauntletElbowAdjustAlpha(execute=@n1.then_2, GauntletElbowAdjustAlpha:double="To Float (Boolean)"(InBool=NOT Boolean(A=@n5))) -> (Output_Get:double)
  ]}
]
variables [
 GauntletElbowAdjustAlpha {type=number}
 Is Gauntlet {type=boolean}
 Is Not Emoting Alpha {type=number}
]
