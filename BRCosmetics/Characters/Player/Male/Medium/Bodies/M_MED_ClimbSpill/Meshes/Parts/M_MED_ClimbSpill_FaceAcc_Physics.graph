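# Physics asset for the face accessory mesh of the ClimbSpill male medium cosmetic body, covering left and right dynamic hair bone chains.
# Each side has a spherical zero-length base capsule followed by three capsules that shrink toward the tip.
# Constraints link each bone to its parent, with tight swing at the root, wider swing down the chain and nearly locked twist.
# Right-side shapes duplicate the left, including their left-side shape names.
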PhysicsAsset /BRCosmetics/Characters/Player/Male/Medium/Bodies/M_MED_ClimbSpill/Meshes/Parts/M_MED_ClimbSpill_FaceAcc_Physics
bodies {
 dyn_hair_l_01 [{shapeType=Capsule, length=4.2587, radius=6.7914, shapeName=dyn_hair_l_01_capsule, center=<5.0998 0 0>, rotation=<rot 90 0 0>}]
 dyn_hair_l_02 [{shapeType=Capsule, length=4.3947, radius=4.8559, shapeName=dyn_hair_l_02_capsule, center=<4.4585 0 0>, rotation=<rot 90 0 0>}]
 dyn_hair_l_03 [{shapeType=Capsule, length=2.6747, radius=2.382, shapeName=dyn_hair_l_03_capsule, center=<2.7517 0 0>, rotation=<rot 90 0 0>}]
 dyn_hair_l_base [{shapeType=Capsule, length=0, radius=8, shapeName=dyn_hair_l_base_capsule, rotation=<rot 0 -90 0>}]
 dyn_hair_r_01 [{shapeType=Capsule, length=4.2587, radius=6.7914, shapeName=dyn_hair_l_01_capsule, center=<5.0998 0 0>, rotation=<rot 90 0 0>}]
 dyn_hair_r_02 [{shapeType=Capsule, length=4.3947, radius=4.8559, shapeName=dyn_hair_l_02_capsule, center=<4.4585 0 0>, rotation=<rot 90 0 0>}]
 dyn_hair_r_03 [{shapeType=Capsule, length=2.6747, radius=2.382, shapeName=dyn_hair_l_03_capsule, center=<2.7517 0 0>, rotation=<rot 90 0 0>}]
 dyn_hair_r_base [{shapeType=Capsule, length=0, radius=8, shapeName=dyn_hair_l_base_capsule, rotation=<rot 0 -90 0>}]
}
constraints [
 {bone1Name=dyn_hair_l_01, bone2Name=dyn_hair_l_base, swing1LimitDegrees=10, swing2LimitDegrees=10, twistLimitDegrees=1}
 {bone1Name=dyn_hair_l_02, bone2Name=dyn_hair_l_01, swing1LimitDegrees=45, swing2LimitDegrees=45, twistLimitDegrees=1}
 {bone1Name=dyn_hair_l_03, bone2Name=dyn_hair_l_02, swing1LimitDegrees=45, swing2LimitDegrees=45, twistLimitDegrees=1}
 {bone1Name=dyn_hair_r_01, bone2Name=dyn_hair_r_base, swing1LimitDegrees=10, swing2LimitDegrees=10, twistLimitDegrees=1}
 {bone1Name=dyn_hair_r_02, bone2Name=dyn_hair_r_01, swing1LimitDegrees=45, swing2LimitDegrees=45, twistLimitDegrees=1}
 {bone1Name=dyn_hair_r_03, bone2Name=dyn_hair_r_02, swing1LimitDegrees=45, swing2LimitDegrees=45, twistLimitDegrees=1}
]
