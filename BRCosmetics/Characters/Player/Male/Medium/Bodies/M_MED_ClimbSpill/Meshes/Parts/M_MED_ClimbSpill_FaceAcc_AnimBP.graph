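# Post-process style animation Blueprint for the ClimbSpill character's face accessory, adding secondary physics motion.
# It copies the attached parent mesh's pose, then runs a RigidBody node simulated in neck_02 bone space with overridden gravity and clamped inertia.
# Hair Alpha blends the simulation, dropping low during tactical sprint and restoring afterward.
# Jump and fall begin/end events re-apply that alpha according to sprint state.
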AnimBlueprint /BRCosmetics/Characters/Player/Male/Medium/Bodies/M_MED_ClimbSpill/Meshes/Parts/M_MED_ClimbSpill_FaceAcc_AnimBP
target_skeleton=&/M_MED_ClimbSpill_FaceAcc_Skeleton
graphs [
 AnimGraph {nodes [
   n1 = RigidBody(ComponentPose=Local To Component(LocalPose=Copy Pose From Mesh() {bUseAttachedParent=true}), Alpha=Get Hair Alpha() -> (Hair Alpha:float)) {
    bClampLinearTranslationLimitToRefPose=true
    bOverrideWorldGravity=true
    overridePhysicsAsset=&/M_MED_ClimbSpill_FaceAcc_Physics
    simulationSpace=BaseBoneSpace
    baseBoneRef {boneName=neck_02}
    overrideWorldGravity {z=-980}
    simSpaceSettings {dampingAlpha=0, maxAngularAcceleration=50, maxAngularVelocity=5, maxLinearAcceleration=50, maxLinearVelocity=80, worldAlpha=0.05}
   }
   Output Pose(Result=Component To Local(ComponentPose=@n1)) {layerGroup=None}
  ]}
 EventGraph {nodes [
   n1 = Event OnBeginJumping()
   n2 = Branch(execute=[@n1.then, @n5.then], Condition=Get bIsTacticalSprinting())
   > Set Hair Alpha(Hair Alpha:float=0.1) -> (Output_Get:float)
   Set Hair Alpha(execute=@n2.else, Hair Alpha=0.67)
   n3 = Event OnEndJumping()
   n4 = Branch(execute=[@n3.then, @n6.then], Condition=Get bIsTacticalSprinting())
   > Set Hair Alpha(Hair Alpha=0.1)
   Set Hair Alpha(execute=[@n7.then, @n4.else], Hair Alpha=0.67)
   n5 = Event OnBeginFalling()
   n6 = Event OnEndFalling()
   Event OnBeginTacticalSprinting() {is=title K2Node_Event}
   > Set Hair Alpha(Hair Alpha=0.1)
   n7 = Event OnEndTacticalSprinting() {is=title K2Node_Event}
   Event BlueprintUpdateAnimation()
  ]}
]
variables [Hair Alpha {type=number, default=0.67}]
